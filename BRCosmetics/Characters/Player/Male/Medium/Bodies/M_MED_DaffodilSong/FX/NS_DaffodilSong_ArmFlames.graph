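# Niagara cosmetic effect that puts mesh flames on the arms of the DaffodilSong player body.
# Two persistent burst particles follow two skeletal sockets, partition index picking per-flame values, with the second blending both sockets' rotations.
# System-level hand-hidden, distance and smoothed values drive scale, while distance fade and time of day tint alpha and color.
# Smoothed, clamped socket velocity is stored for flame motion, alongside dynamic material parameters and front-end-specific position offsets.
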NiagaraSystem /BRCosmetics/Characters/Player/Male/Medium/Bodies/M_MED_DaffodilSong/FX/NS_DaffodilSong_ArmFlames
dependencies {
 usedDataInterfaces [
  &NiagaraDataInterfaceArrayFloat3
  &NiagaraDataInterfaceArrayInt32
  &NiagaraDataInterfaceArrayFloat
  &NiagaraDataInterfaceSkeletalMesh
 ]
 usedDynamicInputs [
  &MakeCustomFloatFromBool
  &SetBoolByFloatComparison
  &VectorLength
 ]
 usedModules [
  &self:SystemSpawnScript.SystemScriptSource.SystemScriptGraph.NiagaraNodeAssignment_0.SetVariables_#1
  &self:SystemSpawnScript.SystemScriptSource.SystemScriptGraph.NiagaraNodeAssignment_1.SetVariables_#2
  &SystemState
  &self:SystemSpawnScript.SystemScriptSource.SystemScriptGraph.NiagaraNodeAssignment_2.SetVariables_#3
  &self:SystemSpawnScript.SystemScriptSource.SystemScriptGraph.NiagaraNodeAssignment_3.SetVariables_#4
  &self:SystemSpawnScript.SystemScriptSource.SystemScriptGraph.NiagaraNodeAssignment_4.SetVariables_#5
 ]
}
emitters {NE_ArmFlames {
  emitterUpdateScript [
   &EmitterState(
    Inactive Response=::NewEnumerator1
    Life Cycle Mode=::NewEnumerator1
    Loop Behavior=::NewEnumerator1
    Loop Duration Mode=::NewEnumerator1
    UseLoopDelay=false
   )
   &SpawnBurst_Instantaneous(Spawn Count=2, Use Spawn Probability=false, UseLoopCountLimit=false)
  ]
  particleSpawnScript [
   &PartitionParticles()
   &InitializeParticle(
    Mesh Scale=&SelectVectorFromArray(
     Direct Array Index=$Particles.PartitionParticles.Partition
     Vector Selection Array=DataInterface() {propertyValues {FloatData [<0.7 0.7 1.3>, <0.7 0.7 1.3>]}}
    )
    Mesh Scale Mode=::NewEnumerator4
    Sprite Size Mode=::NewEnumerator3
    Uniform Sprite Size=12
    UsePositionOffset=false
    Write Material Random=true
   )
  ]
  particleUpdateScript [
   &SkeletalMeshLocation(
    Bone / Socket Index=&SelectIntFromArray(Direct Array Index=$Particles.PartitionParticles.Partition, Int Selection Array=DataInterface() {propertyValues {IntData [0,1]}})
    MaskSpawnGroup=false
    Mesh Sampling Type=::NewEnumerator5
    Orientation Sampling=::NewEnumerator1
    Sampled Position Offset=&FN_FrontEndVector3Modifier(
     Gameplay=<0 0 0>
     Front End=&SelectVectorFromArray(
      Direct Array Index=$Particles.PartitionParticles.Partition
      Vector Selection Array=DataInterface() {propertyValues {FloatData [<0 0.75 0>, <0 0.75 0>]}}
     )
    )
    Skeletal Mesh=$System.Skeletal Mesh
    Socket Sampling Mode=::NewEnumerator5
   )
   &SkeletalMeshLocation(
    Bone / Socket Index=&SelectIntFromArray(Direct Array Index=$Particles.PartitionParticles.Partition, Int Selection Array=DataInterface() {propertyValues {IntData [1,0]}})
    MaskSpawnGroup=false
    Mesh Sampling Type=::NewEnumerator5
    Orientation Sampling=::NewEnumerator1
    Sampled Position Offset=&FN_FrontEndVector3Modifier(
     Gameplay=<0 0 0>
     Front End=&SelectVectorFromArray(
      Direct Array Index=$Particles.PartitionParticles.Partition
      Vector Selection Array=DataInterface() {propertyValues {FloatData [<0 0.75 0>, <0 0.75 0>]}}
     )
    )
    Skeletal Mesh=$System.Skeletal Mesh
    Socket Sampling Mode=::NewEnumerator5
   ) {module=SkeletalMeshLocation001}
   /SetVariables(Particles.ADSFade:NiagaraFloat=&Lerp_Float(
     A=0
     B=1
     Alpha=&FN_LOD_Distance_Remap(Start Distance=250, End Distance=450, Minimum Output=0.5, Maximum Output=1, Clamp Output=true)
    )) {isSetParameters=true}
   /SetVariables(Particles.MeshOrientation=&MultiplyQuaternion(
     Quaternion A=&LerpQuaternion(
      Quaternion A:Quat4f=$Particles.SkeletalMeshLocation.SampledRotation
      Quaternion B:Quat4f=$Particles.SkeletalMeshLocation001.SampledRotation
      Lerp Factor:NiagaraFloat=&SelectFloatFromArray(Direct Array Index=$Particles.PartitionParticles.Partition, Float Selection Array=DataInterface() {propertyValues {FloatData [0,0.5]}})
      Enable Spherical Linear Interpolation:NiagaraBool=false
     )
     Quaternion B=&MakeQuatFromAxisAngle(
      Axis=&SelectVectorFromArray(
       Direct Array Index=$Particles.PartitionParticles.Partition
       Vector Selection Array=DataInterface() {propertyValues {FloatData [<0 0 0>, <0 10 0>]}}
      )
      AngleInDegrees=1
     )
    )) {isSetParameters=true}
   &ParticleState(Kill Particles When Lifetime Has Elapsed=false, Let Infinitely Lived Particles Die When Emitter Deactivates=true)
   &Timeline(
    Play Rate=&SelectFloatFromArray(Direct Array Index=$Particles.PartitionParticles.Partition, Float Selection Array=DataInterface() {propertyValues {FloatData [2.5,2.5]}})
    Timeline End=10000000
   )
   &ScaleMeshSize(Scale Factor=&Multiply_Vector(
     A=&Multiply_VectorByFloat(
      Vector=<0.2 0.2 0.2>
      Float=&Multiply_Float(
       A=&SelectFloatFromArray(Direct Array Index=$Particles.PartitionParticles.Partition, Float Selection Array=DataInterface() {propertyValues {FloatData [0.4,0.4]}})
       B=&Multiply_Float(A=$System.bisHandHidden, B=$System.FlameDistanceScaling)
      )
     )
     B=&Lerp_Vector(
      A=&Lerp_Vector(
       A=&Lerp_Vector(A=<1 1 1>, B=<1 1 1>, Alpha=$System.InterpolateOverTime.InterpolateOverTime_Float.Moving Average)
       B=<0.8 0.8 0.8>
       Alpha=$System.InterpolateOverTime.InterpolateOverTime_Float.Moving Average
      )
      B=<0.5 0.5 0.5>
      Alpha=&ClampFloat(Float=$System.InterpolateOverTime.InterpolateOverTime_Float.Moving Average)
     )
    ))
   &ScaleColor(
    Scale Alpha=&Multiply_Float(A=&Multiply_Float(A=0.5, B=$System.FlameDistanceScaling), B=$Particles.ADSFade)
    Scale RGB=&MakeVectorFromLinearColorRGB(Color=&MultiplyLinearColor(
      Linear Color=<rgba 1 1 1 1>
      Scale Factor=&MakeVector4_FromLinearColor(LinearColor=&FNTimeOfDayColor(Night Color=<rgba 0.5 0.5 0.5 1>, Morning Color=<rgba 1 1 1 1>, Day Color=<rgba 1 1 1 1>, Evening Color=<rgba 1 1 1 1>))
     ))
    ScaleA=true
    ScaleRGB=true
   )
   &InterpolateOverTime(Rate Of Change=10, Target Vector Value:Vector3f=$Particles.SkeletalMeshLocation.SampledVelocity, Variable Type=::NewEnumerator2)
   /SetVariables(Particles.Interpolate DY:Vector4f=&MakeFloat4FromVectorAndFloat(
     "Vector (RGB)"=/Niagara/DynamicInputs/Math/Clamp_Vector(
      Vector:Vector3f=$Particles.InterpolateOverTime.InterpolateOverTime_V3_.Moving Average
      Min:Vector3f=&VectorFromFloat(Value=-50)
      Max:Vector3f=&VectorFromFloat(Value=50)
     )
     "Float (Alpha)"=-0.07
    )) {isSetParameters=true}
   &DynamicMaterialParameters(
    Index 0 Param 1=&Lerp_Float(
     A=&SelectFloatFromArray(Direct Array Index=$Particles.PartitionParticles.Partition, Float Selection Array=DataInterface() {propertyValues {FloatData [0.3,0.6]}})
     B=0.8
     Alpha=0
    )
    Module.Index0Param0Write=true
    Module.Index0Param1Write=true
    Module.Index0Param2Write=true
    Module.Index0Param3Write=true
   )
  ]
  rendererClasses [&NiagaraMeshRendererProperties]
  renderers [
   {rendererClass=&NiagaraMeshRendererProperties, rendererIndex=0}
   {rendererClass=&NiagaraMeshRendererProperties, rendererIndex=1}
  ]
 }}
summary {
 systemName=NS_DaffodilSong_ArmFlames
 emitters [{emitterName=NE_ArmFlames, rendererClasses [&NiagaraMeshRendererProperties]}]
 userVariables [
  User.bIsFrontEnd {type=NiagaraBool}
  User.bIsFrontEndPreview {type=NiagaraBool}
  User.bisHandHidden {type=NiagaraBool}
  User.IsSkydiving {type=NiagaraBool}
  User.VariantNumber {type=NiagaraInt32}
 ]
}
