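# Niagara system for the DaffodilSong character's flaming head hair, drawn as fire-strand mesh particles.
# One emitter bursts three persistent particles, using partition index to pick mesh and vary timeline rate and material parameters.
# Particles follow the skeletal mesh, storing smoothed, clamped sampled velocity in a vector with a constant alpha.
# Color scales with time of day, while alpha combines the system's flame distance scaling with a camera-distance LOD fade.
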NiagaraSystem /BRCosmetics/Characters/Player/Male/Medium/Bodies/M_MED_DaffodilSong/FX/NS_DaffodilSong_HeadFlames
dependencies {
 usedDataInterfaces [&NiagaraDataInterfaceArrayFloat, &NiagaraDataInterfaceSkeletalMesh]
 usedDynamicInputs [
  &MakeCustomFloatFromBool
  &SetBoolByFloatComparison
  &VectorLength
 ]
 usedModules [
  &self:SystemSpawnScript.SystemScriptSource.SystemScriptGraph.NiagaraNodeAssignment_0.SetVariables_#1
  &self:SystemSpawnScript.SystemScriptSource.SystemScriptGraph.NiagaraNodeAssignment_1.SetVariables_#2
  &SystemState
  &self:SystemSpawnScript.SystemScriptSource.SystemScriptGraph.NiagaraNodeAssignment_2.SetVariables_#3
  &self:SystemSpawnScript.SystemScriptSource.SystemScriptGraph.NiagaraNodeAssignment_3.SetVariables_#4
  &self:SystemSpawnScript.SystemScriptSource.SystemScriptGraph.NiagaraNodeAssignment_4.SetVariables_#5
 ]
}
emitters {NE_FireHairStrands {
  emitterUpdateScript [
   &EmitterState(
    Inactive Response=::NewEnumerator1
    Life Cycle Mode=::NewEnumerator1
    Loop Behavior=::NewEnumerator1
    Loop Duration Mode=::NewEnumerator1
    UseLoopDelay=false
   )
   &SpawnBurst_Instantaneous(Spawn Count=3, Use Spawn Probability=false, UseLoopCountLimit=false)
  ]
  particleSpawnScript [
   &PartitionParticles(Number Of Partitions=3)
   &InitializeParticle(
    Mesh Scale Mode=::NewEnumerator4
    Sprite Size Mode=::NewEnumerator3
    Uniform Sprite Size=12
    UsePositionOffset=false
    Write Material Random=true
   )
   /SetVariables(Particles.MeshIndex=$Particles.PartitionParticles.Partition) {isSetParameters=true}
   &InitialMeshOrientation(Mesh Orientation Mode=::NewEnumerator7, Rotation=<0 0.25 0>, Use Rotation Vector=true)
  ]
  particleUpdateScript [
   &ParticleState(Kill Particles When Lifetime Has Elapsed=false, Let Infinitely Lived Particles Die When Emitter Deactivates=true)
   &SkeletalMeshLocation(Bone Sampling Mode=::NewEnumerator4, MaskSpawnGroup=false, Skeletal Mesh=$System.Skeletal Mesh)
   /SetVariables(Particles.ADSFade:NiagaraFloat=&Lerp_Float(
     A=0
     B=1
     Alpha=&FN_LOD_Distance_Remap(Start Distance=250, End Distance=450, Minimum Output=0.5, Maximum Output=1, Clamp Output=true)
    )) {isSetParameters=true}
   &Timeline(
    Play Rate=&SelectFloatFromArray(Direct Array Index=$Particles.PartitionParticles.Partition, Float Selection Array=DataInterface() {propertyValues {FloatData [2.5,2.5]}})
    Timeline End=10000000
   )
   &ScaleColor(
    Scale Alpha=&Multiply_Float(A=&Multiply_Float(A=0.5, B=$System.FlameDistanceScaling), B=$Particles.ADSFade)
    Scale RGB=&MakeVectorFromLinearColorRGB(Color=&MultiplyLinearColor(
      Linear Color=<rgba 1 1 1 1>
      Scale Factor=&MakeVector4_FromLinearColor(LinearColor=&FNTimeOfDayColor(Night Color=<rgba 0.5 0.5 0.5 1>, Morning Color=<rgba 1 1 1 1>, Day Color=<rgba 1 1 1 1>, Evening Color=<rgba 1 1 1 1>))
     ))
    ScaleA=true
    ScaleRGB=true
   )
   &InterpolateOverTime(Rate Of Change=10, Target Vector Value:Vector3f=$Particles.SkeletalMeshLocation.SampledVelocity, Variable Type=::NewEnumerator2)
   /SetVariables(Particles.Interpolate DY:Vector4f=&MakeFloat4FromVectorAndFloat(
     "Vector (RGB)"=/Niagara/DynamicInputs/Math/Clamp_Vector(
      Vector:Vector3f=$Particles.InterpolateOverTime.InterpolateOverTime_V3_.Moving Average
      Min:Vector3f=&VectorFromFloat(Value=-50)
      Max:Vector3f=&VectorFromFloat(Value=50)
     )
     "Float (Alpha)"=-0.07
    )) {isSetParameters=true}
   &DynamicMaterialParameters(
    Index 0 Param 1=&Multiply_FloatByInt(Float=0.25, Integer=$Particles.PartitionParticles.Partition)
    Module.Index0Param0Write=true
    Module.Index0Param1Write=true
    Module.Index0Param2Write=true
    Module.Index0Param3Write=true
   )
  ]
  rendererClasses [&NiagaraMeshRendererProperties]
  renderers [
   {rendererClass=&NiagaraMeshRendererProperties, rendererIndex=0}
   {rendererClass=&NiagaraMeshRendererProperties, rendererIndex=1}
  ]
 }}
summary {
 systemName=NS_DaffodilSong_HeadFlames
 emitters [{emitterName=NE_FireHairStrands, rendererClasses [&NiagaraMeshRendererProperties]}]
 userVariables [
  User.bIsFrontEnd {type=NiagaraBool}
  User.bIsFrontEndPreview {type=NiagaraBool}
  User.bisHandHidden {type=NiagaraBool}
  User.IsSkydiving {type=NiagaraBool}
  User.VariantNumber {type=NiagaraInt32}
 ]
}
