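# Material function for the DaffodilSong character cosmetic that adds animated, tinted wavy noise to a material's emissive.
# Two planar UV sets from pre-skinned position, blended by pre-skinned normal, keep the pattern attached to the animated mesh, with a switch to tiled mesh UVs.
# Scrolling normal maps and a sine-driven flowmap build UV distortion for panned main, detail and smoke noise layers.
# A channel-selectable TechArtMask input supplies masking.
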MaterialFunction /BRCosmetics/Characters/Player/Male/Medium/Bodies/M_MED_DaffodilSong/Materials/MF_DaffodilSong_WavyNois
graph {nodes [
  n1 = GetMaterialAttributes(None=NamedRerouteUsage()) {attributeGetTypes [
    #1
    #2
    #3
    #4
   ]}
  n2 = Add(A=NamedRerouteUsage(), B=NamedRerouteUsage())
  n3 = NamedRerouteUsage()
  n4 = VectorParameter() {group=Styx, parameterName=NoiseEmissiveColor, channelNames {a=Intensity, b="", g="", r=""}, defaultValue=<rgba 1 0.9284 0 1>}
  n5 = Multiply(
   A=MaterialFunctionCall(
    In=LinearInterpolate(A=@n2, Alpha=@n3) {constB=0}
    Contrast=ScalarParameter() {defaultValue=0.05, group=SmokeBeneath, parameterName=NoiseContrast}
   ) {materialFunction=&CheapContrast}.Result
   B=Multiply(A=@n4, B=@n4)
  )
  n6 = NamedRerouteUsage()
  n7 = SetMaterialAttributes!(
   MaterialAttributes=@n1.MaterialAttributes
   BaseColor=@n1.MaterialAttributes
   EmissiveColor=Add(A=@n1.EmissiveColor, B=Multiply(A=@n5, B=@n6))
   Metallic=@n1.MaterialAttributes
   Roughness=@n1.MaterialAttributes
  ) {attributeSetTypes [
    #1
    #2
    #3
    #4
   ]}
  FunctionOutput(None=@n7)
  NamedRerouteDeclaration(None=FunctionInput() {inputName=Material, inputType=FunctionInput_MaterialAttributes}) {name=MA_IN, nodeColor=<rgba 1 0.6797 0 1>}
  n8 = FunctionInput() {bUsePreviewValueAsDefault=true, inputName=TechArtMask, sortPriority=2, previewValue=<1 1 1 1>}
  NamedRerouteDeclaration(None=ChannelMaskParameter(None=@n8) {group=SmokeBeneath, maskChannel=Green, parameterName=WavyNoise, defaultValue=<rgba 0 1 0 0>}.R) {name=TechArtMask, nodeColor=<rgba 0 1 0.8125 1>}
  n9 = NamedRerouteUsage()
  n10 = AppendVector(
   A=ScalarParameter() {group=Styx, parameterName=StyxNoise_Speed2X}
   B=ScalarParameter() {defaultValue=-0.1, group=Styx, parameterName=StyxNoise_Speed2Y}
  )
  n11 = TextureSample(UVs=Add(A=Panner(Coordinate=@n9, Speed=@n10), B=NamedRerouteUsage()), Tex=NamedRerouteUsage()) {samplerSource=SSM_Wrap_WorldGroupSettings}
  n12 = ScalarParameter() {defaultValue=5, group=Styx, parameterName=StyxNoise_Contrast}
  n13 = LinearInterpolate(
   A=Power(Base=@n11.R, Exp=ScalarParameter() {defaultValue=5, group=Styx, parameterName=StyxNoise_Contrast2})
   B=@n11.R
   Alpha=Saturate(None=Power(Base=@n11.R, Exp=@n12))
  )
  n14 = ScalarParameter() {defaultValue=1, group=Styx, parameterName=MainNoiseIntensity}
  n15 = NamedRerouteUsage()
  n16 = AppendVector(
   A=ScalarParameter() {group=Styx, parameterName=StyxNoise_Speed2X}
   B=ScalarParameter() {defaultValue=-0.1, group=Styx, parameterName=StyxNoise_Speed2Y}
  )
  n17 = TextureSample(UVs=Add(A=Panner(Coordinate=@n15, Speed=@n16), B=NamedRerouteUsage()), Tex=NamedRerouteUsage()) {samplerSource=SSM_Wrap_WorldGroupSettings}
  n18 = ScalarParameter() {defaultValue=5, group=Styx, parameterName=StyxNoise_Contrast}
  n19 = LinearInterpolate(
   A=Power(Base=@n17.R, Exp=ScalarParameter() {defaultValue=5, group=Styx, parameterName=StyxNoise_Contrast2})
   B=@n17.R
   Alpha=Saturate(None=Power(Base=@n17.R, Exp=@n18))
  )
  n20 = ScalarParameter() {defaultValue=1, group=Styx, parameterName=MainNoiseIntensity}
  n21 = VertexInterpolator(VS=PreSkinnedNormal())
  n22 = MaterialFunctionCall(In=Abs(None=ComponentMask(None=@n21.PS) {r=true}), Contrast=ScalarParameter() {defaultValue=1, group=SmokeBeneath, parameterName=UVContrast}) {materialFunction=&CheapContrast}
  NamedRerouteDeclaration(None=LinearInterpolate(A=Multiply(A=@n13, B=@n14), B=Multiply(A=@n19, B=@n20), Alpha=@n22.Result)) {name=MainNoisePattern, nodeColor=<rgba 0 1 0.3203 1>}
  n23 = Panner(
   Coordinate=Add(A=NamedRerouteUsage(), B=NamedRerouteUsage())
   Speed=AppendVector(
    A=ScalarParameter() {group=Styx_Detail, parameterName=Styx_Detail_Noise_SpeedX}
    B=ScalarParameter() {defaultValue=-0.1, group=Styx_Detail, parameterName=Styx_Detail_Noise_SpeedY}
   )
  )
  n24 = TextureSampleParameter2D(UVs=@n23) {
   group=Styx_Detail
   parameterName=StyxNoiseDetail
   samplerType=SAMPLERTYPE_Masks
   texture=/BRCosmetics/Characters/Player/Male/Medium/Bodies/M_MED_GateHound_Spectral/Textures/T_StyxGateHound_DetailNoise_M
  }
  n25 = LinearInterpolate(
   A=ScalarParameter() {group=Styx_Detail, parameterName=Styx_Detail_Noise_Low}
   B=ScalarParameter() {defaultValue=1, group=Styx_Detail, parameterName=Styx_Detail_Noise_High}
   Alpha=@n24.R
  )
  n26 = Panner(
   Coordinate=Add(A=NamedRerouteUsage(), B=NamedRerouteUsage())
   Speed=AppendVector(
    A=ScalarParameter() {group=Styx_Detail, parameterName=Styx_Detail_Noise_SpeedX}
    B=ScalarParameter() {defaultValue=-0.1, group=Styx_Detail, parameterName=Styx_Detail_Noise_SpeedY}
   )
  )
  n27 = TextureSampleParameter2D(UVs=@n26) {
   group=Styx_Detail
   parameterName=StyxNoiseDetail
   samplerType=SAMPLERTYPE_Masks
   texture=&T_StyxGateHound_DetailNoise_M
  }
  n28 = LinearInterpolate(
   A=ScalarParameter() {group=Styx_Detail, parameterName=Styx_Detail_Noise_Low}
   B=ScalarParameter() {defaultValue=1, group=Styx_Detail, parameterName=Styx_Detail_Noise_High}
   Alpha=@n27.R
  )
  n29 = NamedRerouteUsage()
  NamedRerouteDeclaration(None=LinearInterpolate(A=Saturate(None=@n25), B=Saturate(None=@n28), Alpha=@n29)) {name=DetailNoise, nodeColor=<rgba 1 0.0938 0 1>}
  n30 = FunctionInput(Preview=VertexInterpolator(VS=PreSkinnedPosition()).PS) {bUsePreviewValueAsDefault=true, inputName=PreSkinnedLocation, sortPriority=1, previewValue=<1 0 0 1>}
  NamedRerouteDeclaration(None=@n30) {name=PreSkinned Loc, nodeColor=<rgba 1 0.6797 0 1>}
  n31 = NamedRerouteUsage()
  n32 = StaticComponentMaskParameter(None=@n31) {defaultB=true, defaultG=true, parameterName=WorldUVChannel02}
  n33 = ScalarParameter() {defaultValue=30, group=Styx_Transit, parameterName=WorldUVSize}
  n34 = Multiply(A=TextureCoordinate(), B=ScalarParameter() {defaultValue=3, group=Styx, parameterName=StyxMainNoiseTilling})
  NamedRerouteDeclaration(None=StaticSwitchParameter(True=Divide(A=@n32, B=@n33), False=@n34) {defaultValue=true, group=Styx_Transit, parameterName=UsePreskinnedAlignedUV}) {name=UV_withParallax_B, nodeColor=<rgba 1 0.6562 0 1>}
  n35 = Divide(A=StaticComponentMaskParameter(None=@n31) {defaultB=true, defaultR=true, parameterName=WorldUVChannel01}, B=@n33)
  NamedRerouteDeclaration(None=StaticSwitchParameter(True=NamedRerouteDeclaration(None=@n35) {name=UVVis, nodeColor=<rgba 0.0312 1 0 1>}, False=@n34) {defaultValue=true, group=Styx_Transit, parameterName=UsePreskinnedAlignedUV}) {name=UV_withParallax_A, nodeColor=<rgba 1 0.6562 0 1>}
  NamedRerouteDeclaration(None=@n22.Result) {name=PreskinnedAlignAlpha, nodeColor=<rgba 1 0 0.9609 1>}
  n36 = TextureSampleParameter2D(UVs=@n23) {
   group=StyxMod
   parameterName=DistortSmoke
   samplerType=SAMPLERTYPE_Grayscale
   texture=/BRCosmetics/Characters/Player/Male/Medium/Bodies/M_MED_GateHound_Spectral/Textures/T_FractalDistortMask_GateHound
  }
  n37 = Saturate(None=Multiply(A=@n36.RGB, B=ScalarParameter() {defaultValue=10, group=StyxMod, parameterName=SmokeAreaMulti}) {constB=5})
  n38 = ScalarParameter() {defaultValue=0.5, group=StyxMod, parameterName=SmokeAreaIntensity}
  n39 = TextureSampleParameter2D(UVs=@n26) {
   group=StyxMod
   parameterName=DistortSmoke
   samplerType=SAMPLERTYPE_Grayscale
   texture=&T_FractalDistortMask_GateHound
  }
  n40 = Saturate(None=Multiply(A=@n39.RGB, B=ScalarParameter() {defaultValue=10, group=StyxMod, parameterName=SmokeAreaMulti}) {constB=5})
  n41 = ScalarParameter() {defaultValue=0.5, group=StyxMod, parameterName=SmokeAreaIntensity}
  NamedRerouteDeclaration(None=LinearInterpolate(A=Multiply(A=@n37, B=@n38), B=Multiply(A=@n40, B=@n41), Alpha=@n29)) {name=SmokeArea, nodeColor=<rgba 0 1 0.1797 1>}
  n42 = ScalarParameter() {group=Styx, parameterName=StyxNoise_SpeedX}
  n43 = ScalarParameter() {defaultValue=-0.1, group=Styx, parameterName=StyxNoise_SpeedY}
  n44 = TextureObjectParameter() {
   group=Styx
   parameterName=StyxDistortNormal
   samplerType=SAMPLERTYPE_Normal
   texture=/BRCosmetics/Characters/Player/Male/Medium/Bodies/M_MED_GateHound_Spectral/Textures/T_StyxGateHound_N
  }
  n45 = TextureSample(UVs=Panner(Coordinate=NamedRerouteUsage(), Speed=AppendVector(A=@n42, B=@n43)), Tex=@n44) {
   samplerSource=SSM_Wrap_WorldGroupSettings
   samplerType=SAMPLERTYPE_Normal
   texture=/Game/Characters/Player/Male/Medium/Bodies/M_MED_DarkStance_Pater/Textures/T_StyxCosmetic_N
  }
  NamedRerouteDeclaration(None=ComponentMask(None=@n45.RGB) {g=true, r=true}) {name=Distortion_NormalNoise_A, nodeColor=<rgba 0.3828 0 1 1>}
  n46 = ScalarParameter() {group=Styx, parameterName=StyxNoise_SpeedX}
  n47 = ScalarParameter() {defaultValue=-0.1, group=Styx, parameterName=StyxNoise_SpeedY}
  n48 = TextureObjectParameter() {
   group=Styx
   parameterName=StyxDistortNormal
   samplerType=SAMPLERTYPE_Normal
   texture=&T_StyxGateHound_N
  }
  n49 = TextureSample(UVs=Panner(Coordinate=NamedRerouteUsage(), Speed=AppendVector(A=@n46, B=@n47)), Tex=@n48) {
   samplerSource=SSM_Wrap_WorldGroupSettings
   samplerType=SAMPLERTYPE_Normal
   texture=&T_StyxCosmetic_N
  }
  NamedRerouteDeclaration(None=ComponentMask(None=@n49.RGB) {g=true, r=true}) {name=Distortion_NormalNoise_B, nodeColor=<rgba 0.1016 0 1 1>}
  n50 = TextureSampleParameter2D(UVs=Multiply(A=NamedRerouteUsage(), B=ScalarParameter() {defaultValue=1, group=SmokeBeneath, parameterName=Smoke_DistortTiling})) {
   desc=Only RG channel works
   group=SmokeBeneath
   parameterName=Smoke_DistortUV_Flowmap
   samplerSource=SSM_Wrap_WorldGroupSettings
   samplerType=SAMPLERTYPE_LinearColor
   texture=/Game/Global/Materials/MaterialGym/Materials/SmokeWithFlowmap/Textures/T_Smokemap
  }
  n51 = Subtract(A=Multiply(A=ComponentMask(None=@n50.RGB) {g=true, r=true}) {constB=2})
  n52 = StaticSwitchParameter(True=AppendVector(A=ComponentMask(None=@n51) {r=true}, B=Constant()), False=@n51) {defaultValue=true, group=SmokeBeneath, parameterName=Smoke_UseXwave}
  n53 = Time()
  n54 = ScalarParameter() {defaultValue=0.2, group=SmokeBeneath, parameterName=Smoke_DistortTimeScale}
  n55 = StaticSwitchParameter(True=Sine(None=Multiply(A=@n53, B=@n54) {constB=0.02}), False=Constant() {r=1}) {defaultValue=true, group=SmokeBeneath, parameterName=Smoke_UseTimeDriver}
  NamedRerouteDeclaration(None=Multiply(A=Multiply(A=@n52, B=@n55), B=ScalarParameter() {defaultValue=0.3, group=SmokeBeneath, parameterName=Smoke_DistortIntensity}) {constB=0.5}) {name=CachedDistortUV, nodeColor=<rgba 0.3594 1 0 1>}
  n56 = NamedRerouteUsage()
  n57 = ScalarParameter() {defaultValue=0.01, group=Styx, parameterName=Styx_Distort_UV}
  n58 = NamedRerouteUsage()
  NamedRerouteDeclaration(None=Add(A=Multiply(A=@n56, B=@n57), B=@n58)) {name=Distortion_Final_A, nodeColor=<rgba 0.2422 0 1 1>}
  n59 = NamedRerouteUsage()
  NamedRerouteDeclaration(None=Add(A=Multiply(A=@n57, B=@n59), B=@n58)) {name=Distortion_Final_B, nodeColor=<rgba 1 0 0.8906 1>}
  n60 = TextureObjectParameter() {
   group=Styx
   parameterName=Styx Mask
   samplerType=SAMPLERTYPE_Masks
   texture=/BRCosmetics/Characters/Player/Male/Medium/Bodies/M_MED_DaffodilSong/Textures/T_DaffodilSong_Mask
  }
  NamedRerouteDeclaration(None=@n60) {name=Styx Mask IN, nodeColor=<rgba 1 0.8906 0 1>}
  NamedRerouteUsage()
  NamedRerouteUsage()
 ]}
