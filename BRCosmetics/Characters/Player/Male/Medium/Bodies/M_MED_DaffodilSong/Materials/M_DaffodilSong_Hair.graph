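# Character material for the DaffodilSong body's hair, adding a patterned, warm golden Fresnel sheen on top of the 2023 character parent functions.
# PreFX attributes get Base Color, Roughness and Specular lerped toward fixed values by an adjusted Fresnel.
# A tiled Neon Jam cloth FX texture, multiplied by the green Tech Art Mask channel, blends that variant over the original before PostFX.
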Material /BRCosmetics/Characters/Player/Male/Medium/Bodies/M_MED_DaffodilSong/Materials/M_DaffodilSong_Hair
graph {
 properties {
  bUseMaterialAttributes=true
  bUsedWithMeshDeformer=true
  bUsedWithSkeletalMesh=true
  pixelDepthOffsetMode=PDOM_Legacy
  shadingModel=MSM_FromMaterialExpression
  usedShadingModels=Default Lit | Subsurface | Clear Coat | Subsurface Profile | Cloth
 }
 nodes [
  n1 = MaterialFunctionCall() {materialFunction=&MF_CharacterParent_2023_PreFX}
  n2 = GetMaterialAttributes(None=@n1.Result) {attributeGetTypes [#1, #2, #3]}
  n3 = Constant() {desc=Fresnel mult, r=1.35}
  n4 = Saturate(None=Add(A=Power(Base=@n1.Fresnel, Exp=@n3), B=Constant() {desc=Fresnel fill, r=0.1}))
  n5 = LinearInterpolate(A=@n2.Base Color, B=Constant3Vector() {desc=Fresnel Color, constant=<rgba 0.7157 0.552 0.2831 1>}, Alpha=@n4)
  n6 = LinearInterpolate(A=@n2.Roughness, B=Constant() {desc=Rough, r=0.45}, Alpha=@n4)
  n7 = LinearInterpolate(A=@n2.Specular, B=Constant() {desc=Spec, r=0.8}, Alpha=@n4)
  n8 = TextureSample(UVs=Multiply(A=TextureCoordinate(), B=Constant() {desc=Texture Tiling, r=2})) {
   samplerType=SAMPLERTYPE_Masks
   texture=/Game/Characters/Player/Female/Medium/Bodies/F_MED_Neon_Jam/Skins/Ch5/Textures/T_F_Med_Neon_Jam_ClothTile_FX
  }
  n9 = ComponentMask(None=@n1.Tech Art Mask) {desc=FX mask channel, g=true}
  MaterialFunctionCall(
   Material=BlendMaterialAttributes(
    A=@n1.Result
    B=SetMaterialAttributes!(MaterialAttributes=@n2.MaterialAttributes, Base Color=@n5, Roughness=@n6, Specular=@n7) {attributeSetTypes [#1, #2, #3]}
    Alpha=Multiply(A=@n8.R, B=@n9)
   )
   M=@n1.M
   Pre Skinned Local Position=@n1.M
   Distance Blend=@n1.M
   Fresnel=@n1.M
  ) {materialFunction=&MF_CharacterParent_2023_PostFX}
 ]
}
