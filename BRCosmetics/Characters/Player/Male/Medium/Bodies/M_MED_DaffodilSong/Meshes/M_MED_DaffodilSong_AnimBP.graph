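# Post-process style AnimBlueprint for the M_MED_DaffodilSong male body cosmetic, adding secondary motion to straps, jacket, suspenders and shoelaces.
# It copies the parent mesh pose through a Control Rig, then uses cached poses, crouch blending, look-ats and constraints.
# Anim Dynamics with spherical limits simulates the dangling parts in sequence.
# Front-end flags switch jacket fronts between per-bone and chain simulation; skydiving alphas swap suspender collision setups.
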AnimBlueprint /BRCosmetics/Characters/Player/Male/Medium/Bodies/M_MED_DaffodilSong/Meshes/M_MED_DaffodilSong_AnimBP
target_skeleton=&/M_MED_DaffodilSong_Skeleton
graphs [
 AnimGraph {nodes [
   n1 = "Transform (Modify) Bone - Bone: strap_trgt"!(ComponentPose, Alpha=1, Translation="0,0,0") -> (Pose) {
    rotationSpace=BCS_BoneSpace
    translationMode=BMM_Additive
    translationSpace=BCS_BoneSpace
    alphaBoolBlend {blendOption=Linear}
    alphaScaleBiasClamp {interpSpeedDecreasing=20, interpSpeedIncreasing=20}
    boneToModify {boneName=strap_trgt}
   }
   "Transform (Modify) Bone - Bone: strap_trgt"(ComponentPose=@n1, Alpha=1, Translation="0,0,0") {
    rotationSpace=BCS_BoneSpace
    translationMode=BMM_Additive
    translationSpace=BCS_BoneSpace
    alphaBoolBlend {blendOption=Linear}
    alphaScaleBiasClamp {interpSpeedDecreasing=20, interpSpeedIncreasing=20}
    boneToModify {boneName=strap_trgt}
   }
   n2 = Get bIsInFrontEnd()
   n3 = Anim Dynamics - Bone: dyn_jacket_fr_01_L!(ComponentPose=Local To Component(LocalPose=Use cached pose 'straps pose'()), bAlphaBoolEnabled:bool=@n2, GravityOverride:Vector="0,0,-980") -> (Pose) {
    alphaInputType=Bool
    angularBiasOverride=1
    angularDampingOverride=1
    angularSpringConstant=1500
    bAngularSpring=true
    bGravityOverrideInSimSpace=true
    bOverrideAngularBias=true
    bOverrideAngularDamping=true
    bUseGravityOverride=true
    bUseSphericalLimits=true
    gravityScale=0.35
    simulationSpace=RootRelative
    alphaBoolBlend {blendInTime=0.1, blendOption=Linear, blendOutTime=0.2}
    boundBone {boneName=dyn_jacket_fr_01_L}
    componentAppliedLinearAccClamp=<1000 1000 1000>
    componentLinearAccScale=<0.4 0.4 0.4>
    componentLinearVelScale=<0.4 0.4 0.4>
    gravityOverride {z=-980}
    physicsBodyDefinitions [{
      collisionType=CustomSphere
      sphereCollisionRadius=2
      boundBone {boneName=dyn_jacket_fr_01_L}
      boxExtents=<35 35 35>
      constraintSetup {angularLimitsMax=<0 45 2.5>, angularLimitsMin=<0 -45 -2.5>, angularTarget=<0 0 0>}
      localJointOffset=<7 0 0>
     }]
    sphericalLimits [
     {limitRadius=11, drivingBone {boneName=spine_04}, sphereLocalOffset=<0 0 0>}
     {limitRadius=11, drivingBone {boneName=spine_03}, sphereLocalOffset=<0 0 0>}
     {limitRadius=11, drivingBone {boneName=spine_02}, sphereLocalOffset=<0 0 0>}
    ]
   }
   n4 = Anim Dynamics - Bone: dyn_jacket_fr_02_L!(ComponentPose=@n3, bAlphaBoolEnabled:bool=@n2, GravityOverride:Vector="0,0,-980") -> (Pose) {
    alphaInputType=Bool
    angularBiasOverride=1
    angularDampingOverride=1
    angularSpringConstant=1500
    bAngularSpring=true
    bGravityOverrideInSimSpace=true
    bOverrideAngularBias=true
    bOverrideAngularDamping=true
    bUseGravityOverride=true
    bUseSphericalLimits=true
    gravityScale=0.35
    simulationSpace=RootRelative
    alphaBoolBlend {blendInTime=0.1, blendOption=Linear, blendOutTime=0.2}
    boundBone {boneName=dyn_jacket_fr_02_L}
    componentAppliedLinearAccClamp=<1000 1000 1000>
    componentLinearAccScale=<0.4 0.4 0.4>
    componentLinearVelScale=<0.4 0.4 0.4>
    gravityOverride {z=-980}
    physicsBodyDefinitions [{
      collisionType=CustomSphere
      sphereCollisionRadius=2
      boundBone {boneName=dyn_jacket_fr_02_L}
      boxExtents=<25 25 25>
      constraintSetup {angularLimitsMax=<0 45 2.5>, angularLimitsMin=<0 -45 -2.5>, angularTarget=<0 0 0>}
      localJointOffset=<7 0 0>
     }]
    sphericalLimits [
     {limitRadius=11, drivingBone {boneName=spine_04}, sphereLocalOffset=<0 0 0>}
     {limitRadius=11, drivingBone {boneName=spine_03}, sphereLocalOffset=<0 0 0>}
     {limitRadius=11, drivingBone {boneName=spine_02}, sphereLocalOffset=<0 0 0>}
    ]
   }
   n5 = Anim Dynamics - Bone: dyn_jacket_fr_03_L!(ComponentPose=@n4, bAlphaBoolEnabled:bool=@n2, GravityOverride:Vector="0,0,-980") -> (Pose) {
    alphaInputType=Bool
    angularBiasOverride=1
    angularDampingOverride=1
    angularSpringConstant=1500
    bAngularSpring=true
    bGravityOverrideInSimSpace=true
    bOverrideAngularBias=true
    bOverrideAngularDamping=true
    bUseGravityOverride=true
    bUseSphericalLimits=true
    gravityScale=0.35
    simulationSpace=RootRelative
    alphaBoolBlend {blendInTime=0.1, blendOption=Linear, blendOutTime=0.2}
    boundBone {boneName=dyn_jacket_fr_03_L}
    componentAppliedLinearAccClamp=<1000 1000 1000>
    componentLinearAccScale=<0.4 0.4 0.4>
    componentLinearVelScale=<0.4 0.4 0.4>
    gravityOverride {z=-980}
    physicsBodyDefinitions [{
      collisionType=CustomSphere
      sphereCollisionRadius=2
      boundBone {boneName=dyn_jacket_fr_03_L}
      boxExtents=<15 15 15>
      constraintSetup {angularLimitsMax=<0 45 2.5>, angularLimitsMin=<0 -45 -2.5>, angularTarget=<0 0 0>}
      localJointOffset=<7 0 0>
     }]
    sphericalLimits [
     {limitRadius=11, drivingBone {boneName=spine_04}, sphereLocalOffset=<0 0 0>}
     {limitRadius=11, drivingBone {boneName=spine_03}, sphereLocalOffset=<0 0 0>}
     {limitRadius=11, drivingBone {boneName=spine_02}, sphereLocalOffset=<0 0 0>}
    ]
   }
   n6 = Anim Dynamics - Bone: dyn_jacket_fr_01_R!(ComponentPose=@n5, bAlphaBoolEnabled:bool=@n2, GravityOverride:Vector="0,0,-980") -> (Pose) {
    alphaInputType=Bool
    angularBiasOverride=1
    angularDampingOverride=1
    angularSpringConstant=1500
    bAngularSpring=true
    bGravityOverrideInSimSpace=true
    bOverrideAngularBias=true
    bOverrideAngularDamping=true
    bUseGravityOverride=true
    bUseSphericalLimits=true
    gravityScale=0.35
    simulationSpace=RootRelative
    alphaBoolBlend {blendInTime=0.1, blendOption=Linear, blendOutTime=0.2}
    boundBone {boneName=dyn_jacket_fr_01_R}
    componentAppliedLinearAccClamp=<1000 1000 1000>
    componentLinearAccScale=<0.4 0.4 0.4>
    componentLinearVelScale=<0.4 0.4 0.4>
    gravityOverride {z=-980}
    physicsBodyDefinitions [{
      collisionType=CustomSphere
      sphereCollisionRadius=2
      boundBone {boneName=dyn_jacket_fr_01_R}
      boxExtents=<35 35 35>
      constraintSetup {angularLimitsMax=<0 45 2.5>, angularLimitsMin=<0 -45 -2.5>, angularTarget=<0 0 0>}
      localJointOffset=<7 0 0>
     }]
    sphericalLimits [
     {limitRadius=12, drivingBone {boneName=spine_04}, sphereLocalOffset=<0 0 0>}
     {limitRadius=12.5, drivingBone {boneName=spine_03}, sphereLocalOffset=<0 0 0>}
     {limitRadius=13, drivingBone {boneName=spine_02}, sphereLocalOffset=<0 0 0>}
    ]
   }
   n7 = Anim Dynamics - Bone: dyn_jacket_fr_02_R!(ComponentPose=@n6, bAlphaBoolEnabled:bool=@n2, GravityOverride:Vector="0,0,-980") -> (Pose) {
    alphaInputType=Bool
    angularBiasOverride=1
    angularDampingOverride=1
    angularSpringConstant=1500
    bAngularSpring=true
    bGravityOverrideInSimSpace=true
    bOverrideAngularBias=true
    bOverrideAngularDamping=true
    bUseGravityOverride=true
    bUseSphericalLimits=true
    gravityScale=0.35
    simulationSpace=RootRelative
    alphaBoolBlend {blendInTime=0.1, blendOption=Linear, blendOutTime=0.2}
    boundBone {boneName=dyn_jacket_fr_02_R}
    componentAppliedLinearAccClamp=<1000 1000 1000>
    componentLinearAccScale=<0.4 0.4 0.4>
    componentLinearVelScale=<0.4 0.4 0.4>
    gravityOverride {z=-980}
    physicsBodyDefinitions [{
      collisionType=CustomSphere
      sphereCollisionRadius=2
      boundBone {boneName=dyn_jacket_fr_02_R}
      boxExtents=<25 25 25>
      constraintSetup {angularLimitsMax=<0 45 2.5>, angularLimitsMin=<0 -45 -2.5>, angularTarget=<0 0 0>}
      localJointOffset=<7 0 0>
     }]
    sphericalLimits [
     {limitRadius=12, drivingBone {boneName=spine_04}, sphereLocalOffset=<0 0 0>}
     {limitRadius=12.5, drivingBone {boneName=spine_03}, sphereLocalOffset=<0 0 0>}
     {limitRadius=13, drivingBone {boneName=spine_02}, sphereLocalOffset=<0 0 0>}
    ]
   }
   n8 = Anim Dynamics - Bone: dyn_jacket_fr_03_R!(ComponentPose=@n7, bAlphaBoolEnabled:bool=@n2, GravityOverride:Vector="0,0,-980") -> (Pose) {
    alphaInputType=Bool
    angularBiasOverride=1
    angularDampingOverride=1
    angularSpringConstant=1500
    bAngularSpring=true
    bGravityOverrideInSimSpace=true
    bOverrideAngularBias=true
    bOverrideAngularDamping=true
    bUseGravityOverride=true
    bUseSphericalLimits=true
    gravityScale=0.35
    simulationSpace=RootRelative
    alphaBoolBlend {blendInTime=0.1, blendOption=Linear, blendOutTime=0.2}
    boundBone {boneName=dyn_jacket_fr_03_R}
    componentAppliedLinearAccClamp=<1000 1000 1000>
    componentLinearAccScale=<0.4 0.4 0.4>
    componentLinearVelScale=<0.4 0.4 0.4>
    gravityOverride {z=-980}
    physicsBodyDefinitions [{
      collisionType=CustomSphere
      sphereCollisionRadius=4.7
      boundBone {boneName=dyn_jacket_fr_03_R}
      boxExtents=<15 15 15>
      constraintSetup {angularLimitsMax=<0 45 2.5>, angularLimitsMin=<0 -45 -2.5>, angularTarget=<0 0 0>}
      localJointOffset=<7 0 0>
     }]
    sphericalLimits [
     {limitRadius=12, drivingBone {boneName=spine_04}, sphereLocalOffset=<0 0 0>}
     {limitRadius=12.5, drivingBone {boneName=spine_03}, sphereLocalOffset=<0 0 0>}
     {limitRadius=13, drivingBone {boneName=spine_02}, sphereLocalOffset=<0 0 0>}
    ]
   }
   n9 = Get Not in FrontEnd() -> (Not in FrontEnd:bool)
   n10 = "Anim Dynamics - Chain: dyn_jacket_fr_01_L -> dyn_jacket_fr_03_L"!(ComponentPose=@n8, bAlphaBoolEnabled:bool=@n9, GravityOverride:Vector="0,0,-980") -> (Pose) {
    alphaInputType=Bool
    angularBiasOverride=1
    angularDampingOverride=1
    angularSpringConstant=1500
    bAngularSpring=true
    bChain=true
    bGravityOverrideInSimSpace=true
    bOverrideAngularBias=true
    bOverrideAngularDamping=true
    bUseGravityOverride=true
    bUseSphericalLimits=true
    gravityScale=0.35
    simulationSpace=RootRelative
    alphaBoolBlend {blendInTime=0.1, blendOption=Linear, blendOutTime=0.2}
    boundBone {boneName=dyn_jacket_fr_01_L}
    chainEnd {boneName=dyn_jacket_fr_03_L}
    componentAppliedLinearAccClamp=<1000 1000 1000>
    componentLinearAccScale=<0.4 0.4 0.4>
    componentLinearVelScale=<0.4 0.4 0.4>
    gravityOverride {z=-980}
    physicsBodyDefinitions [
     {
      collisionType=CustomSphere
      sphereCollisionRadius=2
      boundBone {boneName=dyn_jacket_fr_01_L}
      boxExtents=<35 35 35>
      constraintSetup {angularLimitsMax=<0 45 2.5>, angularLimitsMin=<0 -45 -2.5>, angularTarget=<0 0 0>}
      localJointOffset=<7 0 0>
     }
     {
      collisionType=CustomSphere
      sphereCollisionRadius=2
      boundBone {boneName=dyn_jacket_fr_02_L}
      boxExtents=<25 25 25>
      constraintSetup {angularLimitsMax=<0 45 2.5>, angularLimitsMin=<0 -45 -2.5>, angularTarget=<0 0 0>}
      localJointOffset=<7 0 0>
     }
     {
      collisionType=CustomSphere
      sphereCollisionRadius=2
      boundBone {boneName=dyn_jacket_fr_03_L}
      boxExtents=<15 15 15>
      constraintSetup {angularLimitsMax=<0 45 2.5>, angularLimitsMin=<0 -45 -2.5>, angularTarget=<0 0 0>}
      localJointOffset=<7 0 0>
     }
    ]
    sphericalLimits [
     {limitRadius=11, drivingBone {boneName=spine_04}, sphereLocalOffset=<0 0 0>}
     {limitRadius=11, drivingBone {boneName=spine_03}, sphereLocalOffset=<0 0 0>}
     {limitRadius=11, drivingBone {boneName=spine_02}, sphereLocalOffset=<0 0 0>}
    ]
   }
   n11 = "Anim Dynamics - Chain: dyn_jacket_fr_01_R -> dyn_jacket_fr_03_R"!(ComponentPose=@n10, bAlphaBoolEnabled:bool=@n9, GravityOverride:Vector="0,0,-980") -> (Pose) {
    alphaInputType=Bool
    angularBiasOverride=1
    angularDampingOverride=1
    angularSpringConstant=1500
    bAngularSpring=true
    bChain=true
    bGravityOverrideInSimSpace=true
    bOverrideAngularBias=true
    bOverrideAngularDamping=true
    bUseGravityOverride=true
    bUseSphericalLimits=true
    gravityScale=0.35
    simulationSpace=RootRelative
    alphaBoolBlend {blendInTime=0.1, blendOption=Linear, blendOutTime=0.2}
    boundBone {boneName=dyn_jacket_fr_01_R}
    chainEnd {boneName=dyn_jacket_fr_03_R}
    componentAppliedLinearAccClamp=<1000 1000 1000>
    componentLinearAccScale=<0.4 0.4 0.4>
    componentLinearVelScale=<0.4 0.4 0.4>
    gravityOverride {z=-980}
    physicsBodyDefinitions [
     {
      collisionType=CustomSphere
      sphereCollisionRadius=2
      boundBone {boneName=dyn_jacket_fr_01_R}
      boxExtents=<35 35 35>
      constraintSetup {angularLimitsMax=<0 45 2.5>, angularLimitsMin=<0 -45 -2.5>, angularTarget=<0 0 0>}
      localJointOffset=<7 0 0>
     }
     {
      collisionType=CustomSphere
      sphereCollisionRadius=2
      boundBone {boneName=dyn_jacket_fr_02_R}
      boxExtents=<25 25 25>
      constraintSetup {angularLimitsMax=<0 45 2.5>, angularLimitsMin=<0 -45 -2.5>, angularTarget=<0 0 0>}
      localJointOffset=<7 0 0>
     }
     {
      collisionType=CustomSphere
      sphereCollisionRadius=4.7
      boundBone {boneName=dyn_jacket_fr_03_R}
      boxExtents=<15 15 15>
      constraintSetup {angularLimitsMax=<0 45 2.5>, angularLimitsMin=<0 -45 -2.5>, angularTarget=<0 0 0>}
      localJointOffset=<7 0 0>
     }
    ]
    sphericalLimits [
     {limitRadius=12, drivingBone {boneName=spine_04}, sphereLocalOffset=<0 0 0>}
     {limitRadius=12.5, drivingBone {boneName=spine_03}, sphereLocalOffset=<0 0 0>}
     {limitRadius=13, drivingBone {boneName=spine_02}, sphereLocalOffset=<0 0 0>}
    ]
   }
   n12 = Anim Dynamics - Bone: dyn_jacket_01_L!(ComponentPose=@n11, Alpha=1, GravityOverride:Vector="0,0,-980") -> (Pose) {
    angularBiasOverride=1
    angularDampingOverride=1
    angularSpringConstant=300
    bAngularSpring=true
    bGravityOverrideInSimSpace=true
    bOverrideAngularBias=true
    bOverrideAngularDamping=true
    bUseGravityOverride=true
    bUseSphericalLimits=true
    gravityScale=0.35
    simulationSpace=RootRelative
    alphaBoolBlend {blendOption=Linear}
    boundBone {boneName=dyn_jacket_01_L}
    componentAppliedLinearAccClamp=<1000 1000 1000>
    componentLinearAccScale=<0.4 0.4 0.4>
    componentLinearVelScale=<0.4 0.4 0.4>
    gravityOverride {z=-980}
    physicsBodyDefinitions [{
      collisionType=CustomSphere
      sphereCollisionRadius=2
      boundBone {boneName=dyn_jacket_01_L}
      boxExtents=<20 20 20>
      constraintSetup {angularLimitsMax=<0 45 2.5>, angularLimitsMin=<0 -45 -2.5>, angularTarget=<0 0 0>}
      localJointOffset=<10 0 0>
     }]
    sphericalLimits [
     {limitRadius=11, drivingBone {boneName=spine_04}, sphereLocalOffset=<0 0 0>}
     {limitRadius=11, drivingBone {boneName=spine_03}, sphereLocalOffset=<0 0 0>}
     {limitRadius=11, drivingBone {boneName=spine_02}, sphereLocalOffset=<0 0 0>}
    ]
   }
   n13 = Anim Dynamics - Bone: dyn_jacket_bk_01_L!(ComponentPose=@n12, Alpha=1, GravityOverride:Vector="0,0,-980") -> (Pose) {
    angularBiasOverride=1
    angularDampingOverride=1
    angularSpringConstant=300
    bAngularSpring=true
    bGravityOverrideInSimSpace=true
    bOverrideAngularBias=true
    bOverrideAngularDamping=true
    bUseGravityOverride=true
    bUseSphericalLimits=true
    gravityScale=0.35
    simulationSpace=RootRelative
    alphaBoolBlend {blendOption=Linear}
    boundBone {boneName=dyn_jacket_bk_01_L}
    componentAppliedLinearAccClamp=<1000 1000 1000>
    componentLinearAccScale=<0.4 0.4 0.4>
    componentLinearVelScale=<0.4 0.4 0.4>
    gravityOverride {z=-980}
    physicsBodyDefinitions [{
      collisionType=CustomSphere
      sphereCollisionRadius=2
      boundBone {boneName=dyn_jacket_bk_01_L}
      boxExtents=<20 20 20>
      constraintSetup {angularLimitsMax=<0 45 2.5>, angularLimitsMin=<0 -45 -2.5>, angularTarget=<0 0 0>}
      localJointOffset=<10 0 0>
     }]
    sphericalLimits [
     {limitRadius=11, drivingBone {boneName=spine_04}, sphereLocalOffset=<0 0 0>}
     {limitRadius=11, drivingBone {boneName=spine_03}, sphereLocalOffset=<0 0 0>}
     {limitRadius=11, drivingBone {boneName=spine_02}, sphereLocalOffset=<0 0 0>}
    ]
   }
   n14 = Anim Dynamics - Bone: dyn_jacket_bk_01!(ComponentPose=@n13, Alpha=1, GravityOverride:Vector="0,0,-980") -> (Pose) {
    angularBiasOverride=1
    angularDampingOverride=1
    angularSpringConstant=300
    bAngularSpring=true
    bGravityOverrideInSimSpace=true
    bOverrideAngularBias=true
    bOverrideAngularDamping=true
    bUseGravityOverride=true
    bUseSphericalLimits=true
    gravityScale=0.35
    simulationSpace=RootRelative
    alphaBoolBlend {blendOption=Linear}
    boundBone {boneName=dyn_jacket_bk_01}
    componentAppliedLinearAccClamp=<1000 1000 1000>
    componentLinearAccScale=<0.4 0.4 0.4>
    componentLinearVelScale=<0.4 0.4 0.4>
    gravityOverride {z=-980}
    physicsBodyDefinitions [{
      collisionType=CustomSphere
      sphereCollisionRadius=2
      boundBone {boneName=dyn_jacket_bk_01}
      boxExtents=<20 20 20>
      constraintSetup {angularLimitsMax=<0 45 2.5>, angularLimitsMin=<0 -45 -2.5>, angularTarget=<0 0 0>}
      localJointOffset=<10 0 0>
     }]
    sphericalLimits [
     {limitRadius=11, drivingBone {boneName=spine_04}, sphereLocalOffset=<0 0 0>}
     {limitRadius=11, drivingBone {boneName=spine_03}, sphereLocalOffset=<0 0 0>}
     {limitRadius=11, drivingBone {boneName=spine_02}, sphereLocalOffset=<0 0 0>}
    ]
   }
   n15 = Anim Dynamics - Bone: dyn_jacket_bk_01_R!(ComponentPose=@n14, Alpha=1, GravityOverride:Vector="0,0,-980") -> (Pose) {
    angularBiasOverride=1
    angularDampingOverride=1
    angularSpringConstant=300
    bAngularSpring=true
    bGravityOverrideInSimSpace=true
    bOverrideAngularBias=true
    bOverrideAngularDamping=true
    bUseGravityOverride=true
    bUseSphericalLimits=true
    gravityScale=0.35
    simulationSpace=RootRelative
    alphaBoolBlend {blendOption=Linear}
    boundBone {boneName=dyn_jacket_bk_01_R}
    componentAppliedLinearAccClamp=<1000 1000 1000>
    componentLinearAccScale=<0.4 0.4 0.4>
    componentLinearVelScale=<0.4 0.4 0.4>
    gravityOverride {z=-980}
    physicsBodyDefinitions [{
      collisionType=CustomSphere
      sphereCollisionRadius=2
      boundBone {boneName=dyn_jacket_bk_01_R}
      boxExtents=<20 20 20>
      constraintSetup {angularLimitsMax=<0 45 2.5>, angularLimitsMin=<0 -45 -2.5>, angularTarget=<0 0 0>}
      localJointOffset=<10 0 0>
     }]
    sphericalLimits [
     {limitRadius=11, drivingBone {boneName=spine_04}, sphereLocalOffset=<0 0 0>}
     {limitRadius=11, drivingBone {boneName=spine_03}, sphereLocalOffset=<0 0 0>}
     {limitRadius=11, drivingBone {boneName=spine_02}, sphereLocalOffset=<0 0 0>}
    ]
   }
   n16 = Anim Dynamics - Bone: dyn_jacket_01_R!(ComponentPose=@n15, Alpha=1, GravityOverride:Vector="0,0,-980") -> (Pose) {
    angularBiasOverride=1
    angularDampingOverride=1
    angularSpringConstant=300
    bAngularSpring=true
    bGravityOverrideInSimSpace=true
    bOverrideAngularBias=true
    bOverrideAngularDamping=true
    bUseGravityOverride=true
    bUseSphericalLimits=true
    gravityScale=0.35
    simulationSpace=RootRelative
    alphaBoolBlend {blendOption=Linear}
    boundBone {boneName=dyn_jacket_01_R}
    componentAppliedLinearAccClamp=<1000 1000 1000>
    componentLinearAccScale=<0.4 0.4 0.4>
    componentLinearVelScale=<0.4 0.4 0.4>
    gravityOverride {z=-980}
    physicsBodyDefinitions [{
      collisionType=CustomSphere
      sphereCollisionRadius=2
      boundBone {boneName=dyn_jacket_01_R}
      boxExtents=<20 20 20>
      constraintSetup {angularLimitsMax=<0 45 2.5>, angularLimitsMin=<0 -45 -2.5>, angularTarget=<0 0 0>}
      localJointOffset=<10 0 0>
     }]
    sphericalLimits [
     {limitRadius=11, drivingBone {boneName=spine_04}, sphereLocalOffset=<0 0 0>}
     {limitRadius=11, drivingBone {boneName=spine_03}, sphereLocalOffset=<0 0 0>}
     {limitRadius=11, drivingBone {boneName=spine_02}, sphereLocalOffset=<0 0 0>}
    ]
   }
   n17 = "Transform (Modify) Bone - Bone: dyn_jacket_MSTR_R"!(ComponentPose=@n16, Alpha=1, Translation="0,0,0", Rotation="10,0,0") -> (Pose) {
    rotationMode=BMM_Additive
    rotationSpace=BCS_BoneSpace
    translationMode=BMM_Additive
    translationSpace=BCS_BoneSpace
    alphaBoolBlend {blendOption=Linear}
    alphaScaleBiasClamp {bInterpResult=true, interpSpeedDecreasing=1.5, interpSpeedIncreasing=5}
    boneToModify {boneName=dyn_jacket_MSTR_R}
    rotation {pitch=10}
   }
   n18 = "Transform (Modify) Bone - Bone: dyn_jacket_fr_MSTR_R"!(ComponentPose=@n17, Alpha=1, Translation="0,0,0", Rotation="10,0,0") -> (Pose) {
    rotationMode=BMM_Additive
    rotationSpace=BCS_BoneSpace
    translationMode=BMM_Additive
    translationSpace=BCS_BoneSpace
    alphaBoolBlend {blendOption=Linear}
    alphaScaleBiasClamp {bInterpResult=true, interpSpeedDecreasing=1.5, interpSpeedIncreasing=5}
    boneToModify {boneName=dyn_jacket_fr_MSTR_R}
    rotation {pitch=10}
   }
   n19 = "Transform (Modify) Bone - Bone: dyn_jacket_fr_MSTR_R"(ComponentPose=@n18, Alpha=1, Translation="0,0,0", Rotation="10,0,0") {
    rotationMode=BMM_Additive
    rotationSpace=BCS_BoneSpace
    translationMode=BMM_Additive
    translationSpace=BCS_BoneSpace
    alphaBoolBlend {blendOption=Linear}
    alphaScaleBiasClamp {bInterpResult=true, interpSpeedDecreasing=2.5}
    boneToModify {boneName=dyn_jacket_fr_MSTR_R}
    rotation {pitch=10}
   }
   n20 = "Transform (Modify) Bone - Bone: dyn_jacket_MSTR_R"(ComponentPose=@n19, Alpha=1, Translation="0,0,0", Rotation="10,0,0") {
    rotationMode=BMM_Additive
    rotationSpace=BCS_BoneSpace
    translationMode=BMM_Additive
    translationSpace=BCS_BoneSpace
    alphaBoolBlend {blendOption=Linear}
    alphaScaleBiasClamp {bInterpResult=true, interpSpeedDecreasing=2.5}
    boneToModify {boneName=dyn_jacket_MSTR_R}
    rotation {pitch=10}
   }
   n21 = Anim Dynamics - Bone: dyn_suspender_fr_01!(ComponentPose=@n20, Alpha=1) -> (Pose) {
    angularDampingOverride=1
    angularSpringConstant=1500
    bAngularSpring=true
    bGravityOverrideInSimSpace=true
    bOverrideAngularDamping=true
    bUseSphericalLimits=true
    gravityScale=0.35
    simulationSpace=RootRelative
    alphaBoolBlend {blendOption=Linear}
    alphaScaleBiasClamp {scale=0.3}
    boundBone {boneName=dyn_suspender_fr_01}
    componentAppliedLinearAccClamp=<10000 10000 10000>
    componentLinearAccScale {z=0.45}
    componentLinearVelScale {z=0.45}
    gravityOverride {z=-980}
    physicsBodyDefinitions [{
      collisionType=CustomSphere
      sphereCollisionRadius=2
      boundBone {boneName=dyn_suspender_fr_01}
      boxExtents=<10 20 20>
      constraintSetup {angularLimitsMax=<5 80 10>, angularLimitsMin=<-5 0 -45>, angularTarget=<0 0 0>}
      localJointOffset=<3 0 0>
     }]
    sphericalLimits [{limitRadius=10.5, drivingBone {boneName=thigh_l}, sphereLocalOffset=<-6.2535 0.6173 2.5541>}]
   }
   n22 = Anim Dynamics - Bone: dyn_suspender_fr_02!(ComponentPose=@n21, Alpha=1) -> (Pose) {
    angularDampingOverride=1
    angularSpringConstant=1500
    bAngularSpring=true
    bGravityOverrideInSimSpace=true
    bOverrideAngularDamping=true
    bUseSphericalLimits=true
    gravityScale=0.35
    simulationSpace=RootRelative
    alphaBoolBlend {blendOption=Linear}
    alphaScaleBiasClamp {scale=0.5}
    boundBone {boneName=dyn_suspender_fr_02}
    componentAppliedLinearAccClamp=<10000 10000 10000>
    componentLinearAccScale {z=0.45}
    componentLinearVelScale {z=0.45}
    gravityOverride {z=-980}
    physicsBodyDefinitions [{
      collisionType=CustomSphere
      sphereCollisionRadius=2
      boundBone {boneName=dyn_suspender_fr_02}
      boxExtents=<5 10 10>
      constraintSetup {angularLimitsMax=<5 80 8>, angularLimitsMin=<-5 0 -40>, angularTarget=<0 0 0>}
      localJointOffset=<10 0 0>
     }]
    sphericalLimits [{limitRadius=8, drivingBone {boneName=thigh_l}, sphereLocalOffset=<-15 0 0>}]
   }
   n23 = Look At - Bone: dyn_suspender_fr_02_aim(ComponentPose=@n22, Alpha=1) {
    alphaBoolBlend {blendOption=Linear}
    boneToModify {boneName=dyn_suspender_fr_02_aim}
    lookAtLocation {x=0}
    lookAtTarget {boneReference {boneName=dyn_suspender_bk_03}}
    lookAt_Axis {axis=<1 0>}
   }
   n24 = "Constraint - dyn_suspender_fr_03 (TT)"!(ComponentPose=@n23, Alpha=1, ConstraintWeights_0=0.5, ConstraintWeights_1:float=0.5) -> (Pose) {
    alphaBoolBlend {blendOption=Linear}
    boneToModify {boneName=dyn_suspender_fr_03}
    constraintSetup [
     {offsetOption=None, transformType=Translation, targetBone {boneName=suspender_fr}}
     {offsetOption=None, transformType=Translation, targetBone {boneName=suspender_bk}}
    ]
    constraintWeights [1,1]
   }
   n25 = Get not skydiving() -> (not skydiving:double)
   n26 = Anim Dynamics - Bone: dyn_suspender_fr_03!(ComponentPose=@n24, Alpha=@n25) -> (Pose) {
    angularDampingOverride=1
    angularSpringConstant=1000
    bAngularSpring=true
    bGravityOverrideInSimSpace=true
    bOverrideAngularDamping=true
    bUseSphericalLimits=true
    gravityScale=0.35
    simulationSpace=RootRelative
    alphaBoolBlend {blendOption=Linear}
    alphaScaleBiasClamp {scale=0.8}
    boundBone {boneName=dyn_suspender_fr_03}
    componentAppliedLinearAccClamp=<10000 10000 10000>
    componentLinearAccScale {z=0.7}
    componentLinearVelScale {z=0.7}
    gravityOverride {z=-980}
    physicsBodyDefinitions [{
      collisionType=CustomSphere
      sphereCollisionRadius=2
      boundBone {boneName=dyn_suspender_fr_03}
      boxExtents=<10 20 20>
      constraintSetup {angularLimitsMax=<10 15 0>, angularLimitsMin=<-5 -45 -45>, angularTarget=<0 0 0>}
      localJointOffset=<5 0 0>
     }]
    sphericalLimits [
     {limitRadius=7, drivingBone {boneName=thigh_l}, sphereLocalOffset=<-3.2384 -6.1887 0.9911>}
     {limitRadius=8, drivingBone {boneName=thigh_l}, sphereLocalOffset=<-10 -2.1039 0.8464>}
    ]
   }
   n27 = Anim Dynamics - Bone: dyn_suspender_bk_01!(ComponentPose=@n26, Alpha=@n25) -> (Pose) {
    angularDampingOverride=1
    angularSpringConstant=1500
    bAngularSpring=true
    bGravityOverrideInSimSpace=true
    bOverrideAngularDamping=true
    bUseSphericalLimits=true
    gravityScale=0.35
    simulationSpace=RootRelative
    alphaBoolBlend {blendOption=Linear}
    alphaScaleBiasClamp {scale=0.8}
    boundBone {boneName=dyn_suspender_bk_01}
    componentAppliedLinearAccClamp=<10000 10000 10000>
    componentLinearAccScale {z=0.45}
    componentLinearVelScale {z=0.45}
    gravityOverride {z=-980}
    physicsBodyDefinitions [{
      collisionType=CustomSphere
      sphereCollisionRadius=1.5
      boundBone {boneName=dyn_suspender_bk_01}
      boxExtents=<10 20 20>
      constraintSetup {angularLimitsMax=<5 5 20>, angularLimitsMin=<-5 -45 -10>, angularTarget=<0 0 0>}
      localJointOffset=<5 0 0>
     }]
    sphericalLimits [
     {limitRadius=9, drivingBone {boneName=thigh_l}, sphereLocalOffset=<-3.711 -3.8594 3.4687>}
     {limitRadius=12.5, drivingBone {boneName=belt_L}, sphereLocalOffset=<0 -0.4206 0>}
    ]
   }
   n28 = Anim Dynamics - Bone: dyn_suspender_bk_03!(ComponentPose=@n27, Alpha=@n25) -> (Pose) {
    angularDampingOverride=1
    angularSpringConstant=1500
    bAngularSpring=true
    bGravityOverrideInSimSpace=true
    bOverrideAngularDamping=true
    bUseSphericalLimits=true
    gravityScale=0.35
    simulationSpace=RootRelative
    alphaBoolBlend {blendOption=Linear}
    alphaScaleBiasClamp {scale=0.8}
    boundBone {boneName=dyn_suspender_bk_03}
    componentAppliedLinearAccClamp=<10000 10000 10000>
    componentLinearAccScale {z=0.45}
    componentLinearVelScale {z=0.45}
    gravityOverride {z=-980}
    physicsBodyDefinitions [{
      collisionType=CustomSphere
      sphereCollisionRadius=2
      boundBone {boneName=dyn_suspender_bk_03}
      boxExtents=<10 20 20>
      constraintSetup {angularLimitsMax=<0 0 0>, angularLimitsMin=<0 0 0>, angularTarget=<0 0 0>}
      localJointOffset=<5 0 0>
     }]
    sphericalLimits [
     {limitRadius=10.5, drivingBone {boneName=thigh_l}, sphereLocalOffset=<-17.8325 -2.1546 0.0153>}
     {limitRadius=10, drivingBone {boneName=thigh_l}, sphereLocalOffset=<0 0 0>}
    ]
   }
   n29 = Get skydiving() -> (skydiving:double)
   n30 = Anim Dynamics - Bone: dyn_suspender_fr_03(ComponentPose=@n28, Alpha=@n29) {
    angularDampingOverride=1
    angularSpringConstant=1000
    bAngularSpring=true
    bGravityOverrideInSimSpace=true
    bOverrideAngularDamping=true
    bUseSphericalLimits=true
    gravityScale=0.35
    simulationSpace=RootRelative
    alphaBoolBlend {blendOption=Linear}
    alphaScaleBiasClamp {scale=0.8}
    boundBone {boneName=dyn_suspender_fr_03}
    componentAppliedLinearAccClamp=<10000 10000 10000>
    componentLinearAccScale {z=0.7}
    componentLinearVelScale {z=0.7}
    gravityOverride {z=-980}
    physicsBodyDefinitions [{
      collisionType=CustomSphere
      sphereCollisionRadius=2
      boundBone {boneName=dyn_suspender_fr_03}
      boxExtents=<10 20 20>
      constraintSetup {angularLimitsMax=<10 15 0>, angularLimitsMin=<-5 -45 -45>, angularTarget=<0 0 0>}
      localJointOffset=<5 0 0>
     }]
    sphericalLimits [
     {limitRadius=7, drivingBone {boneName=thigh_l}, sphereLocalOffset=<-3.2384 -6.1887 0.9911>}
     {limitRadius=9, drivingBone {boneName=thigh_l}, sphereLocalOffset=<-10 -2.1039 0.8464>}
    ]
   }
   n31 = Anim Dynamics - Bone: dyn_suspender_bk_01(ComponentPose=@n30, Alpha=@n29) {
    angularDampingOverride=1
    angularSpringConstant=1500
    bAngularSpring=true
    bGravityOverrideInSimSpace=true
    bOverrideAngularDamping=true
    bUseSphericalLimits=true
    gravityScale=0.35
    simulationSpace=RootRelative
    alphaBoolBlend {blendOption=Linear}
    alphaScaleBiasClamp {scale=0.8}
    boundBone {boneName=dyn_suspender_bk_01}
    componentAppliedLinearAccClamp=<10000 10000 10000>
    componentLinearAccScale {z=0.45}
    componentLinearVelScale {z=0.45}
    gravityOverride {z=-980}
    physicsBodyDefinitions [{
      collisionType=CustomSphere
      sphereCollisionRadius=1.5
      boundBone {boneName=dyn_suspender_bk_01}
      boxExtents=<10 20 20>
      constraintSetup {angularLimitsMax=<5 5 20>, angularLimitsMin=<-5 -45 -10>, angularTarget=<0 0 0>}
      localJointOffset=<5 0 0>
     }]
    sphericalLimits [
     {limitRadius=10, drivingBone {boneName=thigh_l}, sphereLocalOffset=<-3.711 -3.8594 3.4687>}
     {limitRadius=13, drivingBone {boneName=belt_L}, sphereLocalOffset=<0 -0.4206 0>}
    ]
   }
   n32 = Anim Dynamics - Bone: dyn_suspender_bk_03(ComponentPose=@n31, Alpha=@n29) {
    angularDampingOverride=1
    angularSpringConstant=1500
    bAngularSpring=true
    bGravityOverrideInSimSpace=true
    bOverrideAngularDamping=true
    bUseSphericalLimits=true
    gravityScale=0.35
    simulationSpace=RootRelative
    alphaBoolBlend {blendOption=Linear}
    alphaScaleBiasClamp {scale=0.8}
    boundBone {boneName=dyn_suspender_bk_03}
    componentAppliedLinearAccClamp=<10000 10000 10000>
    componentLinearAccScale {z=0.45}
    componentLinearVelScale {z=0.45}
    gravityOverride {z=-980}
    physicsBodyDefinitions [{
      collisionType=CustomSphere
      sphereCollisionRadius=2
      boundBone {boneName=dyn_suspender_bk_03}
      boxExtents=<10 20 20>
      constraintSetup {angularLimitsMax=<0 0 0>, angularLimitsMin=<0 0 0>, angularTarget=<0 0 0>}
      localJointOffset=<5 0 0>
     }]
    sphericalLimits [
     {limitRadius=11.5, drivingBone {boneName=thigh_l}, sphereLocalOffset=<-17.8325 -2.1546 0.0153>}
     {limitRadius=10, drivingBone {boneName=thigh_l}, sphereLocalOffset=<0 0 0>}
    ]
   }
   n33 = Anim Dynamics - Bone: dyn_shoelace_out_01_L!(ComponentPose=@n32, Alpha=1, GravityOverride:Vector="0,0,-500") -> (Pose) {
    angularBiasOverride=1
    angularDampingOverride=1
    angularSpringConstant=300
    bAngularSpring=true
    bGravityOverrideInSimSpace=true
    bOverrideAngularBias=true
    bOverrideAngularDamping=true
    bUseSphericalLimits=true
    gravityScale=0.35
    simulationSpace=RootRelative
    alphaBoolBlend {blendOption=Linear}
    boundBone {boneName=dyn_shoelace_out_01_L}
    componentAppliedLinearAccClamp=<5000 5000 5000>
    componentLinearAccScale=<0.7 0.7 0.7>
    componentLinearVelScale=<0.7 0.7 0.7>
    gravityOverride {z=-980}
    physicsBodyDefinitions [{
      collisionType=CustomSphere
      sphereCollisionRadius=1
      boundBone {boneName=dyn_shoelace_out_01_L}
      boxExtents=<12 7 7>
      constraintSetup {angularLimitsMax=<0 0 25>, angularLimitsMin=<0 -45 -25>, angularTarget=<0 0 0>}
      localJointOffset=<2 0 0>
     }]
    sphericalLimits [{limitRadius=6, drivingBone {boneName=foot_l}, sphereLocalOffset=<-4.8823 5.7986 0.6971>}]
   }
   n34 = Anim Dynamics - Bone: dyn_shoelace_inr_01_L!(ComponentPose=@n33, Alpha=1, GravityOverride:Vector="0,0,-500") -> (Pose) {
    angularBiasOverride=1
    angularDampingOverride=1
    angularSpringConstant=300
    bAngularSpring=true
    bGravityOverrideInSimSpace=true
    bOverrideAngularBias=true
    bOverrideAngularDamping=true
    bUseSphericalLimits=true
    gravityScale=0.35
    simulationSpace=RootRelative
    alphaBoolBlend {blendOption=Linear}
    boundBone {boneName=dyn_shoelace_inr_01_L}
    componentAppliedLinearAccClamp=<5000 5000 5000>
    componentLinearAccScale=<0.7 0.7 0.7>
    componentLinearVelScale=<0.7 0.7 0.7>
    gravityOverride {z=-980}
    physicsBodyDefinitions [{
      collisionType=CustomSphere
      sphereCollisionRadius=1
      boundBone {boneName=dyn_shoelace_inr_01_L}
      boxExtents=<12 7 7>
      constraintSetup {angularLimitsMax=<0 0 25>, angularLimitsMin=<0 -45 -25>, angularTarget=<0 0 0>}
      localJointOffset=<2 0 0>
     }]
    sphericalLimits [{limitRadius=6, drivingBone {boneName=foot_l}, sphereLocalOffset=<-4.8823 5.7986 0.6971>}]
   }
   n35 = Anim Dynamics - Bone: dyn_shoelace_out_01_R!(ComponentPose=@n34, Alpha=1, GravityOverride:Vector="0,0,-500") -> (Pose) {
    angularBiasOverride=1
    angularDampingOverride=1
    angularSpringConstant=300
    bAngularSpring=true
    bGravityOverrideInSimSpace=true
    bOverrideAngularBias=true
    bOverrideAngularDamping=true
    bUseSphericalLimits=true
    gravityScale=0.35
    simulationSpace=RootRelative
    alphaBoolBlend {blendOption=Linear}
    boundBone {boneName=dyn_shoelace_out_01_R}
    componentAppliedLinearAccClamp=<5000 5000 5000>
    componentLinearAccScale=<0.7 0.7 0.7>
    componentLinearVelScale=<0.7 0.7 0.7>
    gravityOverride {z=-980}
    physicsBodyDefinitions [{
      collisionType=CustomSphere
      sphereCollisionRadius=1
      boundBone {boneName=dyn_shoelace_out_01_R}
      boxExtents=<12 7 7>
      constraintSetup {angularLimitsMax=<0 0 25>, angularLimitsMin=<0 -45 -25>, angularTarget=<0 0 0>}
      localJointOffset=<2 0 0>
     }]
    sphericalLimits [{limitRadius=5.5, drivingBone {boneName=foot_r}, sphereLocalOffset=<2.8908 -6.4605 -0.2165>}]
   }
   n36 = Anim Dynamics - Bone: dyn_shoelace_inr_01_R!(ComponentPose=@n35, Alpha=1, GravityOverride:Vector="0,0,-500") -> (Pose) {
    angularBiasOverride=1
    angularDampingOverride=1
    angularSpringConstant=300
    bAngularSpring=true
    bGravityOverrideInSimSpace=true
    bOverrideAngularBias=true
    bOverrideAngularDamping=true
    bUseSphericalLimits=true
    gravityScale=0.35
    simulationSpace=RootRelative
    alphaBoolBlend {blendOption=Linear}
    boundBone {boneName=dyn_shoelace_inr_01_R}
    componentAppliedLinearAccClamp=<5000 5000 5000>
    componentLinearAccScale=<0.7 0.7 0.7>
    componentLinearVelScale=<0.7 0.7 0.7>
    gravityOverride {z=-980}
    physicsBodyDefinitions [{
      collisionType=CustomSphere
      sphereCollisionRadius=1
      boundBone {boneName=dyn_shoelace_inr_01_R}
      boxExtents=<12 7 7>
      constraintSetup {angularLimitsMax=<0 0 25>, angularLimitsMin=<0 -45 -25>, angularTarget=<0 0 0>}
      localJointOffset=<2 0 0>
     }]
    sphericalLimits [{limitRadius=5.5, drivingBone {boneName=foot_r}, sphereLocalOffset=<4.2939 -6.4605 -0.7573>}]
   }
   Output Pose(Result=Component To Local(ComponentPose=@n36)) {layerGroup=None}
   n37 = Control Rig(Source=Copy Pose From Mesh() {bUseAttachedParent=true}) {
    alphaBoolBlend {blendOption=Linear}
    controlRigAssetReference {blueprintRigClass=&/M_MED_DaffodilSong_ControlRig.M_MED_DaffodilSong_ControlRig_C}
   }
   n38 = "Constraint - dyn_clasp_lower_MSTR (PPP)"!(
    ComponentPose=Local To Component(LocalPose=@n37)
    Alpha=1
    ConstraintWeights_0=0.089
    ConstraintWeights_1:float=0.8952
    ConstraintWeights_2:float=0.0158
   ) -> (Pose) {
    alphaBoolBlend {blendOption=Linear}
    boneToModify {boneName=dyn_clasp_lower_MSTR}
    constraintSetup [{targetBone {boneName=thigh_r}}, {targetBone {boneName=thigh_twist_01_r}}, {targetBone {boneName=calf_r}}]
    constraintWeights [1,1,1]
   }
   Save cached pose 'start pose'(Pose=Component To Local(ComponentPose=@n38)) {cacheName=start pose}
   n39 = Use cached pose 'idle'()
   n40 = Use cached pose 'crouching'()
   n41 = Get bIsCrouching()
   n42 = Look At - Bone: dyn_strap_fr_aim(
    ComponentPose=Local To Component(LocalPose=Blend Poses by bool(BlendPose_0=@n40, BlendPose_1=@n39, bActiveValue=@n41) {blendType=Linear})
    Alpha=1
   ) {
    alphaBoolBlend {blendOption=Linear}
    boneToModify {boneName=dyn_strap_fr_aim}
    lookAtLocation {x=0}
    lookAtTarget {boneReference {boneName=strap_trgt}}
    lookAt_Axis {axis=<1 0>}
   }
   n43 = Get bIsInFrontEnd()
   n44 = Anim Dynamics - Bone: dyn_strap_fr_01!(ComponentPose=@n42, bAlphaBoolEnabled:bool=@n43, GravityOverride:Vector="0,0,-980") -> (Pose) {
    alphaInputType=Bool
    angularBiasOverride=1
    angularDampingOverride=1
    angularSpringConstant=1500
    bAngularSpring=true
    bGravityOverrideInSimSpace=true
    bOverrideAngularDamping=true
    bUseGravityOverride=true
    bUseSphericalLimits=true
    gravityScale=0.35
    simulationSpace=RootRelative
    alphaBoolBlend {blendInTime=0.1, blendOption=Linear, blendOutTime=0.2}
    boundBone {boneName=dyn_strap_fr_01}
    componentAppliedLinearAccClamp=<1500 1500 1500>
    componentLinearAccScale=<0.1 0.1 0.1>
    componentLinearVelScale=<0.1 0.1 0.1>
    gravityOverride {z=-980}
    physicsBodyDefinitions [{
      collisionType=CustomSphere
      sphereCollisionRadius=2
      boundBone {boneName=dyn_strap_fr_01}
      boxExtents=<25 30 30>
      constraintSetup {angularLimitsMax=<0 45 45>, angularLimitsMin=<0 -25 -45>, angularTarget=<0 0 0>}
      localJointOffset=<10 0 0>
     }]
    sphericalLimits [
     {limitRadius=8.5, drivingBone {boneName=thigh_l}, sphereLocalOffset=<-11.0672 0.4201 1.4364>}
     {limitRadius=10, drivingBone {boneName=thigh_l}, sphereLocalOffset=<-22.4212 -1.5776 2.5336>}
     {limitRadius=8.5, drivingBone {boneName=thigh_l}, sphereLocalOffset=<-31.8361 0.1647 1.8518>}
     {limitRadius=8, drivingBone {boneName=thigh_r}, sphereLocalOffset=<10.407 1.3643 0.9464>}
     {limitRadius=9, drivingBone {boneName=thigh_r}, sphereLocalOffset=<20 0 0>}
     {limitRadius=9, drivingBone {boneName=thigh_r}, sphereLocalOffset=<30 0 0>}
     {limitRadius=15, drivingBone {boneName=pelvis}, sphereLocalOffset=<-2.2013 -3.7053 -2.8198>}
     {limitRadius=8.5, drivingBone {boneName=calf_l}, sphereLocalOffset=<-15.4706 -2.6497 0>}
     {limitRadius=8.5, drivingBone {boneName=calf_twist_02_l}, sphereLocalOffset=<-5.8256 -7.8738 0>}
     {limitRadius=8.5, drivingBone {boneName=calf_twist_01_l}, sphereLocalOffset=<9.2991 0 0>}
    ]
   }
   n45 = Anim Dynamics - Bone: dyn_strap_fr_02!(ComponentPose=@n44, bAlphaBoolEnabled:bool=@n43, GravityOverride:Vector="0,0,-980") -> (Pose) {
    alphaInputType=Bool
    angularBiasOverride=1
    angularDampingOverride=1
    angularSpringConstant=1500
    bAngularSpring=true
    bGravityOverrideInSimSpace=true
    bOverrideAngularDamping=true
    bUseGravityOverride=true
    bUseSphericalLimits=true
    gravityScale=0.35
    simulationSpace=RootRelative
    alphaBoolBlend {blendInTime=0.1, blendOption=Linear, blendOutTime=0.2}
    boundBone {boneName=dyn_strap_fr_02}
    componentAppliedLinearAccClamp=<1500 1500 1500>
    componentLinearAccScale=<0.1 0.1 0.1>
    componentLinearVelScale=<0.1 0.1 0.1>
    gravityOverride {z=-980}
    physicsBodyDefinitions [{
      collisionType=CustomSphere
      sphereCollisionRadius=1.7
      boundBone {boneName=dyn_strap_fr_02}
      boxExtents=<20 25 25>
      constraintSetup {angularLimitsMax=<0 45 80>, angularLimitsMin=<0 -35 -80>, angularTarget=<0 0 0>}
      localJointOffset=<10 0 0>
     }]
    sphericalLimits [
     {limitRadius=8.5, drivingBone {boneName=thigh_l}, sphereLocalOffset=<-11.0672 0.4201 1.4364>}
     {limitRadius=10, drivingBone {boneName=thigh_l}, sphereLocalOffset=<-22.4212 -1.5776 2.5336>}
     {limitRadius=8.5, drivingBone {boneName=thigh_l}, sphereLocalOffset=<-31.8361 0.1647 1.8518>}
     {limitRadius=8, drivingBone {boneName=thigh_r}, sphereLocalOffset=<10.407 1.3643 0.9464>}
     {limitRadius=9, drivingBone {boneName=thigh_r}, sphereLocalOffset=<20 0 0>}
     {limitRadius=9, drivingBone {boneName=thigh_r}, sphereLocalOffset=<30 0 0>}
     {limitRadius=15, drivingBone {boneName=pelvis}, sphereLocalOffset=<-2.2013 -3.7053 -2.8198>}
     {limitRadius=8.5, drivingBone {boneName=calf_l}, sphereLocalOffset=<-15.4706 -2.6497 0>}
     {limitRadius=8.5, drivingBone {boneName=calf_twist_02_l}, sphereLocalOffset=<-5.8256 -7.8738 0>}
     {limitRadius=8.5, drivingBone {boneName=calf_twist_01_l}, sphereLocalOffset=<9.2991 0 0>}
    ]
   }
   n46 = Anim Dynamics - Bone: dyn_strap_fr_03!(ComponentPose=@n45, bAlphaBoolEnabled:bool=@n43, GravityOverride:Vector="0,0,-980") -> (Pose) {
    alphaInputType=Bool
    angularBiasOverride=1
    angularDampingOverride=1
    angularSpringConstant=1500
    bAngularSpring=true
    bGravityOverrideInSimSpace=true
    bOverrideAngularDamping=true
    bUseGravityOverride=true
    bUseSphericalLimits=true
    gravityScale=0.35
    simulationSpace=RootRelative
    alphaBoolBlend {blendInTime=0.1, blendOption=Linear, blendOutTime=0.2}
    boundBone {boneName=dyn_strap_fr_03}
    componentAppliedLinearAccClamp=<1500 1500 1500>
    componentLinearAccScale=<0.1 0.1 0.1>
    componentLinearVelScale=<0.1 0.1 0.1>
    gravityOverride {z=-980}
    physicsBodyDefinitions [{
      collisionType=CustomSphere
      sphereCollisionRadius=1.7
      boundBone {boneName=dyn_strap_fr_03}
      boxExtents=<15 20 20>
      constraintSetup {angularLimitsMax=<0 45 80>, angularLimitsMin=<0 -45 -80>, angularTarget=<0 0 0>}
      localJointOffset=<10 0 0>
     }]
    sphericalLimits [
     {limitRadius=8.5, drivingBone {boneName=thigh_l}, sphereLocalOffset=<-11.0672 0.4201 1.4364>}
     {limitRadius=10, drivingBone {boneName=thigh_l}, sphereLocalOffset=<-22.4212 -1.5776 2.5336>}
     {limitRadius=8.5, drivingBone {boneName=thigh_l}, sphereLocalOffset=<-31.8361 0.1647 1.8518>}
     {limitRadius=8, drivingBone {boneName=thigh_r}, sphereLocalOffset=<10.407 1.3643 0.9464>}
     {limitRadius=9, drivingBone {boneName=thigh_r}, sphereLocalOffset=<20 0 0>}
     {limitRadius=9, drivingBone {boneName=thigh_r}, sphereLocalOffset=<30 0 0>}
     {limitRadius=15, drivingBone {boneName=pelvis}, sphereLocalOffset=<-2.2013 -3.7053 -2.8198>}
     {limitRadius=8.5, drivingBone {boneName=calf_l}, sphereLocalOffset=<-15.4706 -2.6497 0>}
     {limitRadius=8.5, drivingBone {boneName=calf_twist_02_l}, sphereLocalOffset=<-5.8256 -7.8738 0>}
     {limitRadius=8.5, drivingBone {boneName=calf_twist_01_l}, sphereLocalOffset=<9.2991 0 0>}
    ]
   }
   n47 = "Anim Dynamics - Chain: dyn_strap_fr_01 -> dyn_strap_fr_03"!(ComponentPose=@n46, bAlphaBoolEnabled:bool=Get Not in FrontEnd(), GravityOverride:Vector="0,0,-980") -> (Pose) {
    alphaInputType=Bool
    angularBiasOverride=1
    angularDampingOverride=1
    angularSpringConstant=1500
    bAngularSpring=true
    bChain=true
    bGravityOverrideInSimSpace=true
    bOverrideAngularDamping=true
    bUseGravityOverride=true
    bUseSphericalLimits=true
    gravityScale=0.35
    simulationSpace=RootRelative
    alphaBoolBlend {blendInTime=0.1, blendOption=Linear, blendOutTime=0.2}
    boundBone {boneName=dyn_strap_fr_01}
    chainEnd {boneName=dyn_strap_fr_03}
    componentAppliedLinearAccClamp=<1500 1500 1500>
    componentLinearAccScale=<0.1 0.1 0.1>
    componentLinearVelScale=<0.1 0.1 0.1>
    gravityOverride {z=-980}
    physicsBodyDefinitions [
     {
      collisionType=CustomSphere
      sphereCollisionRadius=2
      boundBone {boneName=dyn_strap_fr_01}
      boxExtents=<25 30 30>
      constraintSetup {angularLimitsMax=<0 45 45>, angularLimitsMin=<0 -25 -45>, angularTarget=<0 0 0>}
      localJointOffset=<10 0 0>
     }
     {
      collisionType=CustomSphere
      sphereCollisionRadius=1.7
      boundBone {boneName=dyn_strap_fr_02}
      boxExtents=<20 25 25>
      constraintSetup {angularLimitsMax=<0 45 80>, angularLimitsMin=<0 -35 -80>, angularTarget=<0 0 0>}
      localJointOffset=<10 0 0>
     }
     {
      collisionType=CustomSphere
      sphereCollisionRadius=1.7
      boundBone {boneName=dyn_strap_fr_03}
      boxExtents=<15 20 20>
      constraintSetup {angularLimitsMax=<0 45 80>, angularLimitsMin=<0 -45 -80>, angularTarget=<0 0 0>}
      localJointOffset=<10 0 0>
     }
    ]
    sphericalLimits [
     {limitRadius=8.5, drivingBone {boneName=thigh_l}, sphereLocalOffset=<-11.0672 0.4201 1.4364>}
     {limitRadius=10, drivingBone {boneName=thigh_l}, sphereLocalOffset=<-22.4212 -1.5776 2.5336>}
     {limitRadius=8.5, drivingBone {boneName=thigh_l}, sphereLocalOffset=<-31.8361 0.1647 1.8518>}
     {limitRadius=8, drivingBone {boneName=thigh_r}, sphereLocalOffset=<10.407 1.3643 0.9464>}
     {limitRadius=9, drivingBone {boneName=thigh_r}, sphereLocalOffset=<20 0 0>}
     {limitRadius=9, drivingBone {boneName=thigh_r}, sphereLocalOffset=<30 0 0>}
     {limitRadius=15, drivingBone {boneName=pelvis}, sphereLocalOffset=<-2.2013 -3.7053 -2.8198>}
     {limitRadius=8.5, drivingBone {boneName=calf_l}, sphereLocalOffset=<-15.4706 -2.6497 0>}
     {limitRadius=8.5, drivingBone {boneName=calf_twist_02_l}, sphereLocalOffset=<-5.8256 -7.8738 0>}
     {limitRadius=8.5, drivingBone {boneName=calf_twist_01_l}, sphereLocalOffset=<9.2991 0 0>}
    ]
   }
   n48 = Component To Local(ComponentPose=@n47)
   n49 = Use cached pose 'dbno'()
   n50 = Get bIsDBNO()
   Save cached pose 'straps pose'(Pose=Blend Poses by bool(
     BlendPose_0=Use cached pose 'skydiving'()
     BlendPose_1=Blend Poses by bool(BlendPose_0=@n49, BlendPose_1=@n48, bActiveValue=@n50) {blendType=Linear}
     bActiveValue=Get bIsSkyDiving()
    ) {blendType=Linear}) {cacheName=straps pose}
   n51 = "Transform (Modify) Bone - Bone: dyn_suspender_bk_MSTR"!(ComponentPose=Local To Component(LocalPose=Use cached pose 'start pose'()), Alpha=1, Translation="0,0,0", Rotation="-0.000,-10.000,-0.000") -> (Pose) {
    rotationMode=BMM_Additive
    rotationSpace=BCS_BoneSpace
    translationMode=BMM_Additive
    translationSpace=BCS_BoneSpace
    alphaBoolBlend {blendOption=Linear}
    alphaScaleBiasClamp {interpSpeedDecreasing=20, interpSpeedIncreasing=20}
    boneToModify {boneName=dyn_suspender_bk_MSTR}
    rotation=<rot -0.0002 -10 0>
   }
   n52 = "Transform (Modify) Bone - Bone: suspender_trgt"!(ComponentPose=@n51, Alpha=1, Translation="2.136,12.587,0.729") -> (Pose) {
    rotationSpace=BCS_BoneSpace
    translationMode=BMM_Additive
    translationSpace=BCS_BoneSpace
    alphaBoolBlend {blendOption=Linear}
    boneToModify {boneName=suspender_trgt}
    translation=<2.136 12.587 0.729>
   }
   n53 = Look At - Bone: dyn_suspender_fr_aim(ComponentPose=@n52, Alpha=1) {
    alphaBoolBlend {blendOption=Linear}
    boneToModify {boneName=dyn_suspender_fr_aim}
    lookAtLocation=<-2.0549 -0.0627 5.157>
    lookAtTarget {boneReference {boneName=suspender_trgt}}
    lookAt_Axis {axis=<1 0>}
   }
   n54 = "Transform (Modify) Bone - Bone: dyn_clasp_upper_01"!(ComponentPose=@n53, Alpha=1, Translation="-2.142,-0.435,0.183", Rotation="66.274,-32.929,-2.056") -> (Pose) {
    rotationMode=BMM_Additive
    rotationSpace=BCS_BoneSpace
    translationMode=BMM_Additive
    translationSpace=BCS_BoneSpace
    alphaBoolBlend {blendOption=Linear}
    boneToModify {boneName=dyn_clasp_upper_01}
    rotation=<rot 66.274 -32.9291 -2.056>
    translation=<-2.1419 -0.4355 0.1831>
   }
   n55 = "Transform (Modify) Bone - Bone: dyn_clasp_upper_01"(ComponentPose=@n54, Alpha=1, Translation="0.371,-0.367,0.857", Rotation="38.876,86.931,21.693") {
    rotationMode=BMM_Additive
    rotationSpace=BCS_BoneSpace
    translationMode=BMM_Additive
    translationSpace=BCS_BoneSpace
    alphaBoolBlend {blendOption=Linear}
    boneToModify {boneName=dyn_clasp_upper_01}
    rotation=<rot 38.876 86.931 21.6931>
    translation=<0.371 -0.367 0.857>
   }
   n56 = "Transform (Modify) Bone - Bone: dyn_clasp_lower_01"!(ComponentPose=@n55, Alpha=1, Translation="0,0,0", Rotation="4.742,29.164,0.670") -> (Pose) {
    rotationMode=BMM_Additive
    rotationSpace=BCS_BoneSpace
    translationMode=BMM_Additive
    translationSpace=BCS_BoneSpace
    alphaBoolBlend {blendOption=Linear}
    boneToModify {boneName=dyn_clasp_lower_01}
    rotation=<rot 4.742 29.164 0.67>
   }
   n57 = Look At - Bone: dyn_strap_fr_aim(ComponentPose=@n56, Alpha=1) {
    alphaBoolBlend {blendOption=Linear}
    boneToModify {boneName=dyn_strap_fr_aim}
    lookAtLocation {x=0}
    lookAtTarget {boneReference {boneName=strap_trgt}}
    lookAt_Axis {axis=<1 0>}
   }
   n58 = "Anim Dynamics - Chain: dyn_strap_fr_01 -> dyn_strap_fr_03"!(ComponentPose=@n57, Alpha:float=1, GravityOverride="0,0,-980") -> (Pose) {
    angularBiasOverride=1
    angularDampingOverride=1
    angularSpringConstant=1500
    bAngularSpring=true
    bChain=true
    bGravityOverrideInSimSpace=true
    bOverrideAngularDamping=true
    bUseGravityOverride=true
    bUseSphericalLimits=true
    gravityScale=0.35
    simulationSpace=RootRelative
    alphaBoolBlend {blendOption=Linear}
    boundBone {boneName=dyn_strap_fr_01}
    chainEnd {boneName=dyn_strap_fr_03}
    componentAppliedLinearAccClamp=<1500 1500 1500>
    componentLinearAccScale=<0.1 0.1 0.1>
    componentLinearVelScale=<0.1 0.1 0.1>
    gravityOverride {z=-980}
    physicsBodyDefinitions [
     {
      collisionType=CustomSphere
      sphereCollisionRadius=2
      boundBone {boneName=dyn_strap_fr_01}
      boxExtents=<25 30 30>
      constraintSetup {angularLimitsMax=<0 5 5>, angularLimitsMin=<0 -5 -5>, angularTarget=<0 0 0>}
      localJointOffset=<10 0 0>
     }
     {
      collisionType=CustomSphere
      sphereCollisionRadius=1.7
      boundBone {boneName=dyn_strap_fr_02}
      boxExtents=<20 25 25>
      constraintSetup {angularLimitsMax=<0 10 10>, angularLimitsMin=<0 -10 -10>, angularTarget=<0 0 0>}
      localJointOffset=<10 0 0>
     }
     {
      collisionType=CustomSphere
      sphereCollisionRadius=1.7
      boundBone {boneName=dyn_strap_fr_03}
      boxExtents=<15 20 20>
      constraintSetup {angularLimitsMax=<0 15 15>, angularLimitsMin=<0 -15 -15>, angularTarget=<0 0 0>}
      localJointOffset=<10 0 0>
     }
    ]
    sphericalLimits [
     {limitRadius=8.5, drivingBone {boneName=thigh_l}, sphereLocalOffset=<-11.0672 0.4201 1.4364>}
     {limitRadius=10, drivingBone {boneName=thigh_l}, sphereLocalOffset=<-22.4212 -1.5776 2.5336>}
     {limitRadius=8.5, drivingBone {boneName=thigh_l}, sphereLocalOffset=<-31.8361 0.1647 1.8518>}
     {limitRadius=8, drivingBone {boneName=thigh_r}, sphereLocalOffset=<10.407 1.3643 0.9464>}
     {limitRadius=9, drivingBone {boneName=thigh_r}, sphereLocalOffset=<20 0 0>}
     {limitRadius=9, drivingBone {boneName=thigh_r}, sphereLocalOffset=<30 0 0>}
     {limitRadius=15, drivingBone {boneName=pelvis}, sphereLocalOffset=<-2.2013 -3.7053 -2.8198>}
     {limitRadius=8.5, drivingBone {boneName=calf_l}, sphereLocalOffset=<-15.4706 -2.6497 0>}
     {limitRadius=8.5, drivingBone {boneName=calf_twist_02_l}, sphereLocalOffset=<-5.8256 -7.8738 0>}
     {limitRadius=8.5, drivingBone {boneName=calf_twist_01_l}, sphereLocalOffset=<9.2991 0 0>}
    ]
   }
   Save cached pose 'skydiving'(Pose=Component To Local(ComponentPose=@n58)) {cacheName=skydiving}
   n59 = "Transform (Modify) Bone - Bone: belt_L"!(
    ComponentPose=Local To Component(LocalPose=Use cached pose 'start pose'())
    Alpha=1
    Translation="4.879,0.559,0.158"
    Rotation="9.998,2.000,-0.011"
   ) -> (Pose) {
    rotationMode=BMM_Additive
    rotationSpace=BCS_BoneSpace
    translationMode=BMM_Additive
    translationSpace=BCS_BoneSpace
    alphaBoolBlend {blendOption=Linear}
    boneToModify {boneName=belt_L}
    rotation=<rot 9.998 1.9999 -0.011>
    translation=<4.879 0.559 0.158>
   }
   n60 = "Transform (Modify) Bone - Bone: belt_R"!(ComponentPose=@n59, Alpha=1, Translation="2.863,-0.000,0.045", Rotation="3,0,0") -> (Pose) {
    rotationMode=BMM_Additive
    rotationSpace=BCS_BoneSpace
    translationMode=BMM_Additive
    translationSpace=BCS_BoneSpace
    alphaBoolBlend {blendOption=Linear}
    boneToModify {boneName=belt_R}
    rotation {pitch=3}
    translation {x=2.863, z=0.045}
   }
   n61 = "Constraint - strap_trgt_zero (P)"(ComponentPose=@n60, Alpha=1, ConstraintWeights_0=1) {
    alphaBoolBlend {blendOption=Linear}
    boneToModify {boneName=strap_trgt_zero}
    constraintSetup [{targetBone {boneName=pelvis}}]
    constraintWeights [1]
   }
   n62 = "Transform (Modify) Bone - Bone: dyn_strap_fr_MSTR"!(ComponentPose=@n61, Alpha=1, Translation="0,0,0", Rotation="16.418,2.668,-31.963") -> (Pose) {
    rotationMode=BMM_Additive
    rotationSpace=BCS_BoneSpace
    translationMode=BMM_Additive
    translationSpace=BCS_BoneSpace
    alphaBoolBlend {blendOption=Linear}
    boneToModify {boneName=dyn_strap_fr_MSTR}
    rotation=<rot 16.418 2.668 -31.963>
   }
   n63 = "Transform (Modify) Bone - Bone: strap_trgt"(ComponentPose=@n62, Alpha=1, Translation="12.327,18.939,15.265") {
    rotationSpace=BCS_BoneSpace
    translationMode=BMM_Additive
    translationSpace=BCS_BoneSpace
    alphaBoolBlend {blendOption=Linear}
    alphaScaleBiasClamp {interpSpeedDecreasing=20, interpSpeedIncreasing=20}
    boneToModify {boneName=strap_trgt}
    translation=<12.327 18.939 15.265>
   }
   n64 = Look At - Bone: dyn_strap_fr_aim(ComponentPose=@n63, Alpha=1) {
    alphaBoolBlend {blendOption=Linear}
    boneToModify {boneName=dyn_strap_fr_aim}
    lookAtLocation {x=0}
    lookAtTarget {boneReference {boneName=strap_trgt}}
    lookAt_Axis {axis=<1 0>}
   }
   n65 = "Anim Dynamics - Chain: dyn_strap_fr_01 -> dyn_strap_fr_03"(ComponentPose=@n64, Alpha=1, GravityOverride="0,0,-980") {
    angularBiasOverride=1
    angularDampingOverride=1
    angularSpringConstant=1500
    bAngularSpring=true
    bChain=true
    bGravityOverrideInSimSpace=true
    bOverrideAngularDamping=true
    bUseGravityOverride=true
    bUseSphericalLimits=true
    gravityScale=0.35
    simulationSpace=RootRelative
    alphaBoolBlend {blendOption=Linear}
    boundBone {boneName=dyn_strap_fr_01}
    chainEnd {boneName=dyn_strap_fr_03}
    componentAppliedLinearAccClamp=<1500 1500 1500>
    componentLinearAccScale=<0.1 0.1 0.1>
    componentLinearVelScale=<0.1 0.1 0.1>
    gravityOverride {z=-980}
    physicsBodyDefinitions [
     {
      collisionType=CustomSphere
      sphereCollisionRadius=2
      boundBone {boneName=dyn_strap_fr_01}
      boxExtents=<25 30 30>
      constraintSetup {angularLimitsMax=<0 15 0>, angularLimitsMin=<0 -15 -5>, angularTarget=<0 0 0>}
      localJointOffset=<10 0 0>
     }
     {
      collisionType=CustomSphere
      sphereCollisionRadius=2
      boundBone {boneName=dyn_strap_fr_02}
      boxExtents=<20 25 25>
      constraintSetup {angularLimitsMax=<0 45 0>, angularLimitsMin=<0 -35 -10>, angularTarget=<0 0 0>}
      localJointOffset=<10 0 0>
     }
     {
      collisionType=CustomSphere
      sphereCollisionRadius=2
      boundBone {boneName=dyn_strap_fr_03}
      boxExtents=<15 20 20>
      constraintSetup {angularLimitsMax=<0 45 0>, angularLimitsMin=<0 -45 -15>, angularTarget=<0 0 0>}
      localJointOffset=<10 0 0>
     }
    ]
    sphericalLimits [
     {limitRadius=10.5, drivingBone {boneName=thigh_l}, sphereLocalOffset=<-11.0672 0.4201 1.4364>}
     {limitRadius=10, drivingBone {boneName=thigh_l}, sphereLocalOffset=<-22.4212 -1.5776 2.5336>}
     {limitRadius=8.5, drivingBone {boneName=thigh_l}, sphereLocalOffset=<-31.8361 0.1647 1.8518>}
     {limitRadius=8, drivingBone {boneName=thigh_r}, sphereLocalOffset=<10.407 1.3643 0.9464>}
     {limitRadius=9, drivingBone {boneName=thigh_r}, sphereLocalOffset=<20 0 0>}
     {limitRadius=9, drivingBone {boneName=thigh_r}, sphereLocalOffset=<30 0 0>}
     {limitRadius=15, drivingBone {boneName=pelvis}, sphereLocalOffset=<-2.2013 -3.7053 -2.8198>}
     {limitRadius=8.5, drivingBone {boneName=calf_l}, sphereLocalOffset=<-15.4706 -2.6497 0>}
     {limitRadius=8.5, drivingBone {boneName=calf_twist_02_l}, sphereLocalOffset=<-5.8256 -7.8738 0>}
     {limitRadius=8.5, drivingBone {boneName=calf_twist_01_l}, sphereLocalOffset=<9.2991 0 0>}
    ]
   }
   n66 = "Transform (Modify) Bone - Bone: dyn_clasp_upper_01"(ComponentPose=@n65, Alpha=1, Translation="-2.142,-0.435,0.183", Rotation="66.274,-32.929,-2.056") {
    rotationMode=BMM_Additive
    rotationSpace=BCS_BoneSpace
    translationMode=BMM_Additive
    translationSpace=BCS_BoneSpace
    alphaBoolBlend {blendOption=Linear}
    boneToModify {boneName=dyn_clasp_upper_01}
    rotation=<rot 66.274 -32.9291 -2.056>
    translation=<-2.1419 -0.4355 0.1831>
   }
   n67 = "Transform (Modify) Bone - Bone: dyn_clasp_upper_01"(ComponentPose=@n66, Alpha=1, Translation="0.371,-0.367,0.857", Rotation="38.876,86.931,21.693") {
    rotationMode=BMM_Additive
    rotationSpace=BCS_BoneSpace
    translationMode=BMM_Additive
    translationSpace=BCS_BoneSpace
    alphaBoolBlend {blendOption=Linear}
    boneToModify {boneName=dyn_clasp_upper_01}
    rotation=<rot 38.876 86.931 21.6931>
    translation=<0.371 -0.367 0.857>
   }
   n68 = "Transform (Modify) Bone - Bone: dyn_clasp_lower_01"(ComponentPose=@n67, Alpha=1, Translation="0,0,0", Rotation="4.742,29.164,0.670") {
    rotationMode=BMM_Additive
    rotationSpace=BCS_BoneSpace
    translationMode=BMM_Additive
    translationSpace=BCS_BoneSpace
    alphaBoolBlend {blendOption=Linear}
    boneToModify {boneName=dyn_clasp_lower_01}
    rotation=<rot 4.742 29.164 0.67>
   }
   Save cached pose 'dbno'(Pose=Component To Local(ComponentPose=@n68)) {cacheName=dbno}
   n69 = "Transform (Modify) Bone - Bone: belt_L"(
    ComponentPose=Local To Component(LocalPose=Use cached pose 'start pose'())
    Alpha=1
    Translation="3.510,0.511,-0.081"
    Rotation="9.998,2.000,-0.011"
   ) {
    rotationMode=BMM_Additive
    rotationSpace=BCS_BoneSpace
    translationMode=BMM_Additive
    translationSpace=BCS_BoneSpace
    alphaBoolBlend {blendOption=Linear}
    alphaScaleBiasClamp {bInterpResult=true}
    boneToModify {boneName=belt_L}
    rotation=<rot 9.998 1.9999 -0.011>
    translation=<3.51 0.511 -0.081>
   }
   n70 = "Transform (Modify) Bone - Bone: belt_R"(ComponentPose=@n69, Alpha=1, Translation="2,0,0", Rotation="3,0,0") {
    rotationMode=BMM_Additive
    rotationSpace=BCS_BoneSpace
    translationMode=BMM_Additive
    translationSpace=BCS_BoneSpace
    alphaBoolBlend {blendOption=Linear}
    alphaScaleBiasClamp {bInterpResult=true}
    boneToModify {boneName=belt_R}
    rotation {pitch=3}
    translation {x=2}
   }
   n71 = "Transform (Modify) Bone - Bone: dyn_strap_fr_MSTR"(ComponentPose=@n70, Alpha=1, Translation="-4.340,-1.110,3.448", Rotation="42.275,-6.622,-22.684") {
    rotationMode=BMM_Additive
    rotationSpace=BCS_BoneSpace
    translationMode=BMM_Additive
    translationSpace=BCS_BoneSpace
    alphaBoolBlend {blendOption=Linear}
    boneToModify {boneName=dyn_strap_fr_MSTR}
    rotation=<rot 42.2751 -6.622 -22.6841>
    translation=<-4.34 -1.11 3.448>
   }
   n72 = "Transform (Modify) Bone - Bone: dyn_strap_fr_MSTR"(ComponentPose=@n71, Alpha=1, Translation="-4.340,-1.110,3.448", Rotation="33.130,-34.358,-31.655") {
    rotationMode=BMM_Additive
    rotationSpace=BCS_BoneSpace
    translationMode=BMM_Additive
    translationSpace=BCS_BoneSpace
    alphaBoolBlend {blendOption=Linear}
    alphaScaleBiasClamp {interpSpeedDecreasing=5, interpSpeedIncreasing=5}
    boneToModify {boneName=dyn_strap_fr_MSTR}
    rotation=<rot 33.1301 -34.358 -31.6551>
    translation=<-4.34 -1.11 3.448>
   }
   n73 = "Transform (Modify) Bone - Bone: dyn_strap_fr_01"!(ComponentPose=@n72, Alpha=1, Translation="0,0,0", Rotation="29.405,-44.407,-15.474") -> (Pose) {
    rotationMode=BMM_Additive
    rotationSpace=BCS_BoneSpace
    translationMode=BMM_Additive
    translationSpace=BCS_BoneSpace
    alphaBoolBlend {blendOption=Linear}
    boneToModify {boneName=dyn_strap_fr_01}
    rotation=<rot 29.405 -44.407 -15.4741>
   }
   n74 = "Transform (Modify) Bone - Bone: strap_trgt"(ComponentPose=@n73, Alpha=1, Translation="0.739,-44.893,-8.133") {
    rotationSpace=BCS_BoneSpace
    translationMode=BMM_Additive
    translationSpace=BCS_BoneSpace
    alphaBoolBlend {blendOption=Linear}
    alphaScaleBiasClamp {interpSpeedDecreasing=20, interpSpeedIncreasing=20}
    boneToModify {boneName=strap_trgt}
    translation=<0.739 -44.893 -8.133>
   }
   n75 = "Transform (Modify) Bone - Bone: dyn_clasp_upper_01"(ComponentPose=@n74, Alpha=1, Translation="-2.142,-0.435,0.183", Rotation="66.274,-32.929,-2.056") {
    rotationMode=BMM_Additive
    rotationSpace=BCS_BoneSpace
    translationMode=BMM_Additive
    translationSpace=BCS_BoneSpace
    alphaBoolBlend {blendOption=Linear}
    boneToModify {boneName=dyn_clasp_upper_01}
    rotation=<rot 66.274 -32.9291 -2.056>
    translation=<-2.1419 -0.4355 0.1831>
   }
   n76 = "Transform (Modify) Bone - Bone: dyn_clasp_upper_01"(ComponentPose=@n75, Alpha=1, Translation="0.371,-0.367,0.857", Rotation="38.876,86.931,21.693") {
    rotationMode=BMM_Additive
    rotationSpace=BCS_BoneSpace
    translationMode=BMM_Additive
    translationSpace=BCS_BoneSpace
    alphaBoolBlend {blendOption=Linear}
    boneToModify {boneName=dyn_clasp_upper_01}
    rotation=<rot 38.876 86.931 21.6931>
    translation=<0.371 -0.367 0.857>
   }
   n77 = "Transform (Modify) Bone - Bone: dyn_clasp_lower_01"(ComponentPose=@n76, Alpha=1, Translation="0,0,0", Rotation="-10.214,28.993,-0.482") {
    rotationMode=BMM_Additive
    rotationSpace=BCS_BoneSpace
    translationMode=BMM_Additive
    translationSpace=BCS_BoneSpace
    alphaBoolBlend {blendOption=Linear}
    boneToModify {boneName=dyn_clasp_lower_01}
    rotation=<rot -10.214 28.993 -0.482>
   }
   n78 = "Constraint - suspender_trgt_zero (P)"(ComponentPose=@n77, Alpha=1, ConstraintWeights_0=1) {
    alphaBoolBlend {blendOption=Linear}
    boneToModify {boneName=suspender_trgt_zero}
    constraintSetup [{targetBone {boneName=pelvis}}]
    constraintWeights [1]
   }
   n79 = "Transform (Modify) Bone - Bone: dyn_suspender_fr_MSTR"!(ComponentPose=@n78, Alpha=1, Translation="0.091,-2.221,-0.011", Rotation="1.162,19.286,-4.981") -> (Pose) {
    rotationMode=BMM_Additive
    rotationSpace=BCS_BoneSpace
    translationMode=BMM_Additive
    translationSpace=BCS_BoneSpace
    alphaBoolBlend {blendOption=Linear}
    boneToModify {boneName=dyn_suspender_fr_MSTR}
    rotation=<rot 1.1621 19.2861 -4.9811>
    translation=<0.091 -2.221 -0.011>
   }
   n80 = "Transform (Modify) Bone - Bone: suspender_trgt"(ComponentPose=@n79, Alpha=1, Translation="7,0,0") {
    rotationSpace=BCS_BoneSpace
    translationMode=BMM_Additive
    translationSpace=BCS_BoneSpace
    alphaBoolBlend {blendOption=Linear}
    boneToModify {boneName=suspender_trgt}
    translation {x=7}
   }
   n81 = Look At - Bone: dyn_suspender_fr_aim(ComponentPose=@n80, Alpha=1) {
    alphaBoolBlend {blendOption=Linear}
    boneToModify {boneName=dyn_suspender_fr_aim}
    lookAtLocation=<-2.0549 -0.0627 5.157>
    lookAtTarget {boneReference {boneName=suspender_trgt}}
    lookAt_Axis {axis=<1 0>}
   }
   Save cached pose 'crouching'(Pose=Component To Local(ComponentPose=@n81)) {cacheName=crouching}
   n82 = "Transform (Modify) Bone - Bone: belt_L"(ComponentPose=Local To Component(LocalPose=Use cached pose 'start pose'()), Alpha=1, Translation="1,0,0", Rotation="0,0,0") {
    rotationMode=BMM_Additive
    rotationSpace=BCS_BoneSpace
    translationMode=BMM_Additive
    translationSpace=BCS_BoneSpace
    alphaBoolBlend {blendOption=Linear}
    boneToModify {boneName=belt_L}
    translation {x=1}
   }
   n83 = "Transform (Modify) Bone - Bone: belt_L"(ComponentPose=@n82, Alpha=1, Translation="4.517,0.153,-0.169", Rotation="12.858,10.243,1.733") {
    rotationMode=BMM_Additive
    rotationSpace=BCS_BoneSpace
    translationMode=BMM_Additive
    translationSpace=BCS_BoneSpace
    alphaBoolBlend {blendOption=Linear}
    boneToModify {boneName=belt_L}
    rotation=<rot 12.858 10.2429 1.733>
    translation=<4.517 0.153 -0.169>
   }
   n84 = "Transform (Modify) Bone - Bone: belt_R"(ComponentPose=@n83, Alpha=1, Translation="1.7,0,0", Rotation="0,0,0") {
    rotationMode=BMM_Additive
    rotationSpace=BCS_BoneSpace
    translationMode=BMM_Additive
    translationSpace=BCS_BoneSpace
    alphaBoolBlend {blendOption=Linear}
    boneToModify {boneName=belt_R}
    translation {x=1.7}
   }
   n85 = "Transform (Modify) Bone - Bone: strap_trgt"(ComponentPose=@n84, Alpha=1, Translation="23.317,49.236,-14.078") {
    rotationSpace=BCS_BoneSpace
    translationMode=BMM_Additive
    translationSpace=BCS_BoneSpace
    alphaBoolBlend {blendOption=Linear}
    alphaScaleBiasClamp {interpSpeedDecreasing=20, interpSpeedIncreasing=20}
    boneToModify {boneName=strap_trgt}
    translation=<23.317 49.236 -14.078>
   }
   n86 = "Constraint - strap_trgt_zero (P)"(ComponentPose=@n85, Alpha=1, ConstraintWeights_0=1) {
    alphaBoolBlend {blendOption=Linear}
    boneToModify {boneName=strap_trgt_zero}
    constraintSetup [{targetBone {boneName=pelvis}}]
    constraintWeights [1]
   }
   n87 = "Transform (Modify) Bone - Bone: strap_trgt"(ComponentPose=@n86, Alpha=1, Translation="7.996,4.074,8.821") {
    rotationSpace=BCS_BoneSpace
    translationMode=BMM_Additive
    translationSpace=BCS_BoneSpace
    alphaBoolBlend {blendOption=Linear}
    alphaScaleBiasClamp {interpSpeedDecreasing=20, interpSpeedIncreasing=20}
    boneToModify {boneName=strap_trgt}
    translation=<7.996 4.074 8.821>
   }
   n88 = "Transform (Modify) Bone - Bone: dyn_strap_fr_MSTR"(ComponentPose=@n87, Alpha=1, Translation="0,0,0", Rotation="40,-18.579,-47.338") {
    rotationMode=BMM_Additive
    rotationSpace=BCS_BoneSpace
    translationMode=BMM_Additive
    translationSpace=BCS_BoneSpace
    alphaBoolBlend {blendOption=Linear}
    boneToModify {boneName=dyn_strap_fr_MSTR}
    rotation=<rot 40 -18.579 -47.338>
   }
   n89 = "Transform (Modify) Bone - Bone: dyn_clasp_upper_01"(ComponentPose=@n88, Alpha=1, Translation="-2.002,-0.119,-0.362", Rotation="63.005,-77.648,-59.153") {
    rotationMode=BMM_Additive
    rotationSpace=BCS_BoneSpace
    translationMode=BMM_Additive
    translationSpace=BCS_BoneSpace
    alphaBoolBlend {blendOption=Linear}
    boneToModify {boneName=dyn_clasp_upper_01}
    rotation=<rot 63.005 -77.6481 -59.1531>
    translation=<-2.002 -0.119 -0.362>
   }
   n90 = "Transform (Modify) Bone - Bone: dyn_clasp_upper_01"(ComponentPose=@n89, Alpha=1, Translation="0.371,-0.367,0.857", Rotation="38.876,86.931,21.693") {
    rotationMode=BMM_Additive
    rotationSpace=BCS_BoneSpace
    translationMode=BMM_Additive
    translationSpace=BCS_BoneSpace
    alphaBoolBlend {blendOption=Linear}
    boneToModify {boneName=dyn_clasp_upper_01}
    rotation=<rot 38.876 86.931 21.6931>
    translation=<0.371 -0.367 0.857>
   }
   n91 = "Transform (Modify) Bone - Bone: dyn_clasp_lower_01"(ComponentPose=@n90, Alpha=1, Translation="0,0,0", Rotation="-10.258,28.994,0.604") {
    rotationMode=BMM_Additive
    rotationSpace=BCS_BoneSpace
    translationMode=BMM_Additive
    translationSpace=BCS_BoneSpace
    alphaBoolBlend {blendOption=Linear}
    boneToModify {boneName=dyn_clasp_lower_01}
    rotation=<rot -10.258 28.994 0.604>
   }
   n92 = "Transform (Modify) Bone - Bone: dyn_suspender_fr_MSTR"(ComponentPose=@n91, Alpha=1, Translation="-1.029,-0.131,0.891", Rotation="-11.260,6.341,-31.639") {
    rotationMode=BMM_Additive
    rotationSpace=BCS_BoneSpace
    translationMode=BMM_Additive
    translationSpace=BCS_BoneSpace
    alphaBoolBlend {blendOption=Linear}
    boneToModify {boneName=dyn_suspender_fr_MSTR}
    rotation=<rot -11.2598 6.3414 -31.6391>
    translation=<-1.029 -0.131 0.891>
   }
   n93 = "Transform (Modify) Bone - Bone: dyn_suspender_fr_MSTR"(ComponentPose=@n92, Alpha=1, Translation="0,0,0", Rotation="20,0,0") {
    rotationMode=BMM_Additive
    rotationSpace=BCS_BoneSpace
    translationMode=BMM_Additive
    translationSpace=BCS_BoneSpace
    alphaBoolBlend {blendOption=Linear}
    boneToModify {boneName=dyn_suspender_fr_MSTR}
    rotation {pitch=20}
   }
   n94 = "Transform (Modify) Bone - Bone: suspender_trgt"(ComponentPose=@n93, Alpha=1, Translation="58.128,-8.690,-6.606") {
    rotationSpace=BCS_BoneSpace
    translationMode=BMM_Additive
    translationSpace=BCS_BoneSpace
    alphaBoolBlend {blendOption=Linear}
    boneToModify {boneName=suspender_trgt}
    translation=<58.128 -8.69 -6.606>
   }
   n95 = Look At - Bone: dyn_suspender_fr_aim(ComponentPose=@n94, Alpha=1) {
    alphaBoolBlend {blendOption=Linear}
    boneToModify {boneName=dyn_suspender_fr_aim}
    lookAtLocation=<-2.0549 -0.0627 5.157>
    lookAtTarget {boneReference {boneName=suspender_trgt}}
    lookAt_Axis {axis=<1 0>}
   }
   n96 = "Transform (Modify) Bone - Bone: dyn_suspender_fr_MSTR"!(ComponentPose=@n95, bAlphaBoolEnabled:bool=Get FrontEndIdle() -> (FrontEndIdle:bool), Translation="0,0,0", Rotation="-20,0,0") -> (Pose) {
    alphaInputType=Bool
    rotationMode=BMM_Additive
    rotationSpace=BCS_BoneSpace
    translationMode=BMM_Additive
    translationSpace=BCS_BoneSpace
    alphaBoolBlend {blendInTime=0.1, blendOption=Linear, blendOutTime=0.2}
    boneToModify {boneName=dyn_suspender_fr_MSTR}
    rotation {pitch=-20}
   }
   Save cached pose 'idle'(Pose=Component To Local(ComponentPose=@n96)) {cacheName=idle}
   AND Boolean(A=Get bIsInFrontEnd(), B=NOT Boolean(A=Get bCanPlayCustomAnimations() -> (bCanPlayCustomAnimations:bool)))
   Get bIsMoving2D()
  ]}
 EventGraph {nodes [
   Event BlueprintUpdateAnimation()
   > n1 = Sequence!(execute) -> (then_0, then_1, then_2:Exec, then_3:Exec, then_4:Exec, then_5:Exec)
   n2 = BreakTransform(InTransform=GetDeltaTransformFromRefPose(self=GetOwnerSkeletalMeshComponent(), BoneName=pelvis, BaseName=thigh_l))
   > Set pelvis left yaw(pelvis left yaw:double=MapRangeClamped(Value=BreakRotator(InRot=@n2.Rotation), InRangeA=-10, InRangeB=-30)) -> (Output_Get:double)
   n3 = GetOwnerSkeletalMeshComponent()
   n4 = BreakRotator(InRot=BreakTransform(InTransform=GetDeltaTransformFromRefPose(self=@n3, BoneName=thigh_r, BaseName=pelvis)).Rotation)
   n5 = Get bIsCrouching()
   > Set thigh_rot_r_pitch(thigh_rot_r_pitch:double=MapRangeClamped(Value=@n4.Pitch, InRangeA=0, InRangeB=SelectFloat(A=80, B=65, bPickA=@n5))) -> (Output_Get:double)
   n6 = Get bIsCrouching()
   > Set thigh_rot_r_yaw(thigh_rot_r_yaw:double=MapRangeClamped(Value=@n4.Yaw, InRangeA=SelectFloat(A=-20, B=0, bPickA=@n6), InRangeB=-100)) -> (Output_Get:double)
   n7 = GetOwnerSkeletalMeshComponent()
   n8 = BreakRotator(InRot=BreakTransform(InTransform=GetDeltaTransformFromRefPose(self=@n7, BoneName=thigh_l, BaseName=pelvis)).Rotation)
   Set thigh_rot_l_yaw(execute=@n1.then_1, thigh_rot_l_yaw:double=MapRangeClamped(Value=@n8.Yaw, InRangeA=-15, InRangeB=-110)) -> (Output_Get:double)
   > Set suspender fix(suspender fix:double=MapRangeClamped(Value=@n8.Yaw, InRangeA=-10, InRangeB=-30)) -> (Output_Get:double)
   n9 = NOT Boolean(A=Get bIsDBNO())
   n10 = Get bIsCrouching()
   Set crouching(execute=@n1.then_2, crouching:double=SelectFloat(A=1, B=0, bPickA=AND Boolean(A=@n10, B=@n9))) -> (Output_Get:double)
   > Set is DBNO(is DBNO:double=SelectFloat(A=1, B=0, bPickA=Get bIsDBNO())) -> (Output_Get:double)
   n11 = Get bIsMoving2D()
   > Set moving(moving:double=SelectFloat(A=1, B=0, bPickA=@n11)) -> (Output_Get:double)
   > Set not moving(not moving:double=SelectFloat(A=1, B=0, bPickA=NOT Boolean(A=@n11))) -> (Output_Get:double)
   > Set sprinting(sprinting:double=SelectFloat(A=1, B=0, bPickA=Get bIsSprinting())) -> (Output_Get:double)
   > Set is parachute opening(is parachute opening:double=SelectFloat(A=1, B=0, bPickA=Get bIsParachuteOpened())) -> (Output_Get:double)
   > Set skydive floating(skydive floating:double=SelectFloat(A=1, B=0, bPickA=Get bIsSkydiveFloating())) -> (Output_Get:double)
   n12 = NOT Boolean(A=Get bIsSkyDiving())
   n13 = Get bIsJumping()
   > Set is jumping(is jumping:double=SelectFloat(A=1, B=0, bPickA=AND Boolean(A=@n13, B=@n12))) -> (Output_Get:double)
   > n14 = Gravity Override State Function DEPRECATED(
    CharacterPartAnimInstance=Self-Reference()
    GameTimeInSeconds=GetWorldDeltaSeconds!() -> (ReturnValue:double)
    GravityOverrideParams=Get Gravity Override Params()
   )
   > "Set Gravity Override "("Gravity Override ":Vector=@n14) -> (Output_Get:Vector)
   n15 = Get bIsPlayingEmote()
   > Set FrontEndIdle(FrontEndIdle:bool=AND Boolean(A=Get bIsInFrontEnd(), B=NOT Boolean(A=@n15))) -> (Output_Get:bool)
   > Set Not in FrontEnd(Not in FrontEnd:bool=NOT Boolean(A=Get bIsInFrontEnd())) -> (Output_Get:bool)
   n16 = Branch(execute=@n1.then_3, Condition=Get bIsCrouching())
   > Gate!(Enter:Exec, Open:Exec=@n16.else, Close:Exec=@n17.then, Toggle:Exec, bStartClosed:bool=false) -> (Exit:Exec) {is=title K2Node_MacroInstance}
   > Set strap correction(strap correction:double=1) -> (Output_Get:double)
   > Delay(Duration=0.1)
   > n17 = Set strap correction(strap correction=0)
   n18 = Get bIsMoving2D()
   n19 = Branch(execute=@n1.then_4, Condition=NOT Boolean(A=@n18))
   > Set belt fix when moving(belt fix when moving:double=1) -> (Output_Get:double)
   n20 = Branch(execute=@n19.else, Condition=@n18)
   > Set belt fix when moving(belt fix when moving=0.7)
   Set belt fix when moving(execute=@n20.else, belt fix when moving=0)
   n21 = Branch(execute=@n1.then_5, Condition=Get bIsSkyDiving())
   > Set skydiving(skydiving:double=1) -> (Output_Get:double)
   > Set not skydiving(not skydiving:double=0) -> (Output_Get:double)
   Set skydiving(execute=@n21.else, skydiving=0)
   > Set not skydiving(not skydiving=0.8)
   Get Gravity Override Params_0() -> (Gravity Override Params_0:Gravity Override Params Struct Structure)
  ]}
]
variables [
 belt fix when moving {type=number}
 crouching {type=number}
 FrontEndIdle {type=boolean}
 "Gravity Override " {type=Vector}
 Gravity Override Params {
  type=GravityOverrideParamsStruct
  default {skydive_FloatingDown {gravityOverride=<0 -200 400>}, skydive_Parachuting {gravityOverride=<0 0 -980>}}
 }
 Gravity Override Params_0 {type=GravityOverrideParamsStruct}
 is DBNO {type=number}
 is jumping {type=number}
 is parachute opening {type=number}
 moving {type=number}
 Not in FrontEnd {type=boolean}
 not moving {type=number}
 not skydiving {type=number}
 pelvis left yaw {type=number}
 skydive floating {type=number}
 skydiving {type=number}
 sprinting {type=number}
 strap correction {type=number}
 suspender fix {type=number}
 thigh_rot_l_yaw {type=number}
 thigh_rot_r_pitch {type=number}
 thigh_rot_r_yaw {type=number}
]
