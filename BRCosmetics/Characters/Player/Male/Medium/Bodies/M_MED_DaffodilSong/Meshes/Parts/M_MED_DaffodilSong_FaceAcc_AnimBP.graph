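# Post-style physics AnimBlueprint for the DaffodilSong male face accessory skeleton, adding secondary motion to its dyn_hair bones and two bang bones.
# Serial chains of single-bone AnimDynamics nodes use angular springs, per-bone limits and gravity overrides.
# The chains repeat with different tuning: one simulates mostly head-relative, another in root or world space with stronger gravity.
# Later chains branch from the cached InputPose.
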AnimBlueprint /BRCosmetics/Characters/Player/Male/Medium/Bodies/M_MED_DaffodilSong/Meshes/Parts/M_MED_DaffodilSong_FaceAcc_AnimBP
target_skeleton=&/M_MED_DaffodilSong_FaceAcc_Skeleton
graphs [
 AnimGraph {nodes [
   n1 = Anim Dynamics - Bone: dyn_hair_1(ComponentPose=Local To Component(), Alpha=1, ExternalForce="0,0,0") {
    angularDampingOverride=1
    angularSpringConstant=3000
    bAngularSpring=true
    bGravityOverrideInSimSpace=true
    bOverrideAngularDamping=true
    bUseGravityOverride=true
    gravityScale=0.35
    simulationSpace=BoneRelative
    alphaBoolBlend {blendOption=Linear}
    boundBone {boneName=dyn_hair_1}
    componentAppliedLinearAccClamp=<5000 5000 5000>
    componentLinearAccScale=<0.6 0.5 0.5>
    componentLinearVelScale=<0.6 0.5 0.5>
    gravityOverride {z=-900}
    physicsBodyDefinitions [{
      boundBone {boneName=dyn_hair_1}
      boxExtents=<22 22 22>
      constraintSetup {angularLimitsMax=<30 30 30>, angularLimitsMin=<-15 -15 0>, angularTarget=<1 0 0>}
      localJointOffset=<1 0 0>
     }]
    relativeSpaceBone {boneName=head}
   }
   n2 = Anim Dynamics - Bone: dyn_hair_2(ComponentPose=@n1, Alpha=1, ExternalForce="0,0,0") {
    angularDampingOverride=1
    angularSpringConstant=3000
    bAngularSpring=true
    bGravityOverrideInSimSpace=true
    bOverrideAngularDamping=true
    bUseGravityOverride=true
    gravityScale=0.35
    simulationSpace=BoneRelative
    alphaBoolBlend {blendOption=Linear}
    boundBone {boneName=dyn_hair_2}
    componentAppliedLinearAccClamp=<5000 5000 5000>
    componentLinearAccScale=<0.6 0.5 0.5>
    componentLinearVelScale=<0.6 0.5 0.5>
    gravityOverride {z=-900}
    physicsBodyDefinitions [{
      boundBone {boneName=dyn_hair_2}
      boxExtents=<22 22 22>
      constraintSetup {angularLimitsMax=<30 30 30>, angularLimitsMin=<-15 -15 0>, angularTarget=<1 0 0>}
      localJointOffset=<1 0 0>
     }]
    relativeSpaceBone {boneName=head}
   }
   n3 = Anim Dynamics - Bone: dyn_hair_3(ComponentPose=@n2, Alpha=1, ExternalForce="0,0,0") {
    angularDampingOverride=1
    angularSpringConstant=3000
    bAngularSpring=true
    bGravityOverrideInSimSpace=true
    bOverrideAngularDamping=true
    bUseGravityOverride=true
    gravityScale=0.35
    simulationSpace=BoneRelative
    alphaBoolBlend {blendOption=Linear}
    boundBone {boneName=dyn_hair_3}
    componentAppliedLinearAccClamp=<5000 5000 5000>
    componentLinearAccScale=<0.6 0.5 0.5>
    componentLinearVelScale=<0.6 0.5 0.5>
    gravityOverride {z=-900}
    physicsBodyDefinitions [{
      boundBone {boneName=dyn_hair_3}
      boxExtents=<22 22 22>
      constraintSetup {angularLimitsMax=<30 30 30>, angularLimitsMin=<-15 -15 0>, angularTarget=<1 0 0>}
      localJointOffset=<1 0 0>
     }]
    relativeSpaceBone {boneName=head}
   }
   n4 = Anim Dynamics - Bone: dyn_hair_5(ComponentPose=@n3, Alpha=1, ExternalForce="0,0,0") {
    angularDampingOverride=1
    angularSpringConstant=3000
    bAngularSpring=true
    bGravityOverrideInSimSpace=true
    bOverrideAngularDamping=true
    bUseGravityOverride=true
    gravityScale=0.35
    simulationSpace=BoneRelative
    alphaBoolBlend {blendOption=Linear}
    boundBone {boneName=dyn_hair_5}
    componentAppliedLinearAccClamp=<5000 5000 5000>
    componentLinearAccScale=<0.6 0.5 0.5>
    componentLinearVelScale=<0.6 0.5 0.5>
    gravityOverride {z=-900}
    physicsBodyDefinitions [{
      boundBone {boneName=dyn_hair_5}
      boxExtents=<22 22 22>
      constraintSetup {angularLimitsMax=<30 30 5>, angularLimitsMin=<-15 20 0>, angularTarget=<1 0 0>}
      localJointOffset=<1 0 0>
     }]
    relativeSpaceBone {boneName=head}
   }
   n5 = Anim Dynamics - Bone: dyn_hair_6(ComponentPose=@n4, Alpha=1, ExternalForce="0,0,0") {
    angularDampingOverride=1
    angularSpringConstant=3000
    bAngularSpring=true
    bGravityOverrideInSimSpace=true
    bOverrideAngularDamping=true
    bUseGravityOverride=true
    gravityScale=0.35
    simulationSpace=BoneRelative
    alphaBoolBlend {blendOption=Linear}
    boundBone {boneName=dyn_hair_6}
    componentAppliedLinearAccClamp=<5000 5000 5000>
    componentLinearAccScale=<0.6 0.5 0.5>
    componentLinearVelScale=<0.6 0.5 0.5>
    gravityOverride {z=-900}
    physicsBodyDefinitions [{
      boundBone {boneName=dyn_hair_6}
      boxExtents=<22 22 22>
      constraintSetup {angularLimitsMax=<30 30 30>, angularLimitsMin=<-15 -15 0>, angularTarget=<1 0 0>}
      localJointOffset=<1 0 0>
     }]
    relativeSpaceBone {boneName=head}
   }
   n6 = Anim Dynamics - Bone: dyn_hair_7(ComponentPose=@n5, Alpha=1, ExternalForce="0,0,0") {
    angularDampingOverride=1
    angularSpringConstant=3000
    bAngularSpring=true
    bGravityOverrideInSimSpace=true
    bOverrideAngularDamping=true
    bUseGravityOverride=true
    gravityScale=0.35
    simulationSpace=BoneRelative
    alphaBoolBlend {blendOption=Linear}
    boundBone {boneName=dyn_hair_7}
    componentAppliedLinearAccClamp=<5000 5000 5000>
    componentLinearAccScale=<0.6 0.5 0.5>
    componentLinearVelScale=<0.6 0.5 0.5>
    gravityOverride {z=-900}
    physicsBodyDefinitions [{
      boundBone {boneName=dyn_hair_7}
      boxExtents=<22 22 22>
      constraintSetup {angularLimitsMax=<30 30 30>, angularLimitsMin=<-15 -15 -50>, angularTarget=<1 0 0>}
      localJointOffset=<1 0 0>
     }]
    relativeSpaceBone {boneName=head}
   }
   n7 = Anim Dynamics - Bone: dyn_hair_8(ComponentPose=@n6, Alpha=1, ExternalForce="0,0,0") {
    angularDampingOverride=1
    angularSpringConstant=3000
    bAngularSpring=true
    bGravityOverrideInSimSpace=true
    bOverrideAngularDamping=true
    bUseGravityOverride=true
    gravityScale=0.35
    simulationSpace=BoneRelative
    alphaBoolBlend {blendOption=Linear}
    boundBone {boneName=dyn_hair_8}
    componentAppliedLinearAccClamp=<5000 5000 5000>
    componentLinearAccScale=<0.6 0.5 0.5>
    componentLinearVelScale=<0.6 0.5 0.5>
    gravityOverride {z=-900}
    physicsBodyDefinitions [{
      boundBone {boneName=dyn_hair_8}
      boxExtents=<22 22 22>
      constraintSetup {angularLimitsMax=<30 30 30>, angularLimitsMin=<-15 -15 0>, angularTarget=<1 0 0>}
      localJointOffset=<1 0 0>
     }]
    relativeSpaceBone {boneName=head}
   }
   n8 = Anim Dynamics - Bone: dyn_hair_9(ComponentPose=@n7, Alpha=1, ExternalForce="0,0,0") {
    angularDampingOverride=1
    angularSpringConstant=3000
    bAngularSpring=true
    bGravityOverrideInSimSpace=true
    bOverrideAngularDamping=true
    bUseGravityOverride=true
    gravityScale=0.35
    simulationSpace=BoneRelative
    alphaBoolBlend {blendOption=Linear}
    boundBone {boneName=dyn_hair_9}
    componentAppliedLinearAccClamp=<5000 5000 5000>
    componentLinearAccScale=<0.6 0.5 0.5>
    componentLinearVelScale=<0.6 0.5 0.5>
    gravityOverride {z=-900}
    physicsBodyDefinitions [{
      boundBone {boneName=dyn_hair_9}
      boxExtents=<22 22 22>
      constraintSetup {angularLimitsMax=<5 5 0>, angularLimitsMin=<-5 -5 0>, angularTarget=<1 0 0>}
      localJointOffset=<2 0 0>
     }]
    relativeSpaceBone {boneName=head}
   }
   n9 = Anim Dynamics - Bone: dyn_hair_10(ComponentPose=@n8, Alpha=1, ExternalForce="0,0,0") {
    angularDampingOverride=1
    angularSpringConstant=3000
    bAngularSpring=true
    bGravityOverrideInSimSpace=true
    bOverrideAngularDamping=true
    bUseGravityOverride=true
    gravityScale=0.35
    simulationSpace=BoneRelative
    alphaBoolBlend {blendOption=Linear}
    boundBone {boneName=dyn_hair_10}
    componentAppliedLinearAccClamp=<5000 5000 5000>
    componentLinearAccScale=<0.6 0.5 0.5>
    componentLinearVelScale=<0.6 0.5 0.5>
    gravityOverride {z=-900}
    physicsBodyDefinitions [{
      boundBone {boneName=dyn_hair_10}
      boxExtents=<22 22 22>
      constraintSetup {angularLimitsMax=<30 30 30>, angularLimitsMin=<-15 -15 0>, angularTarget=<1 0 0>}
      localJointOffset=<1 0 0>
     }]
    relativeSpaceBone {boneName=head}
   }
   n10 = Anim Dynamics - Bone: dyn_hair_13(ComponentPose=@n9, Alpha=1, ExternalForce="0,0,0") {
    angularDampingOverride=1
    angularSpringConstant=3000
    bAngularSpring=true
    bGravityOverrideInSimSpace=true
    bOverrideAngularDamping=true
    bUseGravityOverride=true
    gravityScale=0.35
    simulationSpace=BoneRelative
    alphaBoolBlend {blendOption=Linear}
    boundBone {boneName=dyn_hair_13}
    componentAppliedLinearAccClamp=<5000 5000 5000>
    componentLinearAccScale=<0.6 0.5 0.5>
    componentLinearVelScale=<0.6 0.5 0.5>
    gravityOverride {z=-900}
    physicsBodyDefinitions [{
      boundBone {boneName=dyn_hair_13}
      boxExtents=<22 22 22>
      constraintSetup {angularLimitsMax=<30 30 30>, angularLimitsMin=<-15 -15 -50>, angularTarget=<1 0 0>}
      localJointOffset=<1 0 0>
     }]
    relativeSpaceBone {boneName=head}
   }
   n11 = Anim Dynamics - Bone: dyn_hair_20(ComponentPose=@n10, Alpha=1, ExternalForce="0,0,0") {
    angularDampingOverride=1
    angularSpringConstant=3000
    bAngularSpring=true
    bGravityOverrideInSimSpace=true
    bOverrideAngularDamping=true
    bUseGravityOverride=true
    gravityScale=0.35
    simulationSpace=BoneRelative
    alphaBoolBlend {blendOption=Linear}
    boundBone {boneName=dyn_hair_20}
    componentAppliedLinearAccClamp=<5000 5000 5000>
    componentLinearAccScale=<0.6 0.5 0.5>
    componentLinearVelScale=<0.6 0.5 0.5>
    gravityOverride {z=-900}
    physicsBodyDefinitions [{
      boundBone {boneName=dyn_hair_20}
      boxExtents=<22 22 22>
      constraintSetup {angularLimitsMax=<30 30 30>, angularLimitsMin=<-15 -15 0>, angularTarget=<1 0 0>}
      localJointOffset=<1 0 0>
     }]
    relativeSpaceBone {boneName=head}
   }
   n12 = Anim Dynamics - Bone: dyn_hair_17(ComponentPose=@n11, Alpha=1, ExternalForce="0,0,0") {
    angularDampingOverride=1
    angularSpringConstant=3000
    bAngularSpring=true
    bGravityOverrideInSimSpace=true
    bOverrideAngularDamping=true
    bUseGravityOverride=true
    gravityScale=0.35
    simulationSpace=BoneRelative
    alphaBoolBlend {blendOption=Linear}
    boundBone {boneName=dyn_hair_17}
    componentAppliedLinearAccClamp=<5000 5000 5000>
    componentLinearAccScale=<0.6 0.5 0.5>
    componentLinearVelScale=<0.6 0.5 0.5>
    gravityOverride {z=-900}
    physicsBodyDefinitions [{
      boundBone {boneName=dyn_hair_17}
      boxExtents=<22 22 22>
      constraintSetup {angularLimitsMax=<25 25 5>, angularLimitsMin=<-10 -10 0>, angularTarget=<1 0 0>}
      localJointOffset=<1 0 0>
     }]
    relativeSpaceBone {boneName=head}
   }
   n13 = Anim Dynamics - Bone: dyn_hair_19(ComponentPose=@n12, Alpha=1, ExternalForce="0,0,0") {
    angularDampingOverride=1
    angularSpringConstant=3000
    bAngularSpring=true
    bGravityOverrideInSimSpace=true
    bOverrideAngularDamping=true
    bUseGravityOverride=true
    gravityScale=0.35
    simulationSpace=BoneRelative
    alphaBoolBlend {blendOption=Linear}
    boundBone {boneName=dyn_hair_19}
    componentAppliedLinearAccClamp=<5000 5000 5000>
    componentLinearAccScale=<0.6 0.5 0.5>
    componentLinearVelScale=<0.6 0.5 0.5>
    gravityOverride {z=-900}
    physicsBodyDefinitions [{
      boundBone {boneName=dyn_hair_19}
      boxExtents=<22 22 22>
      constraintSetup {angularLimitsMax=<30 30 30>, angularLimitsMin=<-15 -15 -10>, angularTarget=<1 0 0>}
      localJointOffset=<4 0 0>
     }]
    relativeSpaceBone {boneName=head}
   }
   n14 = Anim Dynamics - Bone: dyn_hair_14(ComponentPose=@n13, Alpha=1, ExternalForce="0,0,0") {
    angularDampingOverride=1
    angularSpringConstant=3000
    bAngularSpring=true
    bGravityOverrideInSimSpace=true
    bOverrideAngularDamping=true
    bUseGravityOverride=true
    gravityScale=0.35
    simulationSpace=BoneRelative
    alphaBoolBlend {blendOption=Linear}
    boundBone {boneName=dyn_hair_14}
    componentAppliedLinearAccClamp=<5000 5000 5000>
    componentLinearAccScale=<0.6 0.5 0.5>
    componentLinearVelScale=<0.6 0.5 0.5>
    gravityOverride {z=-900}
    physicsBodyDefinitions [{
      boundBone {boneName=dyn_hair_14}
      boxExtents=<22 22 22>
      constraintSetup {angularLimitsMax=<30 30 30>, angularLimitsMin=<-15 -15 -10>, angularTarget=<1 0 0>}
      localJointOffset=<4 0 0>
     }]
    relativeSpaceBone {boneName=head}
   }
   n15 = Anim Dynamics - Bone: dyn_hair_15(ComponentPose=@n14, Alpha=0.8, ExternalForce="0,0,0") {
    angularDampingOverride=1
    angularSpringConstant=3000
    bAngularSpring=true
    bGravityOverrideInSimSpace=true
    bOverrideAngularDamping=true
    bUseGravityOverride=true
    gravityScale=0.35
    simulationSpace=BoneRelative
    alphaBoolBlend {blendOption=Linear}
    boundBone {boneName=dyn_hair_15}
    componentAppliedLinearAccClamp=<5000 5000 5000>
    componentLinearAccScale=<0.6 0.5 0.5>
    componentLinearVelScale=<0.6 0.5 0.5>
    gravityOverride {z=-900}
    physicsBodyDefinitions [{
      boundBone {boneName=dyn_hair_15}
      boxExtents=<22 22 22>
      constraintSetup {angularLimitsMax=<30 30 30>, angularLimitsMin=<-15 -15 0>, angularTarget=<1 0 0>}
      localJointOffset=<1 0 0>
     }]
    relativeSpaceBone {boneName=head}
   }
   n16 = Anim Dynamics - Bone: dyn_hair_12(ComponentPose=@n15, Alpha=1, ExternalForce="0,0,0") {
    angularDampingOverride=1
    angularSpringConstant=3000
    bAngularSpring=true
    bGravityOverrideInSimSpace=true
    bOverrideAngularDamping=true
    bUseGravityOverride=true
    gravityScale=0.35
    simulationSpace=BoneRelative
    alphaBoolBlend {blendOption=Linear}
    boundBone {boneName=dyn_hair_12}
    componentAppliedLinearAccClamp=<5000 5000 5000>
    componentLinearAccScale=<0.6 0.5 0.5>
    componentLinearVelScale=<0.6 0.5 0.5>
    gravityOverride {z=-900}
    physicsBodyDefinitions [{
      boundBone {boneName=dyn_hair_12}
      boxExtents=<22 22 22>
      constraintSetup {angularLimitsMax=<30 30 30>, angularLimitsMin=<-15 -15 -50>, angularTarget=<1 0 0>}
      localJointOffset=<4 0 0>
     }]
    relativeSpaceBone {boneName=head}
   }
   n17 = Anim Dynamics - Bone: dyn_hair_16(ComponentPose=@n16, Alpha=1, ExternalForce="0,0,0") {
    angularDampingOverride=1
    angularSpringConstant=3000
    bAngularSpring=true
    bGravityOverrideInSimSpace=true
    bOverrideAngularDamping=true
    bUseGravityOverride=true
    gravityScale=0.35
    simulationSpace=BoneRelative
    alphaBoolBlend {blendOption=Linear}
    boundBone {boneName=dyn_hair_16}
    componentAppliedLinearAccClamp=<5000 5000 5000>
    componentLinearAccScale=<0.6 0.5 0.5>
    componentLinearVelScale=<0.6 0.5 0.5>
    gravityOverride {z=-900}
    physicsBodyDefinitions [{
      boundBone {boneName=dyn_hair_16}
      boxExtents=<22 22 22>
      constraintSetup {angularLimitsMax=<30 30 30>, angularLimitsMin=<-15 -15 -50>, angularTarget=<1 0 0>}
      localJointOffset=<3 0 0>
     }]
    relativeSpaceBone {boneName=head}
   }
   n18 = Anim Dynamics - Bone: dyn_hair_18(ComponentPose=@n17, Alpha=1, ExternalForce="0,0,0") {
    angularDampingOverride=1
    angularSpringConstant=3000
    bAngularSpring=true
    bGravityOverrideInSimSpace=true
    bOverrideAngularDamping=true
    bUseGravityOverride=true
    gravityScale=0.35
    simulationSpace=BoneRelative
    alphaBoolBlend {blendOption=Linear}
    boundBone {boneName=dyn_hair_18}
    componentAppliedLinearAccClamp=<5000 5000 5000>
    componentLinearAccScale=<0.6 0.5 0.5>
    componentLinearVelScale=<0.6 0.5 0.5>
    gravityOverride {z=-900}
    physicsBodyDefinitions [{
      boundBone {boneName=dyn_hair_18}
      boxExtents=<22 22 22>
      constraintSetup {angularLimitsMax=<30 30 30>, angularLimitsMin=<-15 -15 -50>, angularTarget=<1 0 0>}
      localJointOffset=<4 0 0>
     }]
    relativeSpaceBone {boneName=head}
   }
   n19 = Anim Dynamics - Bone: dyn_hair_11(ComponentPose=@n18, Alpha=1, ExternalForce="0,0,0") {
    angularDampingOverride=1
    angularSpringConstant=3000
    bAngularSpring=true
    bGravityOverrideInSimSpace=true
    bOverrideAngularDamping=true
    bUseGravityOverride=true
    gravityScale=0.35
    simulationSpace=BoneRelative
    alphaBoolBlend {blendOption=Linear}
    boundBone {boneName=dyn_hair_11}
    componentAppliedLinearAccClamp=<5000 5000 5000>
    componentLinearAccScale=<0.6 0.5 0.5>
    componentLinearVelScale=<0.6 0.5 0.5>
    gravityOverride {z=-900}
    physicsBodyDefinitions [{
      boundBone {boneName=dyn_hair_11}
      boxExtents=<22 22 22>
      constraintSetup {angularLimitsMax=<30 30 30>, angularLimitsMin=<-15 -15 -30>, angularTarget=<1 0 0>}
      localJointOffset=<4 0 0>
     }]
    relativeSpaceBone {boneName=head}
   }
   n20 = Anim Dynamics - Bone: dyn_hair_21(ComponentPose=@n19, Alpha=1, ExternalForce="0,0,0") {
    angularDampingOverride=1
    angularSpringConstant=500
    bAngularSpring=true
    bGravityOverrideInSimSpace=true
    bOverrideAngularDamping=true
    bUseGravityOverride=true
    gravityScale=0.35
    simulationSpace=BoneRelative
    alphaBoolBlend {blendOption=Linear}
    boundBone {boneName=dyn_hair_21}
    componentAppliedLinearAccClamp=<5000 5000 5000>
    componentLinearAccScale=<0.6 0.5 0.5>
    componentLinearVelScale=<0.6 0.5 0.5>
    gravityOverride {z=-900}
    physicsBodyDefinitions [{
      boundBone {boneName=dyn_hair_21}
      boxExtents=<22 22 22>
      constraintSetup {angularLimitsMax=<50 30 50>, angularLimitsMin=<-15 -15 0>, angularTarget=<1 0 0>}
      localJointOffset=<2 0 0>
     }]
    relativeSpaceBone {boneName=head}
   }
   n21 = Anim Dynamics - Bone: dyn_hair_4(ComponentPose=@n20, Alpha=1, ExternalForce="0,0,0") {
    angularDampingOverride=1
    angularSpringConstant=3000
    bAngularSpring=true
    bGravityOverrideInSimSpace=true
    bOverrideAngularDamping=true
    bUseGravityOverride=true
    gravityScale=0.35
    simulationSpace=BoneRelative
    alphaBoolBlend {blendOption=Linear}
    boundBone {boneName=dyn_hair_4}
    componentAppliedLinearAccClamp=<5000 5000 5000>
    componentLinearAccScale=<0.6 0.5 0.5>
    componentLinearVelScale=<0.6 0.5 0.5>
    gravityOverride {z=-900}
    physicsBodyDefinitions [{
      boundBone {boneName=dyn_hair_4}
      boxExtents=<22 22 22>
      constraintSetup {angularLimitsMax=<50 30 50>, angularLimitsMin=<-15 -15 0>, angularTarget=<1 0 0>}
      localJointOffset=<2 0 0>
     }]
    relativeSpaceBone {boneName=head}
   }
   n22 = Anim Dynamics - Bone: dyn_hair_22_bang(ComponentPose=@n21, Alpha=1, ExternalForce="0,0,0") {
    angularDampingOverride=1
    angularSpringConstant=3000
    bAngularSpring=true
    bOverrideAngularDamping=true
    bUseGravityOverride=true
    gravityScale=0.35
    simulationSpace=RootRelative
    alphaBoolBlend {blendOption=Linear}
    boundBone {boneName=dyn_hair_22_bang}
    componentAppliedLinearAccClamp=<5000 5000 5000>
    componentLinearAccScale=<0.3 0.3 0.3>
    componentLinearVelScale=<0.3 0.3 0.3>
    gravityOverride {z=-500}
    physicsBodyDefinitions [{
      boundBone {boneName=dyn_hair_22_bang}
      boxExtents=<22 22 22>
      constraintSetup {angularLimitsMax=<50 0 50>, angularLimitsMin=<-15 -50 0>, angularTarget=<0.5 0 0>}
      localJointOffset=<5 0 0>
     }]
   }
   n23 = Anim Dynamics - Bone: dyn_hair_22_bang1(ComponentPose=@n22, Alpha=1, ExternalForce="0,0,0") {
    angularDampingOverride=1
    angularSpringConstant=3000
    bAngularSpring=true
    bOverrideAngularDamping=true
    bUseGravityOverride=true
    gravityScale=0.35
    simulationSpace=RootRelative
    alphaBoolBlend {blendOption=Linear}
    boundBone {boneName=dyn_hair_22_bang1}
    componentAppliedLinearAccClamp=<5000 5000 5000>
    componentLinearAccScale=<0.3 0.3 0.3>
    componentLinearVelScale=<0.3 0.3 0.3>
    gravityOverride {z=-500}
    physicsBodyDefinitions [{
      boundBone {boneName=dyn_hair_22_bang1}
      boxExtents=<22 22 22>
      constraintSetup {angularLimitsMax=<50 0 50>, angularLimitsMin=<-15 -30 0>, angularTarget=<0.5 0 0>}
      localJointOffset=<2 0 0>
     }]
   }
   Component To Local(ComponentPose=@n23)
   n24 = Anim Dynamics - Bone: dyn_hair_1(ComponentPose=Local To Component(LocalPose=Use cached pose 'InputPose'()), Alpha=1, ExternalForce="0,0,0") {
    angularDampingOverride=0.99
    angularSpringConstant=3000
    bAngularSpring=true
    bOverrideAngularDamping=true
    bUseGravityOverride=true
    gravityScale=0.35
    simulationSpace=RootRelative
    alphaBoolBlend {blendOption=Linear}
    boundBone {boneName=dyn_hair_1}
    componentAppliedLinearAccClamp=<5000 5000 5000>
    componentLinearAccScale=<0.5 0.5 0.5>
    componentLinearVelScale=<0.6 0.5 0.5>
    gravityOverride {z=-1500}
    physicsBodyDefinitions [{
      boundBone {boneName=dyn_hair_1}
      boxExtents=<45 45 45>
      constraintSetup {angularLimitsMax=<30 30 30>, angularLimitsMin=<-15 -15 0>, angularTarget=<1 0 0>}
      localJointOffset=<1 0 0>
     }]
    relativeSpaceBone {boneName=head}
   }
   n25 = Anim Dynamics - Bone: dyn_hair_2(ComponentPose=@n24, Alpha=1, ExternalForce="0,0,0") {
    angularDampingOverride=0.99
    angularSpringConstant=3000
    bAngularSpring=true
    bOverrideAngularDamping=true
    bUseGravityOverride=true
    gravityScale=0.35
    simulationSpace=RootRelative
    alphaBoolBlend {blendOption=Linear}
    boundBone {boneName=dyn_hair_2}
    componentAppliedLinearAccClamp=<5000 5000 5000>
    componentLinearAccScale=<0.5 0.5 0.5>
    componentLinearVelScale=<0.6 0.5 0.5>
    gravityOverride {z=-1500}
    physicsBodyDefinitions [{
      boundBone {boneName=dyn_hair_2}
      boxExtents=<45 45 45>
      constraintSetup {angularLimitsMax=<30 30 30>, angularLimitsMin=<-15 -15 0>, angularTarget=<1 0 0>}
      localJointOffset=<1 0 0>
     }]
    relativeSpaceBone {boneName=head}
   }
   n26 = Anim Dynamics - Bone: dyn_hair_3(ComponentPose=@n25, Alpha=1, ExternalForce="0,0,0") {
    angularDampingOverride=0.99
    angularSpringConstant=3000
    bAngularSpring=true
    bOverrideAngularDamping=true
    bUseGravityOverride=true
    gravityScale=0.35
    simulationSpace=RootRelative
    alphaBoolBlend {blendOption=Linear}
    boundBone {boneName=dyn_hair_3}
    componentAppliedLinearAccClamp=<5000 5000 5000>
    componentLinearAccScale=<0.5 0.5 0.5>
    componentLinearVelScale=<0.6 0.5 0.5>
    gravityOverride {z=-1500}
    physicsBodyDefinitions [{
      boundBone {boneName=dyn_hair_3}
      boxExtents=<45 45 45>
      constraintSetup {angularLimitsMax=<30 30 30>, angularLimitsMin=<-15 -15 0>, angularTarget=<1 0 0>}
      localJointOffset=<1 0 0>
     }]
    relativeSpaceBone {boneName=head}
   }
   n27 = Anim Dynamics - Bone: dyn_hair_5(ComponentPose=@n26, Alpha=1, ExternalForce="0,0,0") {
    angularDampingOverride=0.99
    angularSpringConstant=3000
    bAngularSpring=true
    bOverrideAngularDamping=true
    bUseGravityOverride=true
    gravityScale=0.35
    simulationSpace=RootRelative
    alphaBoolBlend {blendOption=Linear}
    boundBone {boneName=dyn_hair_5}
    componentAppliedLinearAccClamp=<5000 5000 5000>
    componentLinearAccScale=<0.5 0.5 0.5>
    componentLinearVelScale=<0.6 0.5 0.5>
    gravityOverride {z=-1500}
    physicsBodyDefinitions [{
      boundBone {boneName=dyn_hair_5}
      boxExtents=<45 45 45>
      constraintSetup {angularLimitsMax=<30 30 5>, angularLimitsMin=<-15 20 0>, angularTarget=<1 0 0>}
      localJointOffset=<1 0 0>
     }]
    relativeSpaceBone {boneName=head}
   }
   n28 = Anim Dynamics - Bone: dyn_hair_6(ComponentPose=@n27, Alpha=1, ExternalForce="0,0,0") {
    angularDampingOverride=0.99
    angularSpringConstant=3000
    bAngularSpring=true
    bOverrideAngularDamping=true
    bUseGravityOverride=true
    gravityScale=0.35
    simulationSpace=World
    alphaBoolBlend {blendOption=Linear}
    boundBone {boneName=dyn_hair_6}
    componentAppliedLinearAccClamp=<5000 5000 5000>
    componentLinearAccScale=<0.5 0.5 0.5>
    componentLinearVelScale=<0.6 0.5 0.5>
    gravityOverride {z=-1500}
    physicsBodyDefinitions [{
      boundBone {boneName=dyn_hair_6}
      boxExtents=<45 45 45>
      constraintSetup {angularLimitsMax=<30 30 30>, angularLimitsMin=<-15 -15 0>, angularTarget=<1 0 0>}
      localJointOffset=<1 0 0>
     }]
    relativeSpaceBone {boneName=head}
   }
   n29 = Anim Dynamics - Bone: dyn_hair_7(ComponentPose=@n28, Alpha=1, ExternalForce="0,0,0") {
    angularDampingOverride=0.99
    angularSpringConstant=3000
    bAngularSpring=true
    bOverrideAngularDamping=true
    bUseGravityOverride=true
    gravityScale=0.35
    simulationSpace=World
    alphaBoolBlend {blendOption=Linear}
    boundBone {boneName=dyn_hair_7}
    componentAppliedLinearAccClamp=<5000 5000 5000>
    componentLinearAccScale=<0.5 0.5 0.5>
    componentLinearVelScale=<0.6 0.5 0.5>
    gravityOverride {z=-1500}
    physicsBodyDefinitions [{
      boundBone {boneName=dyn_hair_7}
      boxExtents=<45 45 45>
      constraintSetup {angularLimitsMax=<30 30 30>, angularLimitsMin=<-15 -15 -50>, angularTarget=<1 0 0>}
      localJointOffset=<1 0 0>
     }]
    relativeSpaceBone {boneName=head}
   }
   n30 = Anim Dynamics - Bone: dyn_hair_8(ComponentPose=@n29, Alpha=1, ExternalForce="0,0,0") {
    angularDampingOverride=0.99
    angularSpringConstant=3000
    bAngularSpring=true
    bOverrideAngularDamping=true
    bUseGravityOverride=true
    gravityScale=0.35
    simulationSpace=World
    alphaBoolBlend {blendOption=Linear}
    boundBone {boneName=dyn_hair_8}
    componentAppliedLinearAccClamp=<5000 5000 5000>
    componentLinearAccScale=<0.5 0.5 0.5>
    componentLinearVelScale=<0.6 0.5 0.5>
    gravityOverride {z=-1500}
    physicsBodyDefinitions [{
      boundBone {boneName=dyn_hair_8}
      boxExtents=<45 45 45>
      constraintSetup {angularLimitsMax=<30 30 30>, angularLimitsMin=<-15 -15 0>, angularTarget=<1 0 0>}
      localJointOffset=<1 0 0>
     }]
    relativeSpaceBone {boneName=head}
   }
   n31 = Anim Dynamics - Bone: dyn_hair_9(ComponentPose=@n30, Alpha=1, ExternalForce="0,0,0") {
    angularDampingOverride=0.99
    angularSpringConstant=3000
    bAngularSpring=true
    bOverrideAngularDamping=true
    bUseGravityOverride=true
    gravityScale=0.35
    simulationSpace=World
    alphaBoolBlend {blendOption=Linear}
    boundBone {boneName=dyn_hair_9}
    componentAppliedLinearAccClamp=<5000 5000 5000>
    componentLinearAccScale=<0.5 0.5 0.5>
    componentLinearVelScale=<0.6 0.5 0.5>
    gravityOverride {z=-1500}
    physicsBodyDefinitions [{
      boundBone {boneName=dyn_hair_9}
      boxExtents=<45 45 45>
      constraintSetup {angularLimitsMax=<5 5 0>, angularLimitsMin=<-5 -5 0>, angularTarget=<1 0 0>}
      localJointOffset=<2 0 0>
     }]
    relativeSpaceBone {boneName=head}
   }
   n32 = Anim Dynamics - Bone: dyn_hair_10(ComponentPose=@n31, Alpha=1, ExternalForce="0,0,0") {
    angularDampingOverride=0.99
    angularSpringConstant=3000
    bAngularSpring=true
    bOverrideAngularDamping=true
    bUseGravityOverride=true
    gravityScale=0.35
    simulationSpace=World
    alphaBoolBlend {blendOption=Linear}
    boundBone {boneName=dyn_hair_10}
    componentAppliedLinearAccClamp=<5000 5000 5000>
    componentLinearAccScale=<0.5 0.5 0.5>
    componentLinearVelScale=<0.6 0.5 0.5>
    gravityOverride {z=-1500}
    physicsBodyDefinitions [{
      boundBone {boneName=dyn_hair_10}
      boxExtents=<45 45 45>
      constraintSetup {angularLimitsMax=<30 30 30>, angularLimitsMin=<-15 -15 0>, angularTarget=<1 0 0>}
      localJointOffset=<1 0 0>
     }]
    relativeSpaceBone {boneName=head}
   }
   n33 = Anim Dynamics - Bone: dyn_hair_13(ComponentPose=@n32, Alpha=1, ExternalForce="0,0,0") {
    angularDampingOverride=0.99
    angularSpringConstant=3000
    bAngularSpring=true
    bOverrideAngularDamping=true
    bUseGravityOverride=true
    gravityScale=0.35
    simulationSpace=RootRelative
    alphaBoolBlend {blendOption=Linear}
    boundBone {boneName=dyn_hair_13}
    componentAppliedLinearAccClamp=<5000 5000 5000>
    componentLinearAccScale=<0.5 0.5 0.5>
    componentLinearVelScale=<0.6 0.5 0.5>
    gravityOverride {z=-1500}
    physicsBodyDefinitions [{
      boundBone {boneName=dyn_hair_13}
      boxExtents=<45 45 45>
      constraintSetup {angularLimitsMax=<30 30 30>, angularLimitsMin=<-15 -15 -50>, angularTarget=<1 0 0>}
      localJointOffset=<1 0 0>
     }]
    relativeSpaceBone {boneName=head}
   }
   n34 = Anim Dynamics - Bone: dyn_hair_20(ComponentPose=@n33, Alpha=1, ExternalForce="0,0,0") {
    angularDampingOverride=0.99
    angularSpringConstant=3000
    bAngularSpring=true
    bOverrideAngularDamping=true
    bUseGravityOverride=true
    gravityScale=0.35
    simulationSpace=RootRelative
    alphaBoolBlend {blendOption=Linear}
    boundBone {boneName=dyn_hair_20}
    componentAppliedLinearAccClamp=<5000 5000 5000>
    componentLinearAccScale=<0.5 0.5 0.5>
    componentLinearVelScale=<0.6 0.5 0.5>
    gravityOverride {z=-1500}
    physicsBodyDefinitions [{
      boundBone {boneName=dyn_hair_20}
      boxExtents=<45 45 45>
      constraintSetup {angularLimitsMax=<30 30 30>, angularLimitsMin=<-15 -15 0>, angularTarget=<1 0 0>}
      localJointOffset=<1 0 0>
     }]
    relativeSpaceBone {boneName=head}
   }
   n35 = Anim Dynamics - Bone: dyn_hair_17(ComponentPose=@n34, Alpha=1, ExternalForce="0,0,0") {
    angularDampingOverride=0.99
    angularSpringConstant=3000
    bAngularSpring=true
    bOverrideAngularDamping=true
    bUseGravityOverride=true
    gravityScale=0.35
    simulationSpace=RootRelative
    alphaBoolBlend {blendOption=Linear}
    boundBone {boneName=dyn_hair_17}
    componentAppliedLinearAccClamp=<5000 5000 5000>
    componentLinearAccScale=<0.5 0.5 0.5>
    componentLinearVelScale=<0.6 0.5 0.5>
    gravityOverride {z=-1500}
    physicsBodyDefinitions [{
      boundBone {boneName=dyn_hair_17}
      boxExtents=<45 45 45>
      constraintSetup {angularLimitsMax=<25 25 5>, angularLimitsMin=<-10 -10 0>, angularTarget=<1 0 0>}
      localJointOffset=<1 0 0>
     }]
    relativeSpaceBone {boneName=head}
   }
   n36 = Anim Dynamics - Bone: dyn_hair_19(ComponentPose=@n35, Alpha=1, ExternalForce="0,0,0") {
    angularDampingOverride=0.99
    angularSpringConstant=3000
    bAngularSpring=true
    bOverrideAngularDamping=true
    bUseGravityOverride=true
    gravityScale=0.35
    simulationSpace=World
    alphaBoolBlend {blendOption=Linear}
    boundBone {boneName=dyn_hair_19}
    componentAppliedLinearAccClamp=<5000 5000 5000>
    componentLinearAccScale=<0.5 0.5 0.5>
    componentLinearVelScale=<0.6 0.5 0.5>
    gravityOverride {z=-1500}
    physicsBodyDefinitions [{
      boundBone {boneName=dyn_hair_19}
      boxExtents=<45 45 45>
      constraintSetup {angularLimitsMax=<30 30 30>, angularLimitsMin=<-15 -15 -10>, angularTarget=<1 0 0>}
      localJointOffset=<4 0 0>
     }]
    relativeSpaceBone {boneName=head}
   }
   n37 = Anim Dynamics - Bone: dyn_hair_14(ComponentPose=@n36, Alpha=1, ExternalForce="0,0,0") {
    angularDampingOverride=0.99
    angularSpringConstant=3000
    bAngularSpring=true
    bOverrideAngularDamping=true
    bUseGravityOverride=true
    gravityScale=0.35
    simulationSpace=World
    alphaBoolBlend {blendOption=Linear}
    boundBone {boneName=dyn_hair_14}
    componentAppliedLinearAccClamp=<5000 5000 5000>
    componentLinearAccScale=<0.5 0.5 0.5>
    componentLinearVelScale=<0.6 0.5 0.5>
    gravityOverride {z=-1500}
    physicsBodyDefinitions [{
      boundBone {boneName=dyn_hair_14}
      boxExtents=<45 45 45>
      constraintSetup {angularLimitsMax=<30 30 30>, angularLimitsMin=<-15 -15 -10>, angularTarget=<1 0 0>}
      localJointOffset=<4 0 0>
     }]
    relativeSpaceBone {boneName=head}
   }
   n38 = Anim Dynamics - Bone: dyn_hair_15(ComponentPose=@n37, Alpha=0.8, ExternalForce="0,0,0") {
    angularDampingOverride=0.99
    angularSpringConstant=3000
    bAngularSpring=true
    bOverrideAngularDamping=true
    bUseGravityOverride=true
    gravityScale=0.35
    simulationSpace=World
    alphaBoolBlend {blendOption=Linear}
    boundBone {boneName=dyn_hair_15}
    componentAppliedLinearAccClamp=<5000 5000 5000>
    componentLinearAccScale=<0.5 0.5 0.5>
    componentLinearVelScale=<0.6 0.5 0.5>
    gravityOverride {z=-1500}
    physicsBodyDefinitions [{
      boundBone {boneName=dyn_hair_15}
      boxExtents=<45 45 45>
      constraintSetup {angularLimitsMax=<30 30 30>, angularLimitsMin=<-15 -15 0>, angularTarget=<1 0 0>}
      localJointOffset=<1 0 0>
     }]
    relativeSpaceBone {boneName=head}
   }
   n39 = Anim Dynamics - Bone: dyn_hair_12(ComponentPose=@n38, Alpha=1, ExternalForce="0,0,0") {
    angularDampingOverride=0.99
    angularSpringConstant=3000
    bAngularSpring=true
    bOverrideAngularDamping=true
    bUseGravityOverride=true
    gravityScale=0.35
    simulationSpace=World
    alphaBoolBlend {blendOption=Linear}
    boundBone {boneName=dyn_hair_12}
    componentAppliedLinearAccClamp=<5000 5000 5000>
    componentLinearAccScale=<0.5 0.5 0.5>
    componentLinearVelScale=<0.6 0.5 0.5>
    gravityOverride {z=-1500}
    physicsBodyDefinitions [{
      boundBone {boneName=dyn_hair_12}
      boxExtents=<45 45 45>
      constraintSetup {angularLimitsMax=<30 30 30>, angularLimitsMin=<-15 -15 -50>, angularTarget=<1 0 0>}
      localJointOffset=<4 0 0>
     }]
    relativeSpaceBone {boneName=head}
   }
   n40 = Anim Dynamics - Bone: dyn_hair_16(ComponentPose=@n39, Alpha=1, ExternalForce="0,0,0") {
    angularDampingOverride=0.99
    angularSpringConstant=3000
    bAngularSpring=true
    bOverrideAngularDamping=true
    bUseGravityOverride=true
    gravityScale=0.35
    simulationSpace=World
    alphaBoolBlend {blendOption=Linear}
    boundBone {boneName=dyn_hair_16}
    componentAppliedLinearAccClamp=<5000 5000 5000>
    componentLinearAccScale=<0.5 0.5 0.5>
    componentLinearVelScale=<0.6 0.5 0.5>
    gravityOverride {z=-1500}
    physicsBodyDefinitions [{
      boundBone {boneName=dyn_hair_16}
      boxExtents=<45 45 45>
      constraintSetup {angularLimitsMax=<30 30 30>, angularLimitsMin=<-15 -15 -50>, angularTarget=<1 0 0>}
      localJointOffset=<3 0 0>
     }]
    relativeSpaceBone {boneName=head}
   }
   n41 = Anim Dynamics - Bone: dyn_hair_18(ComponentPose=@n40, Alpha=1, ExternalForce="0,0,0") {
    angularDampingOverride=0.99
    angularSpringConstant=3000
    bAngularSpring=true
    bOverrideAngularDamping=true
    bUseGravityOverride=true
    gravityScale=0.35
    simulationSpace=World
    alphaBoolBlend {blendOption=Linear}
    boundBone {boneName=dyn_hair_18}
    componentAppliedLinearAccClamp=<5000 5000 5000>
    componentLinearAccScale=<0.5 0.5 0.5>
    componentLinearVelScale=<0.6 0.5 0.5>
    gravityOverride {z=-1500}
    physicsBodyDefinitions [{
      boundBone {boneName=dyn_hair_18}
      boxExtents=<45 45 45>
      constraintSetup {angularLimitsMax=<30 30 30>, angularLimitsMin=<-15 -15 -50>, angularTarget=<1 0 0>}
      localJointOffset=<4 0 0>
     }]
    relativeSpaceBone {boneName=head}
   }
   n42 = Anim Dynamics - Bone: dyn_hair_11(ComponentPose=@n41, Alpha=1, ExternalForce="0,0,0") {
    angularDampingOverride=0.99
    angularSpringConstant=3000
    bAngularSpring=true
    bOverrideAngularDamping=true
    bUseGravityOverride=true
    gravityScale=0.35
    simulationSpace=World
    alphaBoolBlend {blendOption=Linear}
    boundBone {boneName=dyn_hair_11}
    componentAppliedLinearAccClamp=<5000 5000 5000>
    componentLinearAccScale=<0.5 0.5 0.5>
    componentLinearVelScale=<0.6 0.5 0.5>
    gravityOverride {z=-1500}
    physicsBodyDefinitions [{
      boundBone {boneName=dyn_hair_11}
      boxExtents=<45 45 45>
      constraintSetup {angularLimitsMax=<30 30 30>, angularLimitsMin=<-15 -15 -30>, angularTarget=<1 0 0>}
      localJointOffset=<4 0 0>
     }]
    relativeSpaceBone {boneName=head}
   }
   n43 = Anim Dynamics - Bone: dyn_hair_21(ComponentPose=@n42, Alpha=1, ExternalForce="0,0,0") {
    angularDampingOverride=0.99
    angularSpringConstant=500
    bAngularSpring=true
    bOverrideAngularDamping=true
    bUseGravityOverride=true
    gravityScale=0.35
    simulationSpace=World
    alphaBoolBlend {blendOption=Linear}
    boundBone {boneName=dyn_hair_21}
    componentAppliedLinearAccClamp=<5000 5000 5000>
    componentLinearAccScale=<0.5 0.5 0.5>
    componentLinearVelScale=<0.6 0.5 0.5>
    gravityOverride {z=-1500}
    physicsBodyDefinitions [{
      boundBone {boneName=dyn_hair_21}
      boxExtents=<45 45 45>
      constraintSetup {angularLimitsMax=<50 30 50>, angularLimitsMin=<-15 -15 0>, angularTarget=<1 0 0>}
      localJointOffset=<2 0 0>
     }]
    relativeSpaceBone {boneName=head}
   }
   n44 = Anim Dynamics - Bone: dyn_hair_4(ComponentPose=@n43, Alpha=1, ExternalForce="0,0,0") {
    angularDampingOverride=0.99
    angularSpringConstant=3000
    bAngularSpring=true
    bOverrideAngularDamping=true
    bUseGravityOverride=true
    gravityScale=0.35
    simulationSpace=World
    alphaBoolBlend {blendOption=Linear}
    boundBone {boneName=dyn_hair_4}
    componentAppliedLinearAccClamp=<5000 5000 5000>
    componentLinearAccScale=<0.5 0.5 0.5>
    componentLinearVelScale=<0.6 0.5 0.5>
    gravityOverride {z=-1500}
    physicsBodyDefinitions [{
      boundBone {boneName=dyn_hair_4}
      boxExtents=<45 45 45>
      constraintSetup {angularLimitsMax=<50 30 50>, angularLimitsMin=<-15 -15 0>, angularTarget=<1 0 0>}
      localJointOffset=<2 0 0>
     }]
    relativeSpaceBone {boneName=head}
   }
   n45 = Anim Dynamics - Bone: dyn_hair_22_bang(ComponentPose=@n44, Alpha=1, ExternalForce="0,0,0") {
    angularDampingOverride=1
    angularSpringConstant=3000
    bAngularSpring=true
    bOverrideAngularDamping=true
    bUseGravityOverride=true
    gravityScale=0.35
    simulationSpace=World
    alphaBoolBlend {blendOption=Linear}
    boundBone {boneName=dyn_hair_22_bang}
    componentAppliedLinearAccClamp=<5000 5000 5000>
    componentLinearAccScale=<0.3 0.3 0.3>
    componentLinearVelScale=<0.3 0.3 0.3>
    gravityOverride {z=-500}
    physicsBodyDefinitions [{
      boundBone {boneName=dyn_hair_22_bang}
      boxExtents=<22 22 22>
      constraintSetup {angularLimitsMax=<50 0 50>, angularLimitsMin=<-15 -50 0>, angularTarget=<0.5 0 0>}
      localJointOffset=<5 0 0>
     }]
   }
   n46 = Anim Dynamics - Bone: dyn_hair_22_bang1(ComponentPose=@n45, Alpha=1, ExternalForce="0,0,0") {
    angularDampingOverride=1
    angularSpringConstant=3000
    bAngularSpring=true
    bOverrideAngularDamping=true
    bUseGravityOverride=true
    gravityScale=0.35
    simulationSpace=World
    alphaBoolBlend {blendOption=Linear}
    boundBone {boneName=dyn_hair_22_bang1}
    componentAppliedLinearAccClamp=<5000 5000 5000>
    componentLinearAccScale=<0.3 0.3 0.3>
    componentLinearVelScale=<0.3 0.3 0.3>
    gravityOverride {z=-500}
    physicsBodyDefinitions [{
      boundBone {boneName=dyn_hair_22_bang1}
      boxExtents=<22 22 22>
      constraintSetup {angularLimitsMax=<50 0 50>, angularLimitsMin=<-15 -30 0>, angularTarget=<0.5 0 0>}
      localJointOffset=<2 0 0>
     }]
   }
   n47 = Anim Dynamics - Bone: dyn_hair_1(ComponentPose=Local To Component(LocalPose=Use cached pose 'InputPose'()), Alpha=1, ExternalForce="0,0,0") {
    angularDampingOverride=1
    angularSpringConstant=3000
    bAngularSpring=true
    bGravityOverrideInSimSpace=true
    bOverrideAngularDamping=true
    bUseGravityOverride=true
    gravityScale=0.35
    simulationSpace=BoneRelative
    alphaBoolBlend {blendOption=Linear}
    boundBone {boneName=dyn_hair_1}
    componentAppliedLinearAccClamp=<5000 5000 5000>
    componentLinearAccScale=<0.6 0.5 0.5>
    componentLinearVelScale=<0.6 0.5 0.5>
    gravityOverride {z=-900}
    physicsBodyDefinitions [{
      boundBone {boneName=dyn_hair_1}
      boxExtents=<22 22 22>
      constraintSetup {angularLimitsMax=<30 30 30>, angularLimitsMin=<-15 -15 0>, angularTarget=<1 0 0>}
      localJointOffset=<1 0 0>
     }]
    relativeSpaceBone {boneName=head}
   }
   n48 = Anim Dynamics - Bone: dyn_hair_2(ComponentPose=@n47, Alpha=1, ExternalForce="0,0,0") {
    angularDampingOverride=1
    angularSpringConstant=3000
    bAngularSpring=true
    bGravityOverrideInSimSpace=true
    bOverrideAngularDamping=true
    bUseGravityOverride=true
    gravityScale=0.35
    simulationSpace=BoneRelative
    alphaBoolBlend {blendOption=Linear}
    boundBone {boneName=dyn_hair_2}
    componentAppliedLinearAccClamp=<5000 5000 5000>
    componentLinearAccScale=<0.6 0.5 0.5>
    componentLinearVelScale=<0.6 0.5 0.5>
    gravityOverride {z=-900}
    physicsBodyDefinitions [{
      boundBone {boneName=dyn_hair_2}
      boxExtents=<22 22 22>
      constraintSetup {angularLimitsMax=<30 30 30>, angularLimitsMin=<-15 -15 0>, angularTarget=<1 0 0>}
      localJointOffset=<1 0 0>
     }]
    relativeSpaceBone {boneName=head}
   }
   n49 = Anim Dynamics - Bone: dyn_hair_3(ComponentPose=@n48, Alpha=1, ExternalForce="0,0,0") {
    angularDampingOverride=1
    angularSpringConstant=3000
    bAngularSpring=true
    bGravityOverrideInSimSpace=true
    bOverrideAngularDamping=true
    bUseGravityOverride=true
    gravityScale=0.35
    simulationSpace=BoneRelative
    alphaBoolBlend {blendOption=Linear}
    boundBone {boneName=dyn_hair_3}
    componentAppliedLinearAccClamp=<5000 5000 5000>
    componentLinearAccScale=<0.6 0.5 0.5>
    componentLinearVelScale=<0.6 0.5 0.5>
    gravityOverride {z=-900}
    physicsBodyDefinitions [{
      boundBone {boneName=dyn_hair_3}
      boxExtents=<22 22 22>
      constraintSetup {angularLimitsMax=<30 30 30>, angularLimitsMin=<-15 -15 0>, angularTarget=<1 0 0>}
      localJointOffset=<1 0 0>
     }]
    relativeSpaceBone {boneName=head}
   }
   n50 = Anim Dynamics - Bone: dyn_hair_5(ComponentPose=@n49, Alpha=1, ExternalForce="0,0,0") {
    angularDampingOverride=1
    angularSpringConstant=3000
    bAngularSpring=true
    bGravityOverrideInSimSpace=true
    bOverrideAngularDamping=true
    bUseGravityOverride=true
    gravityScale=0.35
    simulationSpace=BoneRelative
    alphaBoolBlend {blendOption=Linear}
    boundBone {boneName=dyn_hair_5}
    componentAppliedLinearAccClamp=<5000 5000 5000>
    componentLinearAccScale=<0.6 0.5 0.5>
    componentLinearVelScale=<0.6 0.5 0.5>
    gravityOverride {z=-900}
    physicsBodyDefinitions [{
      boundBone {boneName=dyn_hair_5}
      boxExtents=<22 22 22>
      constraintSetup {angularLimitsMax=<30 30 5>, angularLimitsMin=<-15 20 0>, angularTarget=<1 0 0>}
      localJointOffset=<1 0 0>
     }]
    relativeSpaceBone {boneName=head}
   }
   n51 = Anim Dynamics - Bone: dyn_hair_6(ComponentPose=@n50, Alpha=1, ExternalForce="0,0,0") {
    angularDampingOverride=1
    angularSpringConstant=3000
    bAngularSpring=true
    bGravityOverrideInSimSpace=true
    bOverrideAngularDamping=true
    bUseGravityOverride=true
    gravityScale=0.35
    simulationSpace=BoneRelative
    alphaBoolBlend {blendOption=Linear}
    boundBone {boneName=dyn_hair_6}
    componentAppliedLinearAccClamp=<5000 5000 5000>
    componentLinearAccScale=<0.6 0.5 0.5>
    componentLinearVelScale=<0.6 0.5 0.5>
    gravityOverride {z=-900}
    physicsBodyDefinitions [{
      boundBone {boneName=dyn_hair_6}
      boxExtents=<22 22 22>
      constraintSetup {angularLimitsMax=<30 30 30>, angularLimitsMin=<-15 -15 0>, angularTarget=<1 0 0>}
      localJointOffset=<1 0 0>
     }]
    relativeSpaceBone {boneName=head}
   }
   n52 = Anim Dynamics - Bone: dyn_hair_7(ComponentPose=@n51, Alpha=1, ExternalForce="0,0,0") {
    angularDampingOverride=1
    angularSpringConstant=3000
    bAngularSpring=true
    bGravityOverrideInSimSpace=true
    bOverrideAngularDamping=true
    bUseGravityOverride=true
    gravityScale=0.35
    simulationSpace=BoneRelative
    alphaBoolBlend {blendOption=Linear}
    boundBone {boneName=dyn_hair_7}
    componentAppliedLinearAccClamp=<5000 5000 5000>
    componentLinearAccScale=<0.6 0.5 0.5>
    componentLinearVelScale=<0.6 0.5 0.5>
    gravityOverride {z=-900}
    physicsBodyDefinitions [{
      boundBone {boneName=dyn_hair_7}
      boxExtents=<22 22 22>
      constraintSetup {angularLimitsMax=<30 30 30>, angularLimitsMin=<-15 -15 -50>, angularTarget=<1 0 0>}
      localJointOffset=<1 0 0>
     }]
    relativeSpaceBone {boneName=head}
   }
   n53 = Anim Dynamics - Bone: dyn_hair_8(ComponentPose=@n52, Alpha=1, ExternalForce="0,0,0") {
    angularDampingOverride=1
    angularSpringConstant=3000
    bAngularSpring=true
    bGravityOverrideInSimSpace=true
    bOverrideAngularDamping=true
    bUseGravityOverride=true
    gravityScale=0.35
    simulationSpace=BoneRelative
    alphaBoolBlend {blendOption=Linear}
    boundBone {boneName=dyn_hair_8}
    componentAppliedLinearAccClamp=<5000 5000 5000>
    componentLinearAccScale=<0.6 0.5 0.5>
    componentLinearVelScale=<0.6 0.5 0.5>
    gravityOverride {z=-900}
    physicsBodyDefinitions [{
      boundBone {boneName=dyn_hair_8}
      boxExtents=<22 22 22>
      constraintSetup {angularLimitsMax=<30 30 30>, angularLimitsMin=<-15 -15 0>, angularTarget=<1 0 0>}
      localJointOffset=<1 0 0>
     }]
    relativeSpaceBone {boneName=head}
   }
   n54 = Anim Dynamics - Bone: dyn_hair_9(ComponentPose=@n53, Alpha=1, ExternalForce="0,0,0") {
    angularDampingOverride=1
    angularSpringConstant=3000
    bAngularSpring=true
    bGravityOverrideInSimSpace=true
    bOverrideAngularDamping=true
    bUseGravityOverride=true
    gravityScale=0.35
    simulationSpace=BoneRelative
    alphaBoolBlend {blendOption=Linear}
    boundBone {boneName=dyn_hair_9}
    componentAppliedLinearAccClamp=<5000 5000 5000>
    componentLinearAccScale=<0.6 0.5 0.5>
    componentLinearVelScale=<0.6 0.5 0.5>
    gravityOverride {z=-900}
    physicsBodyDefinitions [{
      boundBone {boneName=dyn_hair_9}
      boxExtents=<22 22 22>
      constraintSetup {angularLimitsMax=<5 5 0>, angularLimitsMin=<-5 -5 0>, angularTarget=<1 0 0>}
      localJointOffset=<2 0 0>
     }]
    relativeSpaceBone {boneName=head}
   }
   n55 = Anim Dynamics - Bone: dyn_hair_10(ComponentPose=@n54, Alpha=1, ExternalForce="0,0,0") {
    angularDampingOverride=1
    angularSpringConstant=3000
    bAngularSpring=true
    bGravityOverrideInSimSpace=true
    bOverrideAngularDamping=true
    bUseGravityOverride=true
    gravityScale=0.35
    simulationSpace=BoneRelative
    alphaBoolBlend {blendOption=Linear}
    boundBone {boneName=dyn_hair_10}
    componentAppliedLinearAccClamp=<5000 5000 5000>
    componentLinearAccScale=<0.6 0.5 0.5>
    componentLinearVelScale=<0.6 0.5 0.5>
    gravityOverride {z=-900}
    physicsBodyDefinitions [{
      boundBone {boneName=dyn_hair_10}
      boxExtents=<22 22 22>
      constraintSetup {angularLimitsMax=<30 30 30>, angularLimitsMin=<-15 -15 0>, angularTarget=<1 0 0>}
      localJointOffset=<1 0 0>
     }]
    relativeSpaceBone {boneName=head}
   }
   n56 = Anim Dynamics - Bone: dyn_hair_13(ComponentPose=@n55, Alpha=1, ExternalForce="0,0,0") {
    angularDampingOverride=1
    angularSpringConstant=3000
    bAngularSpring=true
    bGravityOverrideInSimSpace=true
    bOverrideAngularDamping=true
    bUseGravityOverride=true
    gravityScale=0.35
    simulationSpace=BoneRelative
    alphaBoolBlend {blendOption=Linear}
    boundBone {boneName=dyn_hair_13}
    componentAppliedLinearAccClamp=<5000 5000 5000>
    componentLinearAccScale=<0.6 0.5 0.5>
    componentLinearVelScale=<0.6 0.5 0.5>
    gravityOverride {z=-900}
    physicsBodyDefinitions [{
      boundBone {boneName=dyn_hair_13}
      boxExtents=<22 22 22>
      constraintSetup {angularLimitsMax=<30 30 30>, angularLimitsMin=<-15 -15 -50>, angularTarget=<1 0 0>}
      localJointOffset=<1 0 0>
     }]
    relativeSpaceBone {boneName=head}
   }
   n57 = Anim Dynamics - Bone: dyn_hair_20(ComponentPose=@n56, Alpha=1, ExternalForce="0,0,0") {
    angularDampingOverride=1
    angularSpringConstant=3000
    bAngularSpring=true
    bGravityOverrideInSimSpace=true
    bOverrideAngularDamping=true
    bUseGravityOverride=true
    gravityScale=0.35
    simulationSpace=BoneRelative
    alphaBoolBlend {blendOption=Linear}
    boundBone {boneName=dyn_hair_20}
    componentAppliedLinearAccClamp=<5000 5000 5000>
    componentLinearAccScale=<0.6 0.5 0.5>
    componentLinearVelScale=<0.6 0.5 0.5>
    gravityOverride {z=-900}
    physicsBodyDefinitions [{
      boundBone {boneName=dyn_hair_20}
      boxExtents=<22 22 22>
      constraintSetup {angularLimitsMax=<30 30 30>, angularLimitsMin=<-15 -15 0>, angularTarget=<1 0 0>}
      localJointOffset=<1 0 0>
     }]
    relativeSpaceBone {boneName=head}
   }
   n58 = Anim Dynamics - Bone: dyn_hair_17(ComponentPose=@n57, Alpha=1, ExternalForce="0,0,0") {
    angularDampingOverride=1
    angularSpringConstant=3000
    bAngularSpring=true
    bGravityOverrideInSimSpace=true
    bOverrideAngularDamping=true
    bUseGravityOverride=true
    gravityScale=0.35
    simulationSpace=BoneRelative
    alphaBoolBlend {blendOption=Linear}
    boundBone {boneName=dyn_hair_17}
    componentAppliedLinearAccClamp=<5000 5000 5000>
    componentLinearAccScale=<0.6 0.5 0.5>
    componentLinearVelScale=<0.6 0.5 0.5>
    gravityOverride {z=-900}
    physicsBodyDefinitions [{
      boundBone {boneName=dyn_hair_17}
      boxExtents=<22 22 22>
      constraintSetup {angularLimitsMax=<25 25 5>, angularLimitsMin=<-10 -10 0>, angularTarget=<1 0 0>}
      localJointOffset=<1 0 0>
     }]
    relativeSpaceBone {boneName=head}
   }
   n59 = Anim Dynamics - Bone: dyn_hair_19(ComponentPose=@n58, Alpha=1, ExternalForce="0,0,0") {
    angularDampingOverride=1
    angularSpringConstant=3000
    bAngularSpring=true
    bGravityOverrideInSimSpace=true
    bOverrideAngularDamping=true
    bUseGravityOverride=true
    gravityScale=0.35
    simulationSpace=BoneRelative
    alphaBoolBlend {blendOption=Linear}
    boundBone {boneName=dyn_hair_19}
    componentAppliedLinearAccClamp=<5000 5000 5000>
    componentLinearAccScale=<0.6 0.5 0.5>
    componentLinearVelScale=<0.6 0.5 0.5>
    gravityOverride {z=-900}
    physicsBodyDefinitions [{
      boundBone {boneName=dyn_hair_19}
      boxExtents=<22 22 22>
      constraintSetup {angularLimitsMax=<30 30 30>, angularLimitsMin=<-15 -15 -10>, angularTarget=<1 0 0>}
      localJointOffset=<4 0 0>
     }]
    relativeSpaceBone {boneName=head}
   }
   n60 = Anim Dynamics - Bone: dyn_hair_14(ComponentPose=@n59, Alpha=1, ExternalForce="0,0,0") {
    angularDampingOverride=1
    angularSpringConstant=3000
    bAngularSpring=true
    bGravityOverrideInSimSpace=true
    bOverrideAngularDamping=true
    bUseGravityOverride=true
    gravityScale=0.35
    simulationSpace=BoneRelative
    alphaBoolBlend {blendOption=Linear}
    boundBone {boneName=dyn_hair_14}
    componentAppliedLinearAccClamp=<5000 5000 5000>
    componentLinearAccScale=<0.6 0.5 0.5>
    componentLinearVelScale=<0.6 0.5 0.5>
    gravityOverride {z=-900}
    physicsBodyDefinitions [{
      boundBone {boneName=dyn_hair_14}
      boxExtents=<22 22 22>
      constraintSetup {angularLimitsMax=<30 30 30>, angularLimitsMin=<-15 -15 -10>, angularTarget=<1 0 0>}
      localJointOffset=<4 0 0>
     }]
    relativeSpaceBone {boneName=head}
   }
   n61 = Anim Dynamics - Bone: dyn_hair_15(ComponentPose=@n60, Alpha=0.8, ExternalForce="0,0,0") {
    angularDampingOverride=1
    angularSpringConstant=3000
    bAngularSpring=true
    bGravityOverrideInSimSpace=true
    bOverrideAngularDamping=true
    bUseGravityOverride=true
    gravityScale=0.35
    simulationSpace=BoneRelative
    alphaBoolBlend {blendOption=Linear}
    boundBone {boneName=dyn_hair_15}
    componentAppliedLinearAccClamp=<5000 5000 5000>
    componentLinearAccScale=<0.6 0.5 0.5>
    componentLinearVelScale=<0.6 0.5 0.5>
    gravityOverride {z=-900}
    physicsBodyDefinitions [{
      boundBone {boneName=dyn_hair_15}
      boxExtents=<22 22 22>
      constraintSetup {angularLimitsMax=<30 30 30>, angularLimitsMin=<-15 -15 0>, angularTarget=<1 0 0>}
      localJointOffset=<1 0 0>
     }]
    relativeSpaceBone {boneName=head}
   }
   n62 = Anim Dynamics - Bone: dyn_hair_12(ComponentPose=@n61, Alpha=1, ExternalForce="0,0,0") {
    angularDampingOverride=1
    angularSpringConstant=3000
    bAngularSpring=true
    bGravityOverrideInSimSpace=true
    bOverrideAngularDamping=true
    bUseGravityOverride=true
    gravityScale=0.35
    simulationSpace=BoneRelative
    alphaBoolBlend {blendOption=Linear}
    boundBone {boneName=dyn_hair_12}
    componentAppliedLinearAccClamp=<5000 5000 5000>
    componentLinearAccScale=<0.6 0.5 0.5>
    componentLinearVelScale=<0.6 0.5 0.5>
    gravityOverride {z=-900}
    physicsBodyDefinitions [{
      boundBone {boneName=dyn_hair_12}
      boxExtents=<22 22 22>
      constraintSetup {angularLimitsMax=<30 30 30>, angularLimitsMin=<-15 -15 -50>, angularTarget=<1 0 0>}
      localJointOffset=<4 0 0>
     }]
    relativeSpaceBone {boneName=head}
   }
   n63 = Anim Dynamics - Bone: dyn_hair_16(ComponentPose=@n62, Alpha=1, ExternalForce="0,0,0") {
    angularDampingOverride=1
    angularSpringConstant=3000
    bAngularSpring=true
    bGravityOverrideInSimSpace=true
    bOverrideAngularDamping=true
    bUseGravityOverride=true
    gravityScale=0.35
    simulationSpace=BoneRelative
    alphaBoolBlend {blendOption=Linear}
    boundBone {boneName=dyn_hair_16}
    componentAppliedLinearAccClamp=<5000 5000 5000>
    componentLinearAccScale=<0.6 0.5 0.5>
    componentLinearVelScale=<0.6 0.5 0.5>
    gravityOverride {z=-900}
    physicsBodyDefinitions [{
      boundBone {boneName=dyn_hair_16}
      boxExtents=<22 22 22>
      constraintSetup {angularLimitsMax=<30 30 30>, angularLimitsMin=<-15 -15 -50>, angularTarget=<1 0 0>}
      localJointOffset=<3 0 0>
     }]
    relativeSpaceBone {boneName=head}
   }
   n64 = Anim Dynamics - Bone: dyn_hair_18(ComponentPose=@n63, Alpha=1, ExternalForce="0,0,0") {
    angularDampingOverride=1
    angularSpringConstant=3000
    bAngularSpring=true
    bGravityOverrideInSimSpace=true
    bOverrideAngularDamping=true
    bUseGravityOverride=true
    gravityScale=0.35
    simulationSpace=BoneRelative
    alphaBoolBlend {blendOption=Linear}
    boundBone {boneName=dyn_hair_18}
    componentAppliedLinearAccClamp=<5000 5000 5000>
    componentLinearAccScale=<0.6 0.5 0.5>
    componentLinearVelScale=<0.6 0.5 0.5>
    gravityOverride {z=-900}
    physicsBodyDefinitions [{
      boundBone {boneName=dyn_hair_18}
      boxExtents=<22 22 22>
      constraintSetup {angularLimitsMax=<30 30 30>, angularLimitsMin=<-15 -15 -50>, angularTarget=<1 0 0>}
      localJointOffset=<4 0 0>
     }]
    relativeSpaceBone {boneName=head}
   }
   n65 = Anim Dynamics - Bone: dyn_hair_11(ComponentPose=@n64, Alpha=1, ExternalForce="0,0,0") {
    angularDampingOverride=1
    angularSpringConstant=3000
    bAngularSpring=true
    bGravityOverrideInSimSpace=true
    bOverrideAngularDamping=true
    bUseGravityOverride=true
    gravityScale=0.35
    simulationSpace=BoneRelative
    alphaBoolBlend {blendOption=Linear}
    boundBone {boneName=dyn_hair_11}
    componentAppliedLinearAccClamp=<5000 5000 5000>
    componentLinearAccScale=<0.6 0.5 0.5>
    componentLinearVelScale=<0.6 0.5 0.5>
    gravityOverride {z=-900}
    physicsBodyDefinitions [{
      boundBone {boneName=dyn_hair_11}
      boxExtents=<22 22 22>
      constraintSetup {angularLimitsMax=<30 30 30>, angularLimitsMin=<-15 -15 -30>, angularTarget=<1 0 0>}
      localJointOffset=<4 0 0>
     }]
    relativeSpaceBone {boneName=head}
   }
   n66 = Anim Dynamics - Bone: dyn_hair_21(ComponentPose=@n65, Alpha=1, ExternalForce="0,0,0") {
    angularDampingOverride=1
    angularSpringConstant=500
    bAngularSpring=true
    bGravityOverrideInSimSpace=true
    bOverrideAngularDamping=true
    bUseGravityOverride=true
    gravityScale=0.35
    simulationSpace=BoneRelative
    alphaBoolBlend {blendOption=Linear}
    boundBone {boneName=dyn_hair_21}
    componentAppliedLinearAccClamp=<5000 5000 5000>
    componentLinearAccScale=<0.6 0.5 0.5>
    componentLinearVelScale=<0.6 0.5 0.5>
    gravityOverride {z=-900}
    physicsBodyDefinitions [{
      boundBone {boneName=dyn_hair_21}
      boxExtents=<22 22 22>
      constraintSetup {angularLimitsMax=<50 30 50>, angularLimitsMin=<-15 -15 0>, angularTarget=<1 0 0>}
      localJointOffset=<2 0 0>
     }]
    relativeSpaceBone {boneName=head}
   }
   n67 = Anim Dynamics - Bone: dyn_hair_4(ComponentPose=@n66, Alpha=1, ExternalForce="0,0,0") {
    angularDampingOverride=1
    angularSpringConstant=3000
    bAngularSpring=true
    bGravityOverrideInSimSpace=true
    bOverrideAngularDamping=true
    bUseGravityOverride=true
    gravityScale=0.35
    simulationSpace=BoneRelative
    alphaBoolBlend {blendOption=Linear}
    boundBone {boneName=dyn_hair_4}
    componentAppliedLinearAccClamp=<5000 5000 5000>
    componentLinearAccScale=<0.6 0.5 0.5>
    componentLinearVelScale=<0.6 0.5 0.5>
    gravityOverride {z=-900}
    physicsBodyDefinitions [{
      boundBone {boneName=dyn_hair_4}
      boxExtents=<22 22 22>
      constraintSetup {angularLimitsMax=<50 30 50>, angularLimitsMin=<-15 -15 0>, angularTarget=<1 0 0>}
      localJointOffset=<2 0 0>
     }]
    relativeSpaceBone {boneName=head}
   }
   n68 = Anim Dynamics - Bone: dyn_hair_22_bang(ComponentPose=@n67, Alpha=1, ExternalForce="0,0,0") {
    angularDampingOverride=1
    angularSpringConstant=3000
    bAngularSpring=true
    bOverrideAngularDamping=true
    bUseGravityOverride=true
    gravityScale=0.35
    simulationSpace=RootRelative
    alphaBoolBlend {blendOption=Linear}
    boundBone {boneName=dyn_hair_22_bang}
    componentAppliedLinearAccClamp=<5000 5000 5000>
    componentLinearAccScale=<0.3 0.3 0.3>
    componentLinearVelScale=<0.3 0.3 0.3>
    gravityOverride {z=-500}
    physicsBodyDefinitions [{
      boundBone {boneName=dyn_hair_22_bang}
      boxExtents=<22 22 22>
      constraintSetup {angularLimitsMax=<50 0 50>, angularLimitsMin=<-15 -50 0>, angularTarget=<0.5 0 0>}
      localJointOffset=<5 0 0>
     }]
   }
   n69 = Anim Dynamics - Bone: dyn_hair_22_bang1(ComponentPose=@n68, Alpha=1, ExternalForce="0,0,0") {
    angularDampingOverride=1
    angularSpringConstant=3000
    bAngularSpring=true
    bOverrideAngularDamping=true
    bUseGravityOverride=true
    gravityScale=0.35
    simulationSpace=RootRelative
    alphaBoolBlend {blendOption=Linear}
    boundBone {boneName=dyn_hair_22_bang1}
    componentAppliedLinearAccClamp=<5000 5000 5000>
    componentLinearAccScale=<0.3 0.3 0.3>
    componentLinearVelScale=<0.3 0.3 0.3>
    gravityOverride {z=-500}
    physicsBodyDefinitions [{
      boundBone {boneName=dyn_hair_22_bang1}
      boxExtents=<22 22 22>
      constraintSetup {angularLimitsMax=<50 0 50>, angularLimitsMin=<-15 -30 0>, angularTarget=<0.5 0 0>}
      localJointOffset=<2 0 0>
     }]
   }
   Output Pose(Result=Blend Poses by bool(
     BlendPose_0=Component To Local(ComponentPose=@n46)
     BlendPose_1=Component To Local(ComponentPose=@n69)
     BlendTime_0=0.5
     BlendTime_1=0.5
     bActiveValue=Get bIsInFrontEnd()
    ) {blendType=Linear}) {layerGroup=None}
   Save cached pose 'InputPose'(Pose=Copy Pose From Mesh() {bUseAttachedParent=true}) {cacheName=InputPose}
  ]}
 EventGraph {nodes [Event BlueprintUpdateAnimation()]}
]
variables []
